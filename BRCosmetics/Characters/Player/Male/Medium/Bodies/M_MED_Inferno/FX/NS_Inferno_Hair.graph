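# Niagara system rendering the Inferno character's flaming hair as mesh particles, grouped by spawn group and sampled from the Head bone.
# Two proxy emitters simulate lag behind head and owner motion, read via Particle Read into dynamic material parameters to bend the hair.
# System timelines drive brightness pulses and bursts from a Flame emitter.
# User parameters select color variants and head-hidden visibility; two older-style single-layer hair emitters are disabled.
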NiagaraSystem /BRCosmetics/Characters/Player/Male/Medium/Bodies/M_MED_Inferno/FX/NS_Inferno_Hair
dependencies {
 usedDataInterfaces [
  &NiagaraDataInterfaceParticleRead
  &NiagaraDataInterfaceCurve
  &NiagaraDataInterfaceCamera
  &NiagaraDataInterfaceStaticMesh
  &NiagaraDataInterfaceVector2DCurve
  &NiagaraDataInterfaceColorCurve
  &NiagaraDataInterfaceArrayFloat
 ]
 usedDynamicInputs [&Boolean_OrOperation, &ModuloFloat]
 usedModules [
  &self:SystemSpawnScript.SystemScriptSource.SystemScriptGraph.NiagaraNodeAssignment_2.SetVariables_#1
  &self:SystemSpawnScript.SystemScriptSource.SystemScriptGraph.NiagaraNodeAssignment_5.SetVariables_#2
  &self:SystemSpawnScript.SystemScriptSource.SystemScriptGraph.NiagaraNodeAssignment_9.SetVariables_#3
  &self:SystemSpawnScript.SystemScriptSource.SystemScriptGraph.NiagaraNodeAssignment_10.SetVariables_#4
  &self:SystemSpawnScript.SystemScriptSource.SystemScriptGraph.NiagaraNodeAssignment_11.SetVariables_#5
  &SystemState
  &Timeline
  &InterpolateOverTime
  &self:SystemSpawnScript.SystemScriptSource.SystemScriptGraph.NiagaraNodeAssignment_0.SetVariables_#6
  &self:SystemSpawnScript.SystemScriptSource.SystemScriptGraph.NiagaraNodeAssignment_1.SetVariables_#7
  &self:SystemSpawnScript.SystemScriptSource.SystemScriptGraph.NiagaraNodeAssignment_6.SetVariables_#8
  &DoOnce
  &self:SystemSpawnScript.SystemScriptSource.SystemScriptGraph.NiagaraNodeAssignment_3.SetVariables_#9
  &self:SystemSpawnScript.SystemScriptSource.SystemScriptGraph.NiagaraNodeAssignment_4.SetVariables_#10
  &self:SystemSpawnScript.SystemScriptSource.SystemScriptGraph.NiagaraNodeAssignment_7.SetVariables_#11
  &self:SystemSpawnScript.SystemScriptSource.SystemScriptGraph.NiagaraNodeAssignment_8.SetVariables_#12
 ]
}
emitters {
 Empty {
  emitterUpdateScript [
   &EmitterState(
    Inactive Response=::NewEnumerator1
    Life Cycle Mode=::NewEnumerator1
    Loop Behavior=::NewEnumerator1
    Loop Delay=0.0333
    Loop Duration Mode=::NewEnumerator1
    UseLoopDelay=true
   )
   &SpawnBurst_Instantaneous(Use Spawn Probability=false, UseLoopCountLimit=false)
   &SpawnBurst_Instantaneous(Spawn Group=1, Use Spawn Probability=false, UseLoopCountLimit=false) {module=SpawnBurst_Instantaneous001}
   &SpawnBurst_Instantaneous(Spawn Group=2, Use Spawn Probability=false, UseLoopCountLimit=false) {module=SpawnBurst_Instantaneous002}
   &SpawnBurst_Instantaneous(Spawn Count=1, Spawn Group=3, Use Spawn Probability=false) {module=SpawnBurst_Instantaneous003}
  ]
  particleSpawnScript [
   &InitializeParticle(Mesh Scale Mode=::NewEnumerator3, UsePositionOffset=false, Write Material Random=true)
   /SetVariables(Particles.VisibilityTag=&SpawnGroup()) {isSetParameters=true}
  ]
  particleUpdateScript [
   &ParticleState(Kill Particles When Lifetime Has Elapsed=false, Let Infinitely Lived Particles Die When Emitter Deactivates=true)
   &ScaleMeshSize(Scale Factor=&Multiply_VectorByFloat(
     Vector=&MergeHelper_VectorSwitch(0=<1 1 1>, 1=<1 1 1>, 2=<1 1 1>, 3=<1.1 1 1>, 4=<1 1 1>, Index=&SpawnGroup(), Num=::NewEnumerator3)
     Float=&Lerp_Float(A=1, B=2, Alpha=$NPC.FortniteNPC.BigHeadMode)
    ))
   &FN_Timer(Increment Multiplier:NiagaraFloat=&FloatFromCurve(
     FloatCurve=DataInterface() {propertyValues {Curve {keys [
         {interpMode=RCIM_Cubic, leaveTangent=8.8583, tangentMode=RCTM_Break, time=0, value=1}
         {interpMode=RCIM_Cubic, tangentMode=RCTM_Break, time=0.2, value=2.7}
         <at 1.95 1>
        ]}}}
     CurveIndex=$System.Timeline001.Time
    ))
   &SkeletalMeshLocation(
    Bone Sampling Mode=::NewEnumerator4
    MaskSpawnGroup=false
    Skeletal Mesh=DataInterface() {propertyValues {
      DefaultMesh=null
      ExcludeBoneName=root
      OptionalComponentClass=null
      PreviewMesh=/BRCosmetics/Characters/Player/Male/Medium/Bodies/M_MED_Inferno/Meshes/Parts/M_MED_Inferno_FaceAcc
      SkinningMode=SkinOnTheFly
      SoftSourceActor=null
      SourceMode=Default
      UvSetIndex=0
      WholeMeshLOD=-1
      bExcludeBone=false
      bReadDeformedGeometry=true
      bRequireCurrentFrameData=true
      ComponentTags []
      FilteredBones [Head]
      FilteredSockets []
      MeshUserParameter {parameter {name=None, typeDefHandle {registeredTypeIndex=1}, varData []}}
      SamplingRegions []
     }}
   )
   &self:ReadSimInformation(
    DefaultAxis:Vector3f=<0 0 1>
    HeadPos:Vector3f=&TransformPosition(Position=$Particles.SkeletalMeshLocation.SampledPosition, Source Space=::Simulation, Destination Space=::World)
    PA=DataInterface() {propertyValues {EmitterBinding {bindingMode=Other, emitterName=FeatherSimulationProto}}}
    Source Space:ENiagaraCoordinateSpace=::Local
   ) {module=ReadSimInformation}
   &self:ReadSimInformation(
    DefaultAxis:Vector3f=<0 0 1>
    HeadPos:Vector3f=&TransformPosition(Position=$Particles.SkeletalMeshLocation.SampledPosition, Source Space=::Simulation, Destination Space=::World)
    PA=DataInterface() {propertyValues {EmitterBinding {bindingMode=Other, emitterName=FeatherSimulationProto001}}}
    Source Space:ENiagaraCoordinateSpace=::World
   ) {module=ReadSimInformation001}
   &DynamicMaterialParameters(
    Index 0 Param 1=&MakeFloatFromVector(Vector=$Particles.ReadSimInformation.Direction, Channel=::NewEnumerator0)
    Index 0 Param 2=&MakeFloatFromVector(Vector=$Particles.ReadSimInformation.Direction, Channel=::NewEnumerator1)
    Index 0 Param 3=&MakeFloatFromVector(Vector=$Particles.ReadSimInformation.Direction, Channel=::NewEnumerator2)
    Index 0 Param 4=&MaskFloatBySpawnGroup(
     Masked Float=&FloatFromCurve(
      FloatCurve=DataInterface() {propertyValues {Curve {keys [
          {interpMode=RCIM_Cubic, leaveTangentWeight=0.3333, tangentMode=RCTM_Break, time=-1, value=1}
          {arriveTangent=-0.6795, interpMode=RCIM_Cubic, leaveTangent=0.6442, tangentMode=RCTM_Break, time=0, value=0.7}
          {arriveTangentWeight=0.3333, interpMode=RCIM_Cubic, tangentMode=RCTM_Break, time=1, value=1}
         ]}}}
      CurveIndex=&DotProduct(
       Vector A=&ReturnCameraVector(Camera Vectors=::NewEnumerator1, Coordinate Space=::World)
       Vector B=&RotateVectorByQuaternion(Vector To Rotate=<0 0 1>, Quaternion=$StackContext.SkeletalMeshLocation.SampledRotation)
      )
     )
     Passthrough Float=&FloatFromCurve(
      FloatCurve=DataInterface() {propertyValues {Curve {keys [<at -1 1>,<at 0 0>,<at 1 1>]}}}
      CurveIndex=&DotProduct(
       Vector A=&ReturnCameraVector(Camera Vectors=::NewEnumerator1, Coordinate Space=::World)
       Vector B=&MergeHelper_VectorSwitch(
        0=&TransformVector(Vector=<1 0 0>, Destination Space=::World)
        1=&RotateVectorByQuaternion(Vector To Rotate=<0 0 1>, Quaternion=$StackContext.SkeletalMeshLocation.SampledRotation)
        2=<0 0 0>
        3=<0 0 0>
        4=<0 0 0>
        Index=&SpawnGroup()
        Num=::NewEnumerator3
       )
      )
     )
     Spawn Group Mask=2
    )
    Index 1 Param 2=&MakeFloatFromVector(
     Vector=&MaskVectorBySpawnGroup(
      Masked Vector=Linked() {linkedVariable {name=Particles.ReadSimInformation001.Position, type=NiagaraPosition}}
      Passthrough Vector=$Particles.ReadSimInformation001.Direction
      Spawn Group Mask=2
     )
     Channel=::NewEnumerator0
    )
    Index 1 Param 3=&MakeFloatFromVector(
     Vector=&MaskVectorBySpawnGroup(
      Masked Vector=Linked() {linkedVariable {name=Particles.ReadSimInformation001.Position, type=NiagaraPosition}}
      Passthrough Vector=$Particles.ReadSimInformation001.Direction
      Spawn Group Mask=2
     )
     Channel=::NewEnumerator1
    )
    Index 1 Param 4=&MakeFloatFromVector(
     Vector=&MaskVectorBySpawnGroup(
      Masked Vector=Linked() {linkedVariable {name=Particles.ReadSimInformation001.Position, type=NiagaraPosition}}
      Passthrough Vector=$Particles.ReadSimInformation001.Direction
      Spawn Group Mask=2
     )
     Channel=::NewEnumerator2
    )
    Index 2 Param 1=&FloatFromCurve(
     FloatCurve=DataInterface() {propertyValues {Curve {keys [
         {interpMode=RCIM_Cubic, leaveTangent=15.0801, tangentMode=RCTM_Break, time=0.03, value=0}
         {arriveTangent=-1.2977, interpMode=RCIM_Cubic, leaveTangent=-1.2977, tangentMode=RCTM_User, time=0.13, value=1.0371}
         {arriveTangent=-0.0411, interpMode=RCIM_Cubic, tangentMode=RCTM_Break, time=1.95, value=0}
        ]}}}
     CurveIndex=$System.Timeline001.Time
    )
    Index 2 Param 2=&MaskFloatBySpawnGroup(
     Masked Float=&MakeFloatFromVector(Vector=$Particles.ReadSimInformation001.Direction, Channel=::NewEnumerator0)
     Passthrough Float=&FloatFromCurve(
      FloatCurve=DataInterface() {propertyValues {Curve {keys [
          {interpMode=RCIM_Cubic, leaveTangent=-8.5784, tangentMode=RCTM_Break, time=0, value=0}
          {arriveTangent=-3.655, interpMode=RCIM_Cubic, tangentMode=RCTM_Break, time=0.22, value=-1.3}
          <at 1 -1.4>
         ]}}}
      CurveIndex=$System.Timeline001.Time
     )
     Spawn Group Mask=2
    )
    Index 2 Param 3=&MaskFloatBySpawnGroup(
     Masked Float=&MakeFloatFromVector(Vector=$Particles.ReadSimInformation001.Direction, Channel=::NewEnumerator1)
     Passthrough Float=&FloatFromCurve(
      FloatCurve=DataInterface() {propertyValues {Curve {keys [
          <at 0.025 0 RCIM_Cubic>
          {arriveTangent=-0.2818, interpMode=RCIM_Cubic, leaveTangent=-0.2818, tangentMode=RCTM_User, time=0.1, value=1}
          {arriveTangent=-0.0382, interpMode=RCIM_Cubic, tangentMode=RCTM_Break, time=1.2, value=0}
         ]}}}
      CurveIndex=$System.Timeline001.Time
      Scale Curve=&MergeHelper_FloatSwitch(0=0.6, 1=1, 2=0, 3=0, 4=0, Index=&SpawnGroup(), Num=::NewEnumerator3)
     )
     Spawn Group Mask=2
    )
    Index 2 Param 4=&MaskFloatBySpawnGroup(
     Masked Float=&MakeFloatFromVector(Vector=$Particles.ReadSimInformation001.Direction, Channel=::NewEnumerator2)
     Passthrough Float=0
     Spawn Group Mask=2
    )
    Index 3 Param 1=&MaskFloatBySpawnGroup(
     Masked Float=&MakeCustomFloatFromBool(Bool=$User.bIsFrontEndPreview, True Float=1.5, False Float=1)
     Passthrough Float=1
     Spawn Group Mask=2
    )
    Index 3 Param 2=&MergeHelper_FloatSwitch(
     0=$System.Timeline.Time
     1=$System.Timeline.Time
     2=$Particles.FN_Timer.Value
     3=0
     4=0
     Index=&SpawnGroup()
     Num=::NewEnumerator3
    )
    Index 3 Param 3=&MergeHelper_FloatSwitch(
     0=$System.InterpolateOverTime.InterpolateOverTime_Float.Moving Average
     1=$System.InterpolateOverTime.InterpolateOverTime_Float.Moving Average
     2=0
     3=0
     4=0
     Index=&SpawnGroup()
     Num=::NewEnumerator3
    )
    Index 3 Param 4=&MergeHelper_FloatSwitch(
     0=$System.Timeline002.Time
     1=$System.Timeline002.Time
     2=0
     3=0
     4=0
     Index=&SpawnGroup()
     Num=::NewEnumerator3
    )
    Module.Param0WriteEnabled=true
    Module.Param0WriteEnabled1=true
    Module.Param0WriteEnabled2=true
    Module.Param0WriteEnabled3=true
    Module.Param1WriteEnabled=true
    Module.Param1WriteEnabled1=true
    Module.Param1WriteEnabled2=true
    Module.Param1WriteEnabled3=true
    Module.Param2WriteEnabled=true
    Module.Param2WriteEnabled1=true
    Module.Param2WriteEnabled2=true
    Module.Param2WriteEnabled3=true
    Module.Param3WriteEnabled=true
    Module.Param3WriteEnabled1=true
    Module.Param3WriteEnabled2=true
    Module.Param3WriteEnabled3=true
    Write Parameter Index 1=true
    Write Parameter Index 2=true
    Write Parameter Index 3=true
   )
   &ScaleColor(
    Scale Mode=::NewEnumerator1
    Scale RGBA=/Game/Effects/Niagara/Modules/MergeEmitters/MergeHelper_Vector4Switch(
     0:Vector4f=&MakeFloat4FromVectorAndFloat(
      "Vector (RGB)"=&VectorFromFloat(Value=&FloatFromCurve(
        FloatCurve=DataInterface() {propertyValues {Curve {keys [
            {interpMode=RCIM_Cubic, leaveTangent=5.9752, tangentMode=RCTM_Break, time=0.0703, value=1}
            {arriveTangent=-0.4141, interpMode=RCIM_Cubic, leaveTangent=-0.4141, tangentMode=RCTM_User, time=0.1697, value=1.6}
            {arriveTangent=-0.0411, interpMode=RCIM_Cubic, tangentMode=RCTM_Break, time=1.3, value=1}
           ]}}}
        CurveIndex=$System.Timeline001.Time
       ))
      "Float (Alpha)"=1
     )
     1:Vector4f=&MakeFloat4FromVectorAndFloat(
      "Vector (RGB)"=&VectorFromFloat(Value=&FloatFromCurve(
        FloatCurve=DataInterface() {propertyValues {Curve {keys [
            {interpMode=RCIM_Cubic, leaveTangent=5.9752, tangentMode=RCTM_Break, time=0.0703, value=1}
            {arriveTangent=-0.4141, interpMode=RCIM_Cubic, leaveTangent=-0.4141, tangentMode=RCTM_User, time=0.1697, value=1.2}
            {arriveTangent=-0.0411, interpMode=RCIM_Cubic, tangentMode=RCTM_Break, time=1.3, value=1}
           ]}}}
        CurveIndex=$System.Timeline001.Time
       ))
      "Float (Alpha)"=1
     )
     2:Vector4f=&MakeFloat4FromVectorAndFloat(
      "Vector (RGB)"=&VectorFromFloat(Value=&FloatFromCurve(
        FloatCurve=DataInterface() {propertyValues {Curve {keys [
            {interpMode=RCIM_Cubic, leaveTangent=15.0801, tangentMode=RCTM_Break, time=0.03, value=1}
            {arriveTangent=-0.772, interpMode=RCIM_Cubic, leaveTangent=-0.772, tangentMode=RCTM_User, time=0.13, value=2}
            {arriveTangent=-0.0411, interpMode=RCIM_Cubic, tangentMode=RCTM_Break, time=1.95, value=1}
           ]}}}
        CurveIndex=$System.Timeline001.Time
       ))
      "Float (Alpha)"=&FloatFromCurve(
       FloatCurve=DataInterface() {propertyValues {Curve {keys [
           {interpMode=RCIM_Cubic, leaveTangent=15.0801, tangentMode=RCTM_Break, time=0.03, value=1}
           {arriveTangent=-0.16, interpMode=RCIM_Cubic, leaveTangent=-0.16, tangentMode=RCTM_User, time=0.13, value=3}
           {arriveTangent=-0.0411, interpMode=RCIM_Cubic, tangentMode=RCTM_Break, time=1.95, value=1}
          ]}}}
       CurveIndex=$System.Timeline001.Time
       Scale Curve=2
      )
     )
     3:Vector4f=<1 1 1 1>
     4:Vector4f=<1 1 1 1>
     5:Vector4f=?
     6:Vector4f=?
     7:Vector4f=?
     8:Vector4f=?
     9:Vector4f=?
     Index:NiagaraInt32=&SpawnGroup()
     Num:MergeHelper_SwitcherEnum=::NewEnumerator3
    )
    ScaleRGBA=true
   )
   /SetVariables(Particles.New float=2.5) {isSetParameters=true}
   /SetVariables(Particles.VisibilityTag=&SpawnGroup()) {isSetParameters=true}
   /SetVariables(Particles.VisibilityTag=&MakeCustomIntFromBool(Bool=$User.bisHeadHidden, True INT=999, False INT=$Particles.VisibilityTag)) {isSetParameters=true, module=SetVariables_#13}
   /SetVariables(Particles.Color2=&MakeLinearColorFromVectorAndFloat(
     "Vector (RGB)"=&MakeCustomVectorFromBool(
      Bool=$User.bIsSL
      True Vector=&MakeVectorFromLinearColorRGB(Color=&self:SelectColor(
        Color1:LinearColor=<rgba 100 0 20.2548 1>
        Color2:LinearColor=<rgba 15.8333 25 0 1>
        Color3:LinearColor=<rgba 100 16.6667 0 1>
        Index:NiagaraInt32=$User.SuperLevelIndex
       ))
      False Vector=&MakeVectorFromLinearColorRGB(Color=&MakeLinearColorFromVectorAndFloat(
        "Vector (RGB)"=&MakeVectorFromLinearColorRGB(Color=&MakeCustomLinearColorFromBool(
          Bool=$User.bIsWhiteVersion
          True Color=&MakeCustomLinearColorFromBool(Bool=$User.bIsFrontEndPreview, True Color=<rgba 200 65 20 1>, False Color=<rgba 200 65 20 1>)
          False Color=&MakeCustomLinearColorFromBool(Bool=$User.bIsFrontEndPreview, True Color=<rgba 0 400 297.6663 1>, False Color=<rgba 0 200 148.8331 1>)
         ))
        "Float (Alpha)"=1
       ))
     )
     "Float (Alpha)"=&FloatFromCurve(
      FloatCurve=DataInterface() {propertyValues {Curve {keys [
          {interpMode=RCIM_Cubic, leaveTangent=-2.1302, tangentMode=RCTM_Break, time=0.0703, value=1}
          {arriveTangent=0.3473, interpMode=RCIM_Cubic, leaveTangent=0.3473, tangentMode=RCTM_User, time=0.1697, value=0}
          {arriveTangent=-0.0411, interpMode=RCIM_Cubic, tangentMode=RCTM_Break, time=1.95, value=1}
         ]}}}
      CurveIndex=$System.Timeline001.Time
      Scale Curve=&Multiply_Float(A=0.05, B=0.2)
     )
    )) {isSetParameters=true}
   /SetVariables(Particles.DPfor2:Vector4f=&MakeVector4(X=&RandomRangeFloat(Minimum=0, Maximum=1), Y=1, Z=4.35, W=0)) {isSetParameters=true}
  ]
  rendererClasses [&NiagaraMeshRendererProperties]
  renderers [
   {rendererClass=&NiagaraMeshRendererProperties, rendererIndex=0}
   {rendererClass=&NiagaraMeshRendererProperties, rendererIndex=1}
   {rendererClass=&NiagaraMeshRendererProperties, rendererIndex=2}
   {rendererClass=&NiagaraMeshRendererProperties, rendererIndex=3}
   {rendererClass=&NiagaraMeshRendererProperties, rendererIndex=4}
   {rendererClass=&NiagaraMeshRendererProperties, rendererIndex=5}
   {rendererClass=&NiagaraMeshRendererProperties, rendererIndex=6}
   {rendererClass=&NiagaraMeshRendererProperties, rendererIndex=7}
   {rendererClass=&NiagaraMeshRendererProperties, rendererIndex=8}
  ]
 }
 Empty001 {
  bEnabled=false
  emitterUpdateScript [
   &EmitterState(
    Inactive Response=::NewEnumerator1
    Life Cycle Mode=::NewEnumerator1
    Loop Behavior=::NewEnumerator1
    Loop Delay=0.0333
    Loop Duration Mode=::NewEnumerator1
    UseLoopDelay=true
   )
   &SpawnBurst_Instantaneous(Use Spawn Probability=false, UseLoopCountLimit=false)
  ]
  particleSpawnScript [&InitializeParticle(UsePositionOffset=false, Write Material Random=true)]
  particleUpdateScript [
   &ParticleState(Kill Particles When Lifetime Has Elapsed=false, Let Infinitely Lived Particles Die When Emitter Deactivates=true)
   &SkeletalMeshLocation(
    Bone Sampling Mode=::NewEnumerator4
    MaskSpawnGroup=false
    Skeletal Mesh=DataInterface() {propertyValues {
      DefaultMesh=null
      ExcludeBoneName=root
      OptionalComponentClass=null
      PreviewMesh=&M_MED_Inferno_FaceAcc
      SkinningMode=SkinOnTheFly
      SoftSourceActor=null
      SourceMode=Default
      UvSetIndex=0
      WholeMeshLOD=-1
      bExcludeBone=false
      bReadDeformedGeometry=true
      bRequireCurrentFrameData=true
      ComponentTags []
      FilteredBones [Head]
      FilteredSockets []
      MeshUserParameter {parameter {name=None, typeDefHandle {registeredTypeIndex=1}, varData []}}
      SamplingRegions []
     }}
   )
   &self:ReadSimInformation(
    DefaultAxis:Vector3f=<0 0 1>
    HeadPos:Vector3f=&TransformPosition(Position=$Particles.SkeletalMeshLocation.SampledPosition, Source Space=::Simulation, Destination Space=::World)
    PA=DataInterface() {propertyValues {EmitterBinding {bindingMode=Other, emitterName=FeatherSimulationProto}}}
    Source Space:ENiagaraCoordinateSpace=::Local
   ) {module=ReadSimInformation}
   &self:ReadSimInformation(
    DefaultAxis:Vector3f=<0 0 1>
    HeadPos:Vector3f=&TransformPosition(Position=$Particles.SkeletalMeshLocation.SampledPosition, Source Space=::Simulation, Destination Space=::World)
    PA=DataInterface() {propertyValues {EmitterBinding {bindingMode=Other, emitterName=FeatherSimulationProto001}}}
    Source Space:ENiagaraCoordinateSpace=::World
   ) {module=ReadSimInformation001}
   &DynamicMaterialParameters(
    Index 0 Param 1=&MakeFloatFromVector(Vector=$Particles.ReadSimInformation.Direction, Channel=::NewEnumerator0)
    Index 0 Param 2=&MakeFloatFromVector(Vector=$Particles.ReadSimInformation.Direction, Channel=::NewEnumerator1)
    Index 0 Param 3=&MakeFloatFromVector(Vector=$Particles.ReadSimInformation.Direction, Channel=::NewEnumerator2)
    Index 0 Param 4=&FloatFromCurve(
     FloatCurve=DataInterface() {propertyValues {Curve {keys [
         {interpMode=RCIM_Cubic, tangentMode=RCTM_Break, time=-1, value=1}
         {arriveTangent=-0.3929, interpMode=RCIM_Cubic, leaveTangent=0.3894, tangentMode=RCTM_Break, time=0, value=0.5}
         {interpMode=RCIM_Cubic, tangentMode=RCTM_Break, time=1, value=1}
        ]}}}
     CurveIndex=&DotProduct(
      Vector A=&ReturnCameraVector(Camera Vectors=::NewEnumerator1, Coordinate Space=::World)
      Vector B=&RotateVectorByQuaternion(Vector To Rotate=<0 0 1>, Quaternion=$StackContext.SkeletalMeshLocation.SampledRotation)
     )
    )
    Index 1 Param 2=&MakeFloatFromVector(Vector=$Particles.ReadSimInformation001.Direction, Channel=::NewEnumerator0)
    Index 1 Param 3=&MakeFloatFromVector(Vector=$Particles.ReadSimInformation001.Direction, Channel=::NewEnumerator1)
    Index 1 Param 4=&MakeFloatFromVector(Vector=$Particles.ReadSimInformation001.Direction, Channel=::NewEnumerator2)
    Index 2 Param 1=&FloatFromCurve(
     FloatCurve=DataInterface() {propertyValues {Curve {keys [
         {interpMode=RCIM_Cubic, leaveTangent=15.0801, tangentMode=RCTM_Break, time=0.0703, value=0}
         {arriveTangent=-1.2977, interpMode=RCIM_Cubic, leaveTangent=-1.2977, tangentMode=RCTM_User, time=0.13, value=1.0371}
         {arriveTangent=-0.0411, interpMode=RCIM_Cubic, tangentMode=RCTM_Break, time=1.95, value=0}
        ]}}}
     CurveIndex=$System.Timeline001.Time
    )
    Index 2 Param 2=&FloatFromCurve(
     FloatCurve=DataInterface() {propertyValues {Curve {keys [
         {interpMode=RCIM_Cubic, leaveTangent=-8.5784, tangentMode=RCTM_Break, time=0, value=0}
         {arriveTangent=-3.655, interpMode=RCIM_Cubic, tangentMode=RCTM_Break, time=0.3, value=-1.3}
         <at 1 -1.4>
        ]}}}
     CurveIndex=$System.Timeline001.Time
    )
    Index 2 Param 3=&FloatFromCurve(
     FloatCurve=DataInterface() {propertyValues {Curve {keys [
         <at 0.025 0 RCIM_Cubic>
         {arriveTangent=-0.2818, interpMode=RCIM_Cubic, leaveTangent=-0.2818, tangentMode=RCTM_User, time=0.1, value=1}
         {arriveTangent=-0.0382, interpMode=RCIM_Cubic, tangentMode=RCTM_Break, time=1.2, value=0}
        ]}}}
     CurveIndex=$System.Timeline001.Time
     Scale Curve=0.6
    )
    Index 3 Param 1=1
    Index 3 Param 2=$System.Timeline.Time
    Index 3 Param 3=$System.InterpolateOverTime.InterpolateOverTime_Float.Moving Average
    Index 3 Param 4=$System.Timeline002.Time
    Module.Param0WriteEnabled=true
    Module.Param0WriteEnabled1=true
    Module.Param0WriteEnabled2=true
    Module.Param0WriteEnabled3=true
    Module.Param1WriteEnabled=true
    Module.Param1WriteEnabled1=true
    Module.Param1WriteEnabled2=true
    Module.Param1WriteEnabled3=true
    Module.Param2WriteEnabled=true
    Module.Param2WriteEnabled1=true
    Module.Param2WriteEnabled2=true
    Module.Param2WriteEnabled3=true
    Module.Param3WriteEnabled=true
    Module.Param3WriteEnabled1=true
    Module.Param3WriteEnabled2=true
    Module.Param3WriteEnabled3=true
    Write Parameter Index 1=true
    Write Parameter Index 2=true
    Write Parameter Index 3=true
   )
   &ScaleColor(
    Scale Mode=::NewEnumerator1
    Scale RGBA=&MakeFloat4FromVectorAndFloat(
     "Vector (RGB)"=&VectorFromFloat(Value=&FloatFromCurve(
       FloatCurve=DataInterface() {propertyValues {Curve {keys [
           {interpMode=RCIM_Cubic, leaveTangent=5.9752, tangentMode=RCTM_Break, time=0.0703, value=1}
           {arriveTangent=-0.4141, interpMode=RCIM_Cubic, leaveTangent=-0.4141, tangentMode=RCTM_User, time=0.1697, value=1.2}
           {arriveTangent=-0.0411, interpMode=RCIM_Cubic, tangentMode=RCTM_Break, time=1.3, value=1}
          ]}}}
       CurveIndex=$System.Timeline001.Time
      ))
     "Float (Alpha)"=1
    )
    ScaleRGBA=true
   )
   /SetVariables(Particles.VisibilityTag=&MakeCustomIntFromBool(Bool=$User.bisHeadHidden, True INT=1, False INT=0)) {isSetParameters=true, module=SetVariables_#13}
  ]
  rendererClasses [&NiagaraMeshRendererProperties]
  renderers [{rendererClass=&NiagaraMeshRendererProperties, rendererIndex=0}]
 }
 Empty002 {
  bEnabled=false
  emitterUpdateScript [
   &EmitterState(
    Inactive Response=::NewEnumerator1
    Life Cycle Mode=::NewEnumerator1
    Loop Behavior=::NewEnumerator1
    Loop Delay=0.0333
    Loop Duration Mode=::NewEnumerator1
    UseLoopDelay=true
   )
   &SpawnBurst_Instantaneous(Use Spawn Probability=false, UseLoopCountLimit=false)
  ]
  particleSpawnScript [&InitializeParticle(UsePositionOffset=false, Write Material Random=true)]
  particleUpdateScript [
   &ParticleState(Kill Particles When Lifetime Has Elapsed=false, Let Infinitely Lived Particles Die When Emitter Deactivates=true)
   &SkeletalMeshLocation(
    Bone Sampling Mode=::NewEnumerator4
    MaskSpawnGroup=false
    Skeletal Mesh=DataInterface() {propertyValues {
      DefaultMesh=null
      ExcludeBoneName=root
      OptionalComponentClass=null
      PreviewMesh=&M_MED_Inferno_FaceAcc
      SkinningMode=SkinOnTheFly
      SoftSourceActor=null
      SourceMode=Default
      UvSetIndex=0
      WholeMeshLOD=-1
      bExcludeBone=false
      bReadDeformedGeometry=true
      bRequireCurrentFrameData=true
      ComponentTags []
      FilteredBones [Head]
      FilteredSockets []
      MeshUserParameter {parameter {name=None, typeDefHandle {registeredTypeIndex=1}, varData []}}
      SamplingRegions []
     }}
   )
   &self:ReadSimInformation(
    DefaultAxis:Vector3f=<0 0 1>
    HeadPos:Vector3f=&TransformPosition(Position=$Particles.SkeletalMeshLocation.SampledPosition, Source Space=::Simulation, Destination Space=::World)
    PA=DataInterface() {propertyValues {EmitterBinding {bindingMode=Other, emitterName=FeatherSimulationProto}}}
    Source Space:ENiagaraCoordinateSpace=::Local
   ) {module=ReadSimInformation}
   &self:ReadSimInformation(
    DefaultAxis:Vector3f=<0 0 1>
    HeadPos:Vector3f=&TransformPosition(Position=$Particles.SkeletalMeshLocation.SampledPosition, Source Space=::Simulation, Destination Space=::World)
    PA=DataInterface() {propertyValues {EmitterBinding {bindingMode=Other, emitterName=FeatherSimulationProto001}}}
    Source Space:ENiagaraCoordinateSpace=::World
   ) {module=ReadSimInformation001}
   &FN_Timer(Increment Multiplier:NiagaraFloat=&FloatFromCurve(
     FloatCurve=DataInterface() {propertyValues {Curve {keys [
         {interpMode=RCIM_Cubic, leaveTangent=8.8583, tangentMode=RCTM_Break, time=0, value=1}
         {interpMode=RCIM_Cubic, tangentMode=RCTM_Break, time=0.2, value=2.7}
         <at 1.95 1>
        ]}}}
     CurveIndex=$System.Timeline001.Time
    ))
   &DynamicMaterialParameters(
    Index 0 Param 1=&MakeFloatFromVector(Vector=$Particles.ReadSimInformation.Direction, Channel=::NewEnumerator0)
    Index 0 Param 2=&MakeFloatFromVector(Vector=$Particles.ReadSimInformation.Direction, Channel=::NewEnumerator1)
    Index 0 Param 3=&MakeFloatFromVector(Vector=$Particles.ReadSimInformation.Direction, Channel=::NewEnumerator2)
    Index 0 Param 4=&FloatFromCurve(
     FloatCurve=DataInterface() {propertyValues {Curve {keys [
         {interpMode=RCIM_Cubic, leaveTangentWeight=0.3333, tangentMode=RCTM_Break, time=-1, value=1}
         {arriveTangent=-0.6795, interpMode=RCIM_Cubic, leaveTangent=0.6442, tangentMode=RCTM_Break, time=0, value=0.7}
         {arriveTangentWeight=0.3333, interpMode=RCIM_Cubic, tangentMode=RCTM_Break, time=1, value=1}
        ]}}}
     CurveIndex=&DotProduct(
      Vector A=&ReturnCameraVector(Camera Vectors=::NewEnumerator1, Coordinate Space=::World)
      Vector B=&RotateVectorByQuaternion(Vector To Rotate=<0 0 1>, Quaternion=$StackContext.SkeletalMeshLocation.SampledRotation)
     )
    )
    Index 1 Param 2=&MakeFloatFromVector(Vector=$Particles.ReadSimInformation001.Position, Channel=::NewEnumerator0)
    Index 1 Param 3=&MakeFloatFromVector(Vector=$Particles.ReadSimInformation001.Position, Channel=::NewEnumerator1)
    Index 1 Param 4=&MakeFloatFromVector(Vector=$Particles.ReadSimInformation001.Position, Channel=::NewEnumerator2)
    Index 2 Param 1=&FloatFromCurve(
     FloatCurve=DataInterface() {propertyValues {Curve {keys [
         {interpMode=RCIM_Cubic, leaveTangent=15.0801, tangentMode=RCTM_Break, time=0.03, value=0}
         {arriveTangent=-1.2977, interpMode=RCIM_Cubic, leaveTangent=-1.2977, tangentMode=RCTM_User, time=0.13, value=1.0371}
         {arriveTangent=-0.0411, interpMode=RCIM_Cubic, tangentMode=RCTM_Break, time=1.95, value=0}
        ]}}}
     CurveIndex=$System.Timeline001.Time
    )
    Index 2 Param 2=&MakeFloatFromVector(Vector=$Particles.ReadSimInformation001.Direction, Channel=::NewEnumerator0)
    Index 2 Param 3=&MakeFloatFromVector(Vector=$Particles.ReadSimInformation001.Direction, Channel=::NewEnumerator1)
    Index 2 Param 4=&MakeFloatFromVector(Vector=$Particles.ReadSimInformation001.Direction, Channel=::NewEnumerator2)
    Index 3 Param 1=&MakeCustomFloatFromBool(Bool=$User.bIsFrontEndPreview, True Float=1.5, False Float=1)
    Index 3 Param 2=$Particles.FN_Timer.Value
    Module.Param0WriteEnabled=true
    Module.Param0WriteEnabled1=true
    Module.Param0WriteEnabled2=true
    Module.Param0WriteEnabled3=true
    Module.Param1WriteEnabled=true
    Module.Param1WriteEnabled1=true
    Module.Param1WriteEnabled2=true
    Module.Param1WriteEnabled3=true
    Module.Param2WriteEnabled=true
    Module.Param2WriteEnabled1=true
    Module.Param2WriteEnabled2=true
    Module.Param2WriteEnabled3=true
    Module.Param3WriteEnabled=true
    Module.Param3WriteEnabled1=true
    Module.Param3WriteEnabled2=true
    Module.Param3WriteEnabled3=true
    Write Parameter Index 1=true
    Write Parameter Index 2=true
    Write Parameter Index 3=true
   )
   &ScaleColor(
    Scale Mode=::NewEnumerator1
    Scale RGBA=&MakeFloat4FromVectorAndFloat(
     "Vector (RGB)"=&VectorFromFloat(Value=&FloatFromCurve(
       FloatCurve=DataInterface() {propertyValues {Curve {keys [
           {interpMode=RCIM_Cubic, leaveTangent=15.0801, tangentMode=RCTM_Break, time=0.03, value=1}
           {arriveTangent=-0.772, interpMode=RCIM_Cubic, leaveTangent=-0.772, tangentMode=RCTM_User, time=0.13, value=2}
           {arriveTangent=-0.0411, interpMode=RCIM_Cubic, tangentMode=RCTM_Break, time=1.95, value=1}
          ]}}}
       CurveIndex=$System.Timeline001.Time
      ))
     "Float (Alpha)"=&FloatFromCurve(
      FloatCurve=DataInterface() {propertyValues {Curve {keys [
          {interpMode=RCIM_Cubic, leaveTangent=15.0801, tangentMode=RCTM_Break, time=0.03, value=1}
          {arriveTangent=-0.16, interpMode=RCIM_Cubic, leaveTangent=-0.16, tangentMode=RCTM_User, time=0.13, value=3}
          {arriveTangent=-0.0411, interpMode=RCIM_Cubic, tangentMode=RCTM_Break, time=1.95, value=1}
         ]}}}
      CurveIndex=$System.Timeline001.Time
      Scale Curve=2
     )
    )
    ScaleRGBA=true
   )
   /SetVariables(Particles.VisibilityTag=&MakeCustomIntFromBool(Bool=$User.bisHeadHidden, True INT=1, False INT=0)) {isSetParameters=true, module=SetVariables_#13}
  ]
  rendererClasses [&NiagaraMeshRendererProperties]
  renderers [
   {rendererClass=&NiagaraMeshRendererProperties, rendererIndex=0}
   {rendererClass=&NiagaraMeshRendererProperties, rendererIndex=1}
   {rendererClass=&NiagaraMeshRendererProperties, rendererIndex=2}
  ]
 }
 FeatherSimulationProto {
  emitterUpdateScript [
   &EmitterState(
    Inactive Response=::NewEnumerator1
    Life Cycle Mode=::NewEnumerator1
    Loop Behavior=::NewEnumerator1
    Loop Duration Mode=::NewEnumerator1
    UseLoopDelay=false
   )
   &SpawnBurst_Instantaneous(Use Spawn Probability=false, UseLoopCountLimit=false)
  ]
  particleSpawnScript [
   &InitializeParticle(Color Mode=::NewEnumerator0, Lifetime=5, Sprite Size Mode=::NewEnumerator3, UsePositionOffset=false, Write Material Random=true)
   &SkeletalMeshLocation(
    Bone Sampling Mode=::NewEnumerator4
    MaskSpawnGroup=false
    Orientation Sampling=::NewEnumerator1
    Position Sampling=::NewEnumerator2
    Skeletal Mesh=DataInterface() {propertyValues {
      DefaultMesh=null
      ExcludeBoneName=root
      OptionalComponentClass=null
      PreviewMesh=&M_MED_Inferno_FaceAcc
      SkinningMode=SkinOnTheFly
      SoftSourceActor=null
      SourceMode=Default
      UvSetIndex=0
      WholeMeshLOD=-1
      bExcludeBone=false
      bReadDeformedGeometry=true
      bRequireCurrentFrameData=true
      ComponentTags []
      FilteredBones [Head]
      FilteredSockets []
      MeshUserParameter {parameter {name=None, typeDefHandle {registeredTypeIndex=1}, varData []}}
      SamplingRegions []
     }}
   ) {module=SkeletalMeshLocation001}
  ]
  particleUpdateScript [
   &ParticleState(Kill Particles When Lifetime Has Elapsed=false, Let Infinitely Lived Particles Die When Emitter Deactivates=true)
   &SkeletalMeshLocation(
    Bone Sampling Mode=::NewEnumerator4
    MaskSpawnGroup=false
    Orientation Sampling=::NewEnumerator1
    Position Sampling=::NewEnumerator2
    Skeletal Mesh=DataInterface() {propertyValues {
      DefaultMesh=null
      ExcludeBoneName=root
      OptionalComponentClass=null
      PreviewMesh=&M_MED_Inferno_FaceAcc
      SkinningMode=SkinOnTheFly
      SoftSourceActor=null
      SourceMode=Default
      UvSetIndex=0
      WholeMeshLOD=-1
      bExcludeBone=false
      bReadDeformedGeometry=true
      bRequireCurrentFrameData=true
      ComponentTags []
      FilteredBones [Head]
      FilteredSockets []
      MeshUserParameter {parameter {name=None, typeDefHandle {registeredTypeIndex=1}, varData []}}
      SamplingRegions []
     }}
   )
   &self:LimitRange(Center:NiagaraPosition=$Particles.SkeletalMeshLocation.SampledPosition, Pos:NiagaraPosition=$Particles.Position, Radius=35) {module=LimitRange}
   /SetVariables(Particles.Position=&AddVectorToPosition(
     Position=$Particles.Position
     Vector=&Multiply_VectorByFloat(
      Vector=&NormalizeVector(
       Vector To Normalize=&Subtract_Position(A=$Particles.SkeletalMeshLocation.SampledPosition, B=$Particles.Position)
       Safe Normalize=true
       Safe Fallback Threshold=0
       Safe Fallback Vector=<0 0 0.001>
      )
      Float=&MakeCustomFloatFromBool(
       Bool=&SetBoolByFloatComparison(A=&VectorLength(Vector=$Engine.Owner.Velocity), B=100)
       True Float=&MakeCustomFloatFromBool(
        Bool=&SetBoolByFloatComparison(
         A=&DotProduct(
          Vector A=&NormalizeVector(
           Vector To Normalize=&Subtract_Position(A=$Particles.SkeletalMeshLocation.SampledPosition, B=$Particles.Position)
           Safe Normalize=true
           Safe Fallback Threshold=0
           Safe Fallback Vector=<0 0 0.001>
          )
          Vector B=&NormalizeVector(
           Vector To Normalize=&TransformVector(Vector=$Engine.Owner.Velocity, Source Space=::World, Destination Space=::Local)
           Safe Normalize=true
           Safe Fallback Threshold=0
           Safe Fallback Vector=<0 0 0.001>
          )
         )
         B=0
         Comparison Type=::NewEnumerator6
        )
        True Float=&Divide_Float(
         A=&VectorLength(Vector=&Subtract_Position(A=$Particles.SkeletalMeshLocation.SampledPosition, B=$Particles.Position))
         B=25
        )
        False Float=0
       )
       False Float=&Divide_Float(
        A=&VectorLength(Vector=&Subtract_Position(A=$Particles.SkeletalMeshLocation.SampledPosition, B=$Particles.Position))
        B=25
       )
      )
     )
    )) {isSetParameters=true}
   /SetVariables(Particles.Position=&AddVectorToPosition(
     Position=$Particles.Position
     Vector=&Multiply_VectorByFloat(
      Vector=&NormalizeVector(
       Vector To Normalize=&TransformVector(Vector=$Engine.Owner.Velocity, Source Space=::World, Destination Space=::Local)
       Safe Normalize=true
       Safe Fallback Threshold=0
       Safe Fallback Vector=<0 0 0.001>
      )
      Float=&MakeCustomFloatFromBool(
       Bool=&SetBoolByFloatComparison(A=&VectorLength(Vector=$Engine.Owner.Velocity), B=100)
       True Float=&Multiply_Float(
        A=&Divide_Float(
         A=&VectorLength(Vector=&TransformVector(Vector=$Engine.Owner.Velocity, Source Space=::World, Destination Space=::Local))
         B=-200
        )
        B=1
       )
       False Float=0
      )
     )
    )) {isSetParameters=true}
  ]
  rendererClasses [&NiagaraSpriteRendererProperties]
  renderers [{rendererClass=&NiagaraSpriteRendererProperties}]
 }
 FeatherSimulationProto001 {
  emitterUpdateScript [
   &EmitterState(
    Inactive Response=::NewEnumerator1
    Life Cycle Mode=::NewEnumerator1
    Loop Behavior=::NewEnumerator1
    Loop Duration Mode=::NewEnumerator1
    UseLoopDelay=false
   )
   &SpawnBurst_Instantaneous(Use Spawn Probability=false, UseLoopCountLimit=false)
  ]
  particleSpawnScript [
   &InitializeParticle(Color Mode=::NewEnumerator0, Lifetime=5, Sprite Size Mode=::NewEnumerator3, UsePositionOffset=false, Write Material Random=true)
   &SkeletalMeshLocation(
    Bone Sampling Mode=::NewEnumerator4
    MaskSpawnGroup=false
    Orientation Sampling=::NewEnumerator1
    Position Sampling=::NewEnumerator2
    Skeletal Mesh=DataInterface() {propertyValues {
      DefaultMesh=null
      ExcludeBoneName=root
      OptionalComponentClass=null
      PreviewMesh=&M_MED_Inferno_FaceAcc
      SkinningMode=SkinOnTheFly
      SoftSourceActor=null
      SourceMode=Default
      UvSetIndex=0
      WholeMeshLOD=-1
      bExcludeBone=false
      bReadDeformedGeometry=true
      bRequireCurrentFrameData=true
      ComponentTags []
      FilteredBones [Head]
      FilteredSockets []
      MeshUserParameter {parameter {name=None, typeDefHandle {registeredTypeIndex=1}, varData []}}
      SamplingRegions []
     }}
   ) {module=SkeletalMeshLocation001}
  ]
  particleUpdateScript [
   &ParticleState(Kill Particles When Lifetime Has Elapsed=false, Let Infinitely Lived Particles Die When Emitter Deactivates=true)
   &SkeletalMeshLocation(
    Bone Sampling Mode=::NewEnumerator4
    MaskSpawnGroup=false
    Orientation Sampling=::NewEnumerator1
    Position Sampling=::NewEnumerator2
    Skeletal Mesh=DataInterface() {propertyValues {
      DefaultMesh=null
      ExcludeBoneName=root
      OptionalComponentClass=null
      PreviewMesh=&M_MED_Inferno_FaceAcc
      SkinningMode=SkinOnTheFly
      SoftSourceActor=null
      SourceMode=Default
      UvSetIndex=0
      WholeMeshLOD=-1
      bExcludeBone=false
      bReadDeformedGeometry=true
      bRequireCurrentFrameData=true
      ComponentTags []
      FilteredBones [Head]
      FilteredSockets []
      MeshUserParameter {parameter {name=None, typeDefHandle {registeredTypeIndex=1}, varData []}}
      SamplingRegions []
     }}
   )
   &PointAttractionForce(
    Attraction Radius=100000
    Attraction Strength=&Lerp_Float(
     A=0
     B=300
     Alpha=&ClampFloat(Float=&Divide_Float(
       A=&VectorLength(Vector=&SubtractVectorFromPosition(Position=$Particles.Position, Vector=$Particles.SkeletalMeshLocation.SampledPosition))
       B=50
      ))
    )
    Attractor Position:NiagaraPosition=$Particles.SkeletalMeshLocation.SampledPosition
    Kill Within Radius=false
    Module.Use Falloff=false
    UseAttractorOffset=false
   )
   &Drag(Drag=3, Use Linear Drag=true, Use Rotational Drag=false)
   &SolveForcesAndVelocity(Acceleration Limit=30, Clamp Velocity=true, Limit Acceleration=false, Speed Limit=300)
   &self:LimitRange_01(Center:NiagaraPosition=$Particles.SkeletalMeshLocation.SampledPosition, New float=0.9, Pos:NiagaraPosition=$Particles.Position, Radius=50) {module=LimitRange_01}
  ]
  rendererClasses [&NiagaraSpriteRendererProperties]
  renderers [{rendererClass=&NiagaraSpriteRendererProperties}]
 }
 Flame {
  emitterUpdateScript [
   &EmitterState(Life Cycle Mode=::NewEnumerator1, Loop Behavior=::NewEnumerator1, Loop Delay=1, Loop Duration Mode=::NewEnumerator1, UseLoopDelay=false)
   &SpawnPerFrame(Spawn:NiagaraBool=$System.DoOnce.Execute, Spawn Count=15, Use Spawn Probability=false)
   &SpawnRate(
    SpawnRate=&FloatFromCurve(
     FloatCurve=DataInterface() {propertyValues {Curve {keys [
         {arriveTangent=5, interpMode=RCIM_Cubic, time=0, value=0}
         {arriveTangent=-2.5322, interpMode=RCIM_Cubic, leaveTangent=-2.5322, tangentMode=RCTM_User, time=0.2, value=1}
         <at 0.8 0>
        ]}}}
     CurveIndex=$System.Timeline001.Time
     Scale Curve=35
    )
    Use Spawn Probability=false
   )
  ]
  particleSpawnScript [
   /Niagara/Modules/Spawn/Initialization/InitializeParticle(
    Color=<rgba 7 7 7 1>
    Lifetime=&UniformRangedFloat(Minimum=0.5, Maximum=0.7)
    Module.Write Color=true
    Module.Write Lifetime=true
    Module.Write Mass=false
    Module.Write Material Random=true
    Module.Write Position=false
    Module.Write Scale=false
    Module.Write SpriteRotation=true
    Module.Write SpriteSize=true
    Position=&SimulationPosition()
    Sprite Rotation=&UniformRangedFloat(Minimum=0, Maximum=360)
    Sprite Size=&Vector2DFromFloat(Value=&RandomRangeFloat(Minimum=25, Maximum=40))
   )
   /Niagara/Modules/Spawn/Location/StaticMeshLocation(
    Interpolate Mesh Sampling=false
    MaskSpawnGroup=false
    OffsetAlongNormal=false
    OverrideRandomSeed=false
    SamplePosOffset=false
    Static Mesh=DataInterface() {propertyValues {
      DefaultMesh=/BRCosmetics/Characters/Player/Male/Medium/Bodies/M_MED_Inferno/FX/Mat/SM_TempTopEmissiveArea
      InstanceIndex=-1
      LODIndex=0
      PreviewMesh=&SM_TempTopEmissiveArea
      SoftSourceActor=null
      SourceMode=DefaultMeshOnly
      bAllowSamplingFromStreamingLODs=false
      bCaptureTransformsPerFrame=true
      bUsePhysicsBodyVelocity=false
      FilteredSockets []
      LODIndexUserParameter {parameter {name=None, typeDefHandle {registeredTypeIndex=108}, varData []}}
      MeshParameterBinding {parameter {name=None, typeDefHandle {registeredTypeIndex=1}, varData []}}
      SectionFilter {allowedMaterialSlots []}
     }}
   )
   &SphereLocation(
    MaskSpawnGroup=false
    Offset=<0 -8 -2>
    Override Seed=false
    Sphere Radius=5
    Surface Only=false
    UseNonUniformScale=false
    UseOffset=false
    UseSphereOrientation=false
   ) {enabled=false}
   &AddVelocity(
    Module.Use Falloff Distance=false
    Override Seed=false
    Velocity Mode=::NewEnumerator1
    Velocity Speed=&RandomRangeFloat(Minimum=4, Maximum=8)
   )
   &AddVelocity(
    Cone Angle=75
    Module.UseConeFalloff=false
    Module.UseVelDistribution=false
    Override Seed=false
    Velocity Mode=::NewEnumerator2
    Velocity Speed=&RandomRangeFloat(Minimum=20, Maximum=40)
   ) {module=AddVelocity001}
   &AddVelocity(Velocity=<0 -15 0>) {module=AddVelocity002}
   &AddVelocity(Velocity=&Multiply_VectorByFloat(
     Vector=&RandomRangeVector(Minimum=<-1 -1 -1>, Maximum=<1 1 1>)
     Float=&RandomRangeFloat(Minimum=10, Maximum=20)
    )) {module=AddVelocity003}
   &DynamicMaterialParameters(
    Index 0 Param 2=&UniformRangedFloat(Minimum=0, Maximum=1)
    Index 0 Param 3=1
    Index 0 Param 4=&UniformRangedFloat(Minimum=1, Maximum=2)
    Module.Param0WriteEnabled=false
    Module.Param1WriteEnabled=true
    Module.Param2WriteEnabled=false
    Module.Param3WriteEnabled=true
   ) {module=DynamicMaterialParameters002}
   &CameraOffset(Camera Offset Amount=&RandomRangeFloat(Minimum=1, Maximum=3))
  ]
  particleUpdateScript [
   &ParticleState()
   &ScaleSpriteSize(Scale Factor=&Vector2DFromFloat(Value=&FloatFromCurve(FloatCurve=DataInterface() {propertyValues {Curve {keys [
          {arriveTangent=-0.1061, interpMode=RCIM_Cubic, leaveTangent=-0.1061, tangentMode=RCTM_User, time=0, value=0.2}
          {arriveTangent=-0.083, interpMode=RCIM_Cubic, leaveTangent=-0.083, tangentMode=RCTM_User, time=1, value=0.8}
         ]}}})))
   &DynamicMaterialParameters(
    Index 0 Param 1=&FloatFromCurve(FloatCurve=DataInterface() {propertyValues {Curve {keys [{arriveTangent=2.7551, interpMode=RCIM_Cubic, leaveTangent=2.7551, tangentMode=RCTM_User, time=0, value=0}, <at 0.7 1>]}}})
    Index 0 Param 2=1
    Index 0 Param 3=1
    Index 0 Param 4=1
    Module.Param0WriteEnabled=true
    Module.Param1WriteEnabled=true
    Module.Param2WriteEnabled=true
    Module.Param3WriteEnabled=false
   ) {module=DynamicMaterialParameters001}
   &GravityForce(Coordinate Space=::Local, Gravity=&RandomRangeVector(Minimum=<0 0 0>, Maximum=<0 0 0>))
   &SubUVAnimation(End Frame=40, SubUV Loop Count:NiagaraFloat=&RandomRangeFloat(Minimum=0.8, Maximum=1))
   &Drag(Drag=1.5, Use Linear Drag=true, Use Rotational Drag=false)
   &SolveForcesAndVelocity(Clamp Velocity=false, Limit Acceleration=false)
   &ScaleColor(
    Scale Mode=::NewEnumerator1
    Scale RGBA=/Niagara/DynamicInputs/TypeConversions/MakeCustomVector4FromBool(
     Bool:NiagaraBool=$User.bIsSL
     True Vector4:Vector4f=&MakeVector4_FromLinearColor(LinearColor=&self:SelectColor(
       Color1:LinearColor=&ColorFromCurve(ColorCurve=DataInterface() {propertyValues {
          AlphaCurve {keys [<at 0.0864 0>,<at 0.393 1>,<at 0.6884 1>,<at 1 0>]}
          BlueCurve {keys [<at 0.0049 1.405>,<at 0.2078 0.9433>,<at 0.7305 1>]}
          GreenCurve {keys [<at 0.0049 0>,<at 0.2078 0>,<at 0.7305 0>]}
          RedCurve {keys [<at 0.0049 2>,<at 0.2078 1>,<at 0.7305 0.1>]}
         }})
       Color2:LinearColor=&ColorFromCurve(ColorCurve=DataInterface() {propertyValues {
          AlphaCurve {keys [<at 0.0967 0>,<at 0.3879 0.8>,<at 0.6626 0.8>,<at 1 0>]}
          BlueCurve {keys [<at 0.0049 0>,<at 0.5515 0>,<at 0.996 0.3333>]}
          GreenCurve {keys [<at 0.0049 1.5>,<at 0.5515 1>,<at 0.996 0.8>]}
          RedCurve {keys [<at 0.0049 0.125>,<at 0.5515 0>,<at 0.996 0>]}
         }})
       Color3:LinearColor=&ColorFromCurve(ColorCurve=DataInterface() {propertyValues {
          AlphaCurve {keys [<at 0.1019 0>,<at 0.391 1>,<at 0.6527 1>,<at 1 0>]}
          BlueCurve {keys [<at 0.0049 0>,<at 0.2078 0.3333>,<at 0.9203 0.5>]}
          GreenCurve {keys [<at 0.0049 0.2683>,<at 0.2078 0>,<at 0.9203 0>]}
          RedCurve {keys [<at 0.0049 3>,<at 0.2078 2>,<at 0.9203 1>]}
         }})
       Index:NiagaraInt32=$User.SuperLevelIndex
      ))
     False Vector4:Vector4f=&MakeCustomVector4FromBool(
      Bool:NiagaraBool=$User.bIsWhiteVersion
      True Vector4:Vector4f=&Vector4FromCurve(Vector4Curve=DataInterface() {propertyValues {
         AlphaCurve {keys [<at 0.0971 0>,<at 0.3448 0.8>,<at 0.6619 0.8>,<at 1 0>]}
         BlueCurve {keys [<at 0.0049 0.2>,<at 0.1955 0.1>,<at 0.9203 0.2>]}
         GreenCurve {keys [<at 0.0049 0.95>,<at 0.1955 0.4>,<at 0.9203 0.4>]}
         RedCurve {keys [<at 0.0049 2>,<at 0.1955 1>,<at 0.9203 1>]}
        }})
      False Vector4:Vector4f=&Vector4FromCurve(Vector4Curve=DataInterface() {propertyValues {
         AlphaCurve {keys [<at 0.107 0>,<at 0.3111 1>,<at 0.6718 1>,<at 1 0>]}
         BlueCurve {keys [<at 0.0049 3>,<at 0.1955 1>,<at 0.9203 1>]}
         GreenCurve {keys [<at 0.0049 1.5>,<at 0.1955 0.3333>,<at 0.9203 0>]}
         RedCurve {keys [<at 0.0049 0>,<at 0.1955 0>,<at 0.9203 0.1167>]}
        }})
     )
    )
    ScaleRGBA=true
   )
   /SetVariables(Particles.VisibilityTag=&MakeCustomIntFromBool(Bool=$User.bisHeadHidden, True INT=1, False INT=0)) {isSetParameters=true, module=SetVariables_#13}
  ]
  rendererClasses [&NiagaraSpriteRendererProperties]
  renderers [
   {rendererClass=&NiagaraSpriteRendererProperties}
   {rendererClass=&NiagaraSpriteRendererProperties, rendererIndex=1}
  ]
 }
 Flame001 {
  bEnabled=false
  emitterUpdateScript [
   &EmitterState(Life Cycle Mode=::NewEnumerator1, Loop Behavior=::NewEnumerator1, Loop Delay=1, Loop Duration Mode=::NewEnumerator1, UseLoopDelay=false)
   &SpawnPerFrame(Spawn:NiagaraBool=$System.DoOnce.Execute, Spawn Count=5, Use Spawn Probability=false)
   &SpawnRate(SpawnRate=25, Use Spawn Probability=false) {enabled=false}
  ]
  particleSpawnScript [
   &InitializeParticle(
    Color=<rgba 3.5 3.5 3.5 1>
    Lifetime=&UniformRangedFloat(Minimum=0.5, Maximum=0.7)
    Module.Write Color=true
    Module.Write Lifetime=true
    Module.Write Mass=false
    Module.Write Material Random=true
    Module.Write Position=false
    Module.Write Scale=false
    Module.Write SpriteRotation=true
    Module.Write SpriteSize=true
    Position=&SimulationPosition()
    Sprite Rotation=&UniformRangedFloat(Minimum=0, Maximum=360)
    Sprite Size=&Vector2DFromFloat(Value=&RandomRangeFloat(Minimum=25, Maximum=35))
   )
   &StaticMeshLocation(
    Interpolate Mesh Sampling=false
    MaskSpawnGroup=false
    OffsetAlongNormal=false
    OverrideRandomSeed=false
    SamplePosOffset=false
    Static Mesh=DataInterface() {propertyValues {
      DefaultMesh=&SM_TempTopEmissiveArea
      InstanceIndex=-1
      LODIndex=0
      PreviewMesh=&SM_TempTopEmissiveArea
      SoftSourceActor=null
      SourceMode=DefaultMeshOnly
      bAllowSamplingFromStreamingLODs=false
      bCaptureTransformsPerFrame=true
      bUsePhysicsBodyVelocity=false
      FilteredSockets []
      LODIndexUserParameter {parameter {name=None, typeDefHandle {registeredTypeIndex=108}, varData []}}
      MeshParameterBinding {parameter {name=None, typeDefHandle {registeredTypeIndex=1}, varData []}}
      SectionFilter {allowedMaterialSlots []}
     }}
   ) {enabled=false}
   &SphereLocation(
    MaskSpawnGroup=false
    Offset=<10 -7 0>
    Override Seed=false
    Sphere Radius=2
    Surface Only=false
    UseNonUniformScale=false
    UseOffset=true
    UseSphereOrientation=false
   )
   &AddVelocity(
    Module.Use Falloff Distance=false
    Origin Offset=<0 -5 0>
    Override Seed=false
    Velocity Mode=::NewEnumerator1
    Velocity Speed=&RandomRangeFloat(Minimum=0, Maximum=0)
   ) {enabled=false}
   &AddVelocity(
    Cone Angle=75
    Cone Axis=<-1 -1 0>
    Module.UseConeFalloff=false
    Module.UseVelDistribution=false
    Override Seed=false
    Velocity Mode=::NewEnumerator2
    Velocity Speed=&RandomRangeFloat(Minimum=32.2912, Maximum=47.0232)
   ) {module=AddVelocity001}
   &AddVelocity(Velocity=<-10 0 0>) {enabled=false, module=AddVelocity002}
   &DynamicMaterialParameters(
    Index 0 Param 2=&UniformRangedFloat(Minimum=0, Maximum=1)
    Index 0 Param 3=1
    Index 0 Param 4=&UniformRangedFloat(Minimum=1, Maximum=2)
    Module.Param0WriteEnabled=false
    Module.Param1WriteEnabled=true
    Module.Param2WriteEnabled=false
    Module.Param3WriteEnabled=true
   ) {module=DynamicMaterialParameters002}
   &CameraOffset(Camera Offset Amount=&RandomRangeFloat(Minimum=1, Maximum=3))
  ]
  particleUpdateScript [
   &ParticleState()
   &ScaleSpriteSize(Scale Factor=&Vector2DFromFloat(Value=&FloatFromCurve(FloatCurve=DataInterface() {propertyValues {Curve {keys [
          {arriveTangent=-0.1061, interpMode=RCIM_Cubic, leaveTangent=-0.1061, tangentMode=RCTM_User, time=0, value=0.2}
          {arriveTangent=-0.083, interpMode=RCIM_Cubic, leaveTangent=-0.083, tangentMode=RCTM_User, time=1, value=0.8}
         ]}}})))
   &DynamicMaterialParameters(
    Index 0 Param 1=&FloatFromCurve(FloatCurve=DataInterface() {propertyValues {Curve {keys [{arriveTangent=2.7551, interpMode=RCIM_Cubic, leaveTangent=2.7551, tangentMode=RCTM_User, time=0, value=0}, <at 0.7 1>]}}})
    Index 0 Param 2=1
    Index 0 Param 3=1
    Index 0 Param 4=1
    Module.Param0WriteEnabled=true
    Module.Param1WriteEnabled=true
    Module.Param2WriteEnabled=true
    Module.Param3WriteEnabled=false
   ) {module=DynamicMaterialParameters001}
   &GravityForce(Coordinate Space=::Local, Gravity=&RandomRangeVector(Minimum=<0 -10 0>, Maximum=<0 -25 0>))
   &SubUVAnimation(End Frame=40, SubUV Loop Count:NiagaraFloat=&RandomRangeFloat(Minimum=0.8, Maximum=1))
   &Drag(Use Linear Drag=true, Use Rotational Drag=false)
   &SolveForcesAndVelocity(Clamp Velocity=false, Limit Acceleration=false)
   &ScaleColor(
    Scale Mode=::NewEnumerator1
    Scale RGBA=&MakeCustomVector4FromBool(
     Bool:NiagaraBool=$User.bIsSL
     True Vector4:Vector4f=&MakeVector4_FromLinearColor(LinearColor=&self:SelectColor(
       Color1:LinearColor=&ColorFromCurve(ColorCurve=DataInterface() {propertyValues {
          AlphaCurve {keys [<at 0.1296 0>,<at 0.3555 1>,<at 0.6527 1>,<at 1 0>]}
          BlueCurve {keys [<at 0.0049 2.1075>,<at 0.2078 1.8866>,<at 0.9203 1>]}
          GreenCurve {keys [<at 0.0049 0>,<at 0.2078 0>,<at 0.9203 0>]}
          RedCurve {keys [<at 0.0049 3>,<at 0.2078 2>,<at 0.9203 0.1167>]}
         }})
       Color2:LinearColor=&ColorFromCurve(ColorCurve=DataInterface() {propertyValues {
          AlphaCurve {keys [<at 0.1296 0>,<at 0.3555 0.8>,<at 0.6527 0.8>,<at 1 0>]}
          BlueCurve {keys [<at 0.0049 0>,<at 0.4229 0>,<at 0.996 0.4167>]}
          GreenCurve {keys [<at 0.0049 3>,<at 0.4229 2>,<at 0.996 1>]}
          RedCurve {keys [<at 0.0049 0.25>,<at 0.4229 0>,<at 0.996 0>]}
         }})
       Color3:LinearColor=&ColorFromCurve(ColorCurve=DataInterface() {propertyValues {
          AlphaCurve {keys [<at 0.1296 0>,<at 0.3555 1>,<at 0.6527 1>,<at 1 0>]}
          BlueCurve {keys [<at 0.0049 0>,<at 0.2078 0.3333>,<at 0.9203 0.5>]}
          GreenCurve {keys [<at 0.0049 0.2683>,<at 0.2078 0>,<at 0.9203 0>]}
          RedCurve {keys [<at 0.0049 3>,<at 0.2078 2>,<at 0.9203 1>]}
         }})
       Index:NiagaraInt32=$User.SuperLevelIndex
      ))
     False Vector4:Vector4f=&MakeCustomVector4FromBool(
      Bool:NiagaraBool=$User.bIsWhiteVersion
      True Vector4:Vector4f=&Vector4FromCurve(Vector4Curve=DataInterface() {propertyValues {
         AlphaCurve {keys [<at 0.2497 0>,<at 0.3448 0.8>,<at 0.6619 0.8>,<at 1 0>]}
         BlueCurve {keys [<at 0.0049 0.2>,<at 0.1955 0.1>,<at 0.9203 0.2>]}
         GreenCurve {keys [<at 0.0049 0.95>,<at 0.1955 0.4>,<at 0.9203 0.4>]}
         RedCurve {keys [<at 0.0049 2>,<at 0.1955 1>,<at 0.9203 1>]}
        }})
      False Vector4:Vector4f=&Vector4FromCurve(Vector4Curve=DataInterface() {propertyValues {
         AlphaCurve {keys [<at 0.2299 0>,<at 0.3555 1>,<at 0.6527 1>,<at 1 0>]}
         BlueCurve {keys [<at 0.0049 3>,<at 0.1955 1>,<at 0.9203 1>]}
         GreenCurve {keys [<at 0.0049 1.5>,<at 0.1955 0.3333>,<at 0.9203 0>]}
         RedCurve {keys [<at 0.0049 0>,<at 0.1955 0>,<at 0.9203 0.1167>]}
        }})
     )
    )
    ScaleRGBA=true
   )
   /SetVariables(Particles.VisibilityTag=&MakeCustomIntFromBool(Bool=$User.bisHeadHidden, True INT=1, False INT=0)) {isSetParameters=true, module=SetVariables_#13}
  ]
  rendererClasses [&NiagaraSpriteRendererProperties]
  renderers [
   {rendererClass=&NiagaraSpriteRendererProperties}
   {rendererClass=&NiagaraSpriteRendererProperties, rendererIndex=1}
  ]
 }
 FlameMesh {
  bEnabled=false
  emitterUpdateScript [
   &EmitterState(
    Inactive Response=::NewEnumerator1
    Life Cycle Mode=::NewEnumerator1
    Loop Behavior=::NewEnumerator1
    Loop Duration Mode=::NewEnumerator1
    Reset Age On Awaken=true
    Scalability Mode=::NewEnumerator1
    UseLoopDelay=false
   )
   &SpawnBurst_Instantaneous(Spawn Count=1, Use Spawn Probability=false)
  ]
  particleSpawnScript [
   &InitializeParticle(
    Mesh Scale=<1.1 1 1>
    Module.Write Color=true
    Module.Write Lifetime=true
    Module.Write Mass=true
    Module.Write Material Random=true
    Module.Write Position=false
    Module.Write Scale=true
    Module.Write SpriteRotation=false
    Module.Write SpriteSize=false
    Position=&SimulationPosition()
    Sprite Size=<5 5>
   )
   &DynamicMaterialParameters(
    Index 0 Param 1=&RandomRangeFloat(Minimum=0, Maximum=1)
    Index 0 Param 2=1
    Index 0 Param 3=4.35
    Module.Param0WriteEnabled=true
    Module.Param1WriteEnabled=true
    Module.Param2WriteEnabled=true
    Module.Param3WriteEnabled=true
   )
   &InitialMeshOrientation(Mesh Orientation Mode=::NewEnumerator2, Rotation=$User.RotationPreview, Use Rotation Vector=true)
  ]
  particleUpdateScript [
   &ParticleState(Kill Particles When Lifetime Has Elapsed=false, Let Infinitely Lived Particles Die When Emitter Deactivates=true)
   &SolveForcesAndVelocity(Clamp Velocity=false, Limit Acceleration=false)
   &CameraOffset()
   &SkeletalMeshLocation(
    Bone Sampling Mode=::NewEnumerator4
    MaskSpawnGroup=false
    Sampled Position Offset Coordinate Space:ENiagaraSkelMeshTransforms=::NewEnumerator2
    Skeletal Mesh=DataInterface() {propertyValues {
      DefaultMesh=null
      ExcludeBoneName=root
      OptionalComponentClass=null
      PreviewMesh=&M_MED_Inferno_FaceAcc
      SkinningMode=SkinOnTheFly
      SoftSourceActor=null
      SourceMode=Default
      UvSetIndex=0
      WholeMeshLOD=-1
      bExcludeBone=false
      bReadDeformedGeometry=true
      bRequireCurrentFrameData=true
      ComponentTags []
      FilteredBones [Head]
      FilteredSockets []
      MeshUserParameter {parameter {name=None, typeDefHandle {registeredTypeIndex=1}, varData []}}
      SamplingRegions []
     }}
   )
   &ScaleColor(
    Scale Alpha=&FloatFromCurve(
     FloatCurve=DataInterface() {propertyValues {Curve {keys [
         {interpMode=RCIM_Cubic, leaveTangent=-2.1302, tangentMode=RCTM_Break, time=0.0703, value=1}
         {arriveTangent=0.3473, interpMode=RCIM_Cubic, leaveTangent=0.3473, tangentMode=RCTM_User, time=0.1697, value=0}
         {arriveTangent=-0.0411, interpMode=RCIM_Cubic, tangentMode=RCTM_Break, time=1.95, value=1}
        ]}}}
     CurveIndex=$System.Timeline001.Time
     Scale Curve=&Multiply_Float(A=0.05, B=0.2)
    )
    Scale RGB=&MakeCustomVectorFromBool(
     Bool=$User.bIsSL
     True Vector=&MakeVectorFromLinearColorRGB(Color=&self:SelectColor(
       Color1:LinearColor=<rgba 100 0 20.2548 1>
       Color2:LinearColor=<rgba 15.8333 25 0 1>
       Color3:LinearColor=<rgba 100 16.6667 0 1>
       Index:NiagaraInt32=$User.SuperLevelIndex
      ))
     False Vector=&MakeVectorFromLinearColorRGB(Color=&MakeLinearColorFromVectorAndFloat(
       "Vector (RGB)"=&MakeVectorFromLinearColorRGB(Color=&MakeCustomLinearColorFromBool(
         Bool=$User.bIsWhiteVersion
         True Color=&MakeCustomLinearColorFromBool(Bool=$User.bIsFrontEndPreview, True Color=<rgba 200 65 20 1>, False Color=<rgba 200 65 20 1>)
         False Color=&MakeCustomLinearColorFromBool(Bool=$User.bIsFrontEndPreview, True Color=<rgba 0 400 297.6663 1>, False Color=<rgba 0 200 148.8331 1>)
        ))
       "Float (Alpha)"=1
      ))
    )
    ScaleA=true
    ScaleRGB=true
   ) {module=ScaleColor001}
   /SetVariables(Particles.VisibilityTag=&MakeCustomIntFromBool(Bool=$User.bisHeadHidden, True INT=1, False INT=0)) {isSetParameters=true, module=SetVariables_#13}
  ]
  rendererClasses [&NiagaraMeshRendererProperties]
  renderers [{rendererClass=&NiagaraMeshRendererProperties, rendererIndex=0}]
 }
 Flare_Burst {
  bEnabled=false
  emitterUpdateScript [
   &EmitterState(Life Cycle Mode=::NewEnumerator1, Loop Behavior=::NewEnumerator1, Loop Duration Mode=::NewEnumerator1, UseLoopDelay=false)
   &SpawnPerFrame(Spawn:NiagaraBool=$System.DoOnce.Execute, Use Spawn Probability=false)
  ]
  particleSpawnScript [/Niagara/Modules/Spawn/Initialization/V2/InitializeParticle(
    Color=<rgba 1 1 1 0.25>
    Lifetime=1.5
    Position Offset=<16 0 0>
    Sprite Size=<150 60>
    Sprite Size Mode=::NewEnumerator4
    UsePositionOffset=true
    Write Material Random=true
   )]
  particleUpdateScript [
   &ParticleState()
   &ScaleSpriteSize(
    Scale Sprite Size Mode=::NewEnumerator0
    Uniform Scale Factor=&Multiply_Float(
     A=&FloatFromCurve(FloatCurve=DataInterface() {propertyValues {Curve {keys [
          {arriveTangent=2.2885, interpMode=RCIM_Cubic, leaveTangent=2.2885, tangentMode=RCTM_User, time=0, value=0}
          {arriveTangent=-2.3026, interpMode=RCIM_Cubic, leaveTangent=-2.3026, tangentMode=RCTM_User, time=0.2, value=1.7}
          <at 1 0>
         ]}}})
     B=1
    )
   )
   &ScaleColor(
    Scale Alpha=&FloatFromCurve(FloatCurve=DataInterface() {propertyValues {Curve {keys [
         {interpMode=RCIM_Cubic, tangentMode=RCTM_User, time=0.4, value=1}
         {arriveTangent=-1.5698, interpMode=RCIM_Cubic, leaveTangent=-1.5698, tangentMode=RCTM_User, time=1, value=0}
        ]}}})
    Scale RGB=&MakeVectorFromLinearColorRGB(Color=&MakeCustomLinearColorFromBool(
      Bool=$User.bIsSL
      True Color=&self:SelectColor(
       Color1:LinearColor=&ColorFromCurve(ColorCurve=DataInterface() {propertyValues {
          AlphaCurve {keys [<at 0 1 RCIM_Cubic>,<at 1 1 RCIM_Cubic>]}
          BlueCurve {keys [<at 0.0713 2.5 RCIM_Cubic>,<at 1 4 RCIM_Cubic>]}
          GreenCurve {keys [<at 0.0713 0.25 RCIM_Cubic>,<at 1 0 RCIM_Cubic>]}
          RedCurve {keys [<at 0.0713 2.5 RCIM_Cubic>,<at 1 0.6667 RCIM_Cubic>]}
         }})
       Color2:LinearColor=&ColorFromCurve(ColorCurve=DataInterface() {propertyValues {
          AlphaCurve {keys [<at 0 1 RCIM_Cubic>,<at 1 1 RCIM_Cubic>]}
          BlueCurve {keys [<at 0.0713 0.75 RCIM_Cubic>,<at 1 0.8277 RCIM_Cubic>]}
          GreenCurve {keys [<at 0.0713 1.5 RCIM_Cubic>,<at 1 2 RCIM_Cubic>]}
          RedCurve {keys [<at 0.0713 0 RCIM_Cubic>,<at 1 0 RCIM_Cubic>]}
         }})
       Color3:LinearColor=&ColorFromCurve(ColorCurve=DataInterface() {propertyValues {
          AlphaCurve {keys [<at 0 1 RCIM_Cubic>,<at 1 1 RCIM_Cubic>]}
          BlueCurve {keys [<at 0.0713 1.25 RCIM_Cubic>,<at 1 0.3478 RCIM_Cubic>]}
          GreenCurve {keys [<at 0.0713 0 RCIM_Cubic>,<at 1 0 RCIM_Cubic>]}
          RedCurve {keys [<at 0.0713 2.5 RCIM_Cubic>,<at 1 2 RCIM_Cubic>]}
         }})
       Index:NiagaraInt32=$User.SuperLevelIndex
      )
      False Color=&MakeCustomLinearColorFromBool(
       Bool=$User.bIsWhiteVersion
       True Color=&ColorFromCurve(ColorCurve=DataInterface() {propertyValues {
          AlphaCurve {keys [<at 0 1 RCIM_Cubic>,<at 1 1 RCIM_Cubic>]}
          BlueCurve {keys [<at 0.0713 0.5875 RCIM_Cubic>]}
          GreenCurve {keys [<at 0.0713 1.3219 RCIM_Cubic>]}
          RedCurve {keys [<at 0.0713 2.35 RCIM_Cubic>]}
         }})
       False Color=&ColorFromCurve(ColorCurve=DataInterface() {propertyValues {
          AlphaCurve {keys [<at 0 1 RCIM_Cubic>,<at 1 1 RCIM_Cubic>]}
          BlueCurve {keys [<at 0.0713 4 RCIM_Cubic>,<at 1 4 RCIM_Cubic>]}
          GreenCurve {keys [<at 0.0713 1.6 RCIM_Cubic>,<at 1 0 RCIM_Cubic>]}
          RedCurve {keys [<at 0.0713 0.4 RCIM_Cubic>,<at 1 0.3333 RCIM_Cubic>]}
         }})
      )
     ))
    ScaleA=true
    ScaleRGB=true
   )
   &FN_Light_v2(
    Light Exponent=2
    Light Radius:NiagaraFloat=&SelectFloatFromArray(Direct Array Index=$Particles.SpawnGroup, Float Selection Array=DataInterface() {propertyValues {FloatData [40,40,80]}})
    LightEnabled:NiagaraBool=&Boolean_NotOperation(Bool=$User.bisHeadHidden)
    Position Offset:Vector3f=<0 15 0>
    Scale Color=1
   )
   /SetVariables(Particles.Light Color:LinearColor=&MultiplyLinearColor(
     Linear Color=&MultiplyLinearColor(
      Linear Color=&MakeCustomLinearColorFromBool(
       Bool=$User.bIsSL
       True Color=&self:SelectColor(
        Color1:LinearColor=<rgba 50 10 50 1>
        Color2:LinearColor=<rgba 10 50 23.3333 1>
        Color3:LinearColor=<rgba 50 0 16.6667 1>
        Index:NiagaraInt32=$User.SuperLevelIndex
       )
       False Color=&MakeCustomLinearColorFromBool(Bool=$User.bIsWhiteVersion, True Color=<rgba 35 18.8125 8.75 1>, False Color=<rgba 10 23.3333 50 1>)
      )
      Scale Factor=&MakeFloat4FromVectorAndFloat(
       "Vector (RGB)"=&VectorFromFloat(Value=&RandomRangeFloat(Minimum=0.45, Maximum=1, Evaluation Type=::NewEnumerator1))
       "Float (Alpha)"=1
      )
     )
     Scale Factor=&MakeVector4_FromFloat(Float=&FloatFromCurve(FloatCurve=DataInterface() {propertyValues {Curve {keys [
           <at 0 0>
           <at 0.1 1>
           {arriveTangent=-1.2862, interpMode=RCIM_Cubic, leaveTangent=-1.2862, tangentMode=RCTM_User, time=0.4, value=1.5}
           {arriveTangent=-0.2557, interpMode=RCIM_Cubic, leaveTangent=-0.2557, tangentMode=RCTM_User, time=1, value=0}
          ]}}}))
    )) {isSetParameters=true}
   &CameraOffset(Camera Offset Amount=15)
   &DynamicMaterialParameters(
    Index 0 Param 1=0.4
    Index 0 Param 2=0.5
    Module.Param0WriteEnabled=true
    Module.Param1WriteEnabled=true
    Module.Param2WriteEnabled=true
    Module.Param3WriteEnabled=true
   )
   /SetVariables(Particles.VisibilityTag=&MakeCustomIntFromBool(Bool=$User.bisHeadHidden, True INT=1, False INT=0)) {isSetParameters=true, module=SetVariables_#13}
  ]
  rendererClasses [&NiagaraSpriteRendererProperties, &NiagaraLightRendererProperties]
  renderers [
   {rendererClass=&NiagaraSpriteRendererProperties}
   {rendererClass=&NiagaraLightRendererProperties, rendererIndex=1}
  ]
 }
 Flare_Burst001 {
  emitterUpdateScript [
   &EmitterState(Life Cycle Mode=::NewEnumerator1, Loop Duration=1.5, UseLoopDelay=false)
   &SpawnBurst_Instantaneous(Use Spawn Probability=false, UseLoopCountLimit=false)
   &SpawnPerFrame(Spawn:NiagaraBool=$System.DoOnce.Execute, Spawn Group=1, Use Spawn Probability=false)
  ]
  particleSpawnScript [
   &InitializeParticle(
    Color=&MaskLinearColorBySpawnGroup(
     Masked Linear Color=<rgba 1 1 1 0.25>
     Passthrough Linear Color=&MakeCustomLinearColorFromBool(
      Bool=$User.bIsSL
      True Color=&self:SelectColor(
       Color1:LinearColor=<rgba 0.5833 0 1 0.066>
       Color2:LinearColor=<rgba 0 1 0.2 0.045>
       Color3:LinearColor=<rgba 1 0 0.1667 0.066>
       Index:NiagaraInt32=$User.SuperLevelIndex
      )
      False Color=&MakeCustomLinearColorFromBool(Bool=$User.bIsWhiteVersion, True Color=<rgba 1 0.7156 0.243 0.05>, False Color=<rgba 0 0.3333 1 0.066>)
     )
     Spawn Group Mask=1
    )
    Lifetime=1.5
    Position Offset=&MaskVectorBySpawnGroup(Masked Vector=<16 0 0>, Passthrough Vector=<21 4 0>, Spawn Group Mask=1)
    Sprite Size=&MaskVector2dBySpawnGroup(Masked Vector2D=<150 60>, Passthrough Vector2D=<40 40>, Spawn Group Mask=1)
    Sprite Size Mode=::NewEnumerator4
    UsePositionOffset=true
    Write Material Random=true
   )
   /SetVariables(Particles.VisibilityTag=&SpawnGroup()) {isSetParameters=true}
  ]
  particleUpdateScript [
   &ParticleState()
   &ScaleSpriteSize(
    Scale Sprite Size Mode=::NewEnumerator0
    Uniform Scale Factor=&MaskFloatBySpawnGroup(
     Masked Float=&Multiply_Float(
      A=&FloatFromCurve(FloatCurve=DataInterface() {propertyValues {Curve {keys [
           {arriveTangent=2.2885, interpMode=RCIM_Cubic, leaveTangent=2.2885, tangentMode=RCTM_User, time=0, value=0}
           {arriveTangent=-2.3026, interpMode=RCIM_Cubic, leaveTangent=-2.3026, tangentMode=RCTM_User, time=0.2, value=1.7}
           <at 1 0>
          ]}}})
      B=1
     )
     Passthrough Float=1
     Spawn Group Mask=1
    )
   )
   /SetVariables(Particles.New float:NiagaraFloat=&Waveform("Waveform [1]"=::NewEnumerator12, "[1] Phase Scale"=15.9584, "[1] Frequency"=1, "[1] Amplitude Min/Max"=<0.7026 1>)) {isSetParameters=true}
   &CameraOffset(Camera Offset Amount=20)
   /SetVariables(Particles.New Position:NiagaraPosition=&AddVectorToPosition(Position=$Particles.Position, Vector=<-5 -4 0>)) {isSetParameters=true}
   &ScaleColor(
    Scale Alpha=&MaskFloatBySpawnGroup(
     Masked Float=&FloatFromCurve(FloatCurve=DataInterface() {propertyValues {Curve {keys [
          {interpMode=RCIM_Cubic, tangentMode=RCTM_User, time=0.4, value=1}
          {arriveTangent=-1.5698, interpMode=RCIM_Cubic, leaveTangent=-1.5698, tangentMode=RCTM_User, time=1, value=0}
         ]}}})
     Passthrough Float=$Particles.New float
     Spawn Group Mask=1
    )
    Scale RGB=&MaskVectorBySpawnGroup(
     Masked Vector=&MakeVectorFromLinearColorRGB(Color=&MakeCustomLinearColorFromBool(
       Bool=$User.bIsSL
       True Color=&self:SelectColor(
        Color1:LinearColor=&ColorFromCurve(ColorCurve=DataInterface() {propertyValues {
           AlphaCurve {keys [<at 0 1 RCIM_Cubic>,<at 1 1 RCIM_Cubic>]}
           BlueCurve {keys [<at 0.0713 2.5 RCIM_Cubic>,<at 1 4 RCIM_Cubic>]}
           GreenCurve {keys [<at 0.0713 0.25 RCIM_Cubic>,<at 1 0 RCIM_Cubic>]}
           RedCurve {keys [<at 0.0713 2.5 RCIM_Cubic>,<at 1 0.6667 RCIM_Cubic>]}
          }})
        Color2:LinearColor=&ColorFromCurve(ColorCurve=DataInterface() {propertyValues {
           AlphaCurve {keys [<at 0 1 RCIM_Cubic>,<at 1 1 RCIM_Cubic>]}
           BlueCurve {keys [<at 0.0713 0.75 RCIM_Cubic>,<at 1 0.8277 RCIM_Cubic>]}
           GreenCurve {keys [<at 0.0713 1.5 RCIM_Cubic>,<at 1 2 RCIM_Cubic>]}
           RedCurve {keys [<at 0.0713 0 RCIM_Cubic>,<at 1 0 RCIM_Cubic>]}
          }})
        Color3:LinearColor=&ColorFromCurve(ColorCurve=DataInterface() {propertyValues {
           AlphaCurve {keys [<at 0 1 RCIM_Cubic>,<at 1 1 RCIM_Cubic>]}
           BlueCurve {keys [<at 0.0713 1.25 RCIM_Cubic>,<at 1 0.3478 RCIM_Cubic>]}
           GreenCurve {keys [<at 0.0713 0 RCIM_Cubic>,<at 1 0 RCIM_Cubic>]}
           RedCurve {keys [<at 0.0713 2.5 RCIM_Cubic>,<at 1 2 RCIM_Cubic>]}
          }})
        Index:NiagaraInt32=$User.SuperLevelIndex
       )
       False Color=&MakeCustomLinearColorFromBool(
        Bool=$User.bIsWhiteVersion
        True Color=&ColorFromCurve(ColorCurve=DataInterface() {propertyValues {
           AlphaCurve {keys [<at 0 1 RCIM_Cubic>,<at 1 1 RCIM_Cubic>]}
           BlueCurve {keys [<at 0.0713 0.5875 RCIM_Cubic>]}
           GreenCurve {keys [<at 0.0713 1.3219 RCIM_Cubic>]}
           RedCurve {keys [<at 0.0713 2.35 RCIM_Cubic>]}
          }})
        False Color=&ColorFromCurve(ColorCurve=DataInterface() {propertyValues {
           AlphaCurve {keys [<at 0 1 RCIM_Cubic>,<at 1 1 RCIM_Cubic>]}
           BlueCurve {keys [<at 0.0713 4 RCIM_Cubic>,<at 1 4 RCIM_Cubic>]}
           GreenCurve {keys [<at 0.0713 1.6 RCIM_Cubic>,<at 1 0 RCIM_Cubic>]}
           RedCurve {keys [<at 0.0713 0.4 RCIM_Cubic>,<at 1 0.3333 RCIM_Cubic>]}
          }})
       )
      ))
     Passthrough Vector=<1 1 1>
     Spawn Group Mask=1
    )
    ScaleA=true
    ScaleRGB=true
   )
   /SetVariables(Particles.VisibilityTag=&SpawnGroup()) {isSetParameters=true}
   /SetVariables(Particles.VisibilityTag=&MakeCustomIntFromBool(Bool=$User.bisHeadHidden, True INT=1, False INT=$Particles.VisibilityTag)) {isSetParameters=true, module=SetVariables_#13}
   &FN_Light_v2(
    Light Exponent=2
    Light Radius:NiagaraFloat=&SelectFloatFromArray(Direct Array Index=$Particles.SpawnGroup, Float Selection Array=DataInterface() {propertyValues {FloatData [40,40,80]}})
    LightEnabled:NiagaraBool=&Boolean_NotOperation(Bool=$User.bisHeadHidden)
    Position Offset:Vector3f=<0 15 0>
    Scale Color=1
   )
   &DynamicMaterialParameters(
    Index 0 Param 1=0.4
    Index 0 Param 2=0.5
    Module.Param0WriteEnabled=true
    Module.Param1WriteEnabled=true
    Module.Param2WriteEnabled=true
    Module.Param3WriteEnabled=true
   )
   /SetVariables(Particles.Light Color:LinearColor=&MultiplyLinearColor(
     Linear Color=&MultiplyLinearColor(
      Linear Color=&MakeCustomLinearColorFromBool(
       Bool=$User.bIsSL
       True Color=&self:SelectColor(
        Color1:LinearColor=<rgba 50 10 50 1>
        Color2:LinearColor=<rgba 10 50 23.3333 1>
        Color3:LinearColor=<rgba 50 0 16.6667 1>
        Index:NiagaraInt32=$User.SuperLevelIndex
       )
       False Color=&MakeCustomLinearColorFromBool(Bool=$User.bIsWhiteVersion, True Color=<rgba 35 18.8125 8.75 1>, False Color=<rgba 10 23.3333 50 1>)
      )
      Scale Factor=&MakeFloat4FromVectorAndFloat(
       "Vector (RGB)"=&VectorFromFloat(Value=&RandomRangeFloat(Minimum=0.45, Maximum=1, Evaluation Type=::NewEnumerator1))
       "Float (Alpha)"=1
      )
     )
     Scale Factor=&MakeVector4_FromFloat(Float=&FloatFromCurve(FloatCurve=DataInterface() {propertyValues {Curve {keys [
           <at 0 0>
           <at 0.1 1>
           {arriveTangent=-1.2862, interpMode=RCIM_Cubic, leaveTangent=-1.2862, tangentMode=RCTM_User, time=0.4, value=1.5}
           {arriveTangent=-0.2557, interpMode=RCIM_Cubic, leaveTangent=-0.2557, tangentMode=RCTM_User, time=1, value=0}
          ]}}}))
    )) {isSetParameters=true}
  ]
  rendererClasses [&NiagaraSpriteRendererProperties, &NiagaraLightRendererProperties]
  renderers [
   {rendererClass=&NiagaraSpriteRendererProperties}
   {rendererClass=&NiagaraSpriteRendererProperties, rendererIndex=1}
   {rendererClass=&NiagaraLightRendererProperties, rendererIndex=2}
  ]
 }
 Fountain {
  simTarget=GPUComputeSim
  emitterUpdateScript [
   &EmitterState(Life Cycle Mode=::NewEnumerator1, Loop Behavior=::NewEnumerator1, Loop Duration Mode=::NewEnumerator1, UseLoopDelay=false)
   &SpawnRate(
    SpawnRate=&FloatFromCurve(
     FloatCurve=DataInterface() {propertyValues {Curve {keys [
         {arriveTangent=5, interpMode=RCIM_Cubic, time=0, value=0}
         {arriveTangent=-0.8065, interpMode=RCIM_Cubic, leaveTangent=-0.8065, tangentMode=RCTM_User, time=0.2, value=1}
         <at 1.35 0>
        ]}}}
     CurveIndex=$System.Timeline001.Time
     Scale Curve=120
    )
    Use Spawn Probability=false
   )
   &SpawnPerFrame(Spawn:NiagaraBool=$System.DoOnce.Execute, Spawn Count=30, Spawn Group=1, Use Spawn Probability=false)
  ]
  particleSpawnScript [
   &InitializeParticle(
    Color=<rgba 100 100 100 1>
    Lifetime=&MaskFloatBySpawnGroup(
     Masked Float=&RandomRangeFloat(Minimum=0.8, Maximum=1.5)
     Passthrough Float=&RandomRangeFloat(Minimum=0.3, Maximum=0.4)
     Spawn Group Mask=1
    )
    Mass Mode=::NewEnumerator3
    Sprite Rotation Mode=::NewEnumerator2
    Sprite Size Max=<0.25 0.45>
    Sprite Size Min=<0.1 0.25>
    Sprite Size Mode=::NewEnumerator2
    UsePositionOffset=false
    Write Material Random=true
   )
   &StaticMeshLocation(
    Interpolate Mesh Sampling=false
    MaskSpawnGroup=false
    OffsetAlongNormal=false
    OverrideRandomSeed=false
    SamplePosOffset=false
    Static Mesh=DataInterface() {propertyValues {
      DefaultMesh=&SM_TempTopEmissiveArea
      InstanceIndex=-1
      LODIndex=0
      PreviewMesh=&SM_TempTopEmissiveArea
      SoftSourceActor=null
      SourceMode=DefaultMeshOnly
      bAllowSamplingFromStreamingLODs=false
      bCaptureTransformsPerFrame=true
      bUsePhysicsBodyVelocity=false
      FilteredSockets []
      LODIndexUserParameter {parameter {name=None, typeDefHandle {registeredTypeIndex=108}, varData []}}
      MeshParameterBinding {parameter {name=None, typeDefHandle {registeredTypeIndex=1}, varData []}}
      SectionFilter {allowedMaterialSlots []}
     }}
   )
   &ShapeLocation(MaskSpawnGroup=false, Override Seed=false, Sphere Radius=1, UseOwnerScale=true)
   &AddVelocity(
    Module.Use Falloff Distance=false
    Origin Offset=<10 0 0>
    Override Seed=false
    Velocity Mode=::NewEnumerator1
    Velocity Speed=&RandomRangeFloat(Minimum=10, Maximum=30)
   )
   &AddVelocity(
    Cone Angle=88.3311
    Module.UseConeFalloff=false
    Module.UseVelDistribution=false
    Override Seed=false
    Velocity Mode=::NewEnumerator2
    Velocity Speed=&RandomRangeFloat(Minimum=10, Maximum=40)
   ) {module=AddVelocity001}
   &AddVelocity(Velocity=&RandomRangeVector(Minimum=<0 -25 -25>, Maximum=<0 -25 25>)) {module=AddVelocity002}
  ]
  particleUpdateScript [
   &ParticleState()
   &ScaleSpriteSize(
    Scale Sprite Size Mode=::NewEnumerator0
    Uniform Scale Factor=&Lerp_Float(
     A=0.75
     B=3
     Alpha=&ClampFloat(Float=&Divide_Float(A=$Engine.Owner.LODDistance, B=300))
    )
   ) {module=ScaleSpriteSize001}
   &ScaleSpriteSize(
    Scale Factor=&MakeVector2D(
     X=1
     Y=&FloatFromCurve(FloatCurve=DataInterface() {propertyValues {Curve {keys [
          <at 0.03 0>
          {interpMode=RCIM_Cubic, leaveTangent=1.9191, tangentMode=RCTM_Break, time=0.035, value=5}
          {arriveTangent=-30.5313, interpMode=RCIM_Cubic, tangentMode=RCTM_Break, time=0.2, value=1}
          {interpMode=RCIM_Cubic, leaveTangent=-6.4088, tangentMode=RCTM_Break, time=0.6, value=1}
          {arriveTangent=-2.0449, interpMode=RCIM_Cubic, leaveTangent=-2.0449, tangentMode=RCTM_User, time=1, value=0}
         ]}}})
    )
    Scale Sprite Size Mode=::NewEnumerator2
   )
   &Drag(Drag=&RandomRangeFloat(Minimum=1, Maximum=6), Use Linear Drag=true, Use Rotational Drag=false)
   &ScaleColor(
    Scale Alpha=&FloatFromCurve(FloatCurve=DataInterface() {propertyValues {Curve {keys [
         {tangentMode=RCTM_User, time=0, value=1}
         {arriveTangent=-0.0405, interpMode=RCIM_Cubic, leaveTangent=-0.0405, tangentMode=RCTM_User, time=0.8, value=1}
         {arriveTangent=-8.0089, interpMode=RCIM_Cubic, leaveTangent=-8.0089, tangentMode=RCTM_User, time=1, value=0}
        ]}}})
    Scale RGB=&MakeVectorFromLinearColorRGB(Color=&MakeCustomLinearColorFromBool(
      Bool=$User.bIsSL
      True Color=&self:SelectColor(
       Color1:LinearColor=&ColorFromCurve(ColorCurve=DataInterface() {propertyValues {
          AlphaCurve {keys [<at 0 1 RCIM_Cubic>,<at 1 1 RCIM_Cubic>]}
          BlueCurve {keys [<at 0 50 RCIM_Cubic>,<at 0.2463 1.2645>,<at 0.3519 1>,<at 0.64 0.5>]}
          GreenCurve {keys [<at 0 10.4166 RCIM_Cubic>,<at 0.2463 0.0675>,<at 0.3519 0>,<at 0.64 0.0417>]}
          RedCurve {keys [<at 0 2.5 RCIM_Cubic>,<at 0.2463 1.35>,<at 0.3519 0.75>,<at 0.64 0>]}
         }})
       Color2:LinearColor=&ColorFromCurve(ColorCurve=DataInterface() {propertyValues {
          AlphaCurve {keys [<at 0 1 RCIM_Cubic>,<at 1 1 RCIM_Cubic>]}
          BlueCurve {keys [<at 0 2.5 RCIM_Cubic>,<at 0.2463 0.0675>,<at 0.3519 0.0833>,<at 0.64 0.1667>]}
          GreenCurve {keys [<at 0 50 RCIM_Cubic>,<at 0.2463 1.35>,<at 0.3519 1>,<at 0.64 0.5>]}
          RedCurve {keys [<at 0 10.4167 RCIM_Cubic>,<at 0.2463 0.2812>,<at 0.3519 0>,<at 0.64 0>]}
         }})
       Color3:LinearColor=&ColorFromCurve(ColorCurve=DataInterface() {propertyValues {
          AlphaCurve {keys [<at 0 1 RCIM_Cubic>,<at 1 1 RCIM_Cubic>]}
          BlueCurve {keys [<at 0 0 RCIM_Cubic>,<at 0.2463 0>,<at 0.3519 0.0942>,<at 0.64 0.0732>]}
          GreenCurve {keys [<at 0 6.3289 RCIM_Cubic>,<at 0.2463 0.1246>,<at 0.3519 0>,<at 0.64 0>]}
          RedCurve {keys [<at 0 50 RCIM_Cubic>,<at 0.2463 1.35>,<at 0.3519 1>,<at 0.64 0.5>]}
         }})
       Index:NiagaraInt32=$User.SuperLevelIndex
      )
      False Color=&MakeCustomLinearColorFromBool(
       Bool=$User.bIsWhiteVersion
       True Color=&ColorFromCurve(ColorCurve=DataInterface() {propertyValues {
          AlphaCurve {keys [<at 0 1 RCIM_Cubic>,<at 1 1 RCIM_Cubic>]}
          BlueCurve {keys [<at 0 2.5 RCIM_Cubic>,<at 0.2596 0.4>,<at 0.6372 0.1>,<at 0.9289 0>]}
          GreenCurve {keys [<at 0 27.0417 RCIM_Cubic>,<at 0.2596 1.0667>,<at 0.6372 0.43>,<at 0.9289 0.175>]}
          RedCurve {keys [<at 0 50 RCIM_Cubic>,<at 0.2596 2>,<at 0.6372 1>,<at 0.9289 0.5>]}
         }})
       False Color=&ColorFromCurve(ColorCurve=DataInterface() {propertyValues {
          AlphaCurve {keys [<at 0 1 RCIM_Cubic>,<at 1 1 RCIM_Cubic>]}
          BlueCurve {keys [<at 0 50 RCIM_Cubic>,<at 0.2463 1.35>,<at 0.6372 1>,<at 0.9289 0.5>]}
          GreenCurve {keys [<at 0 10.4166 RCIM_Cubic>,<at 0.2463 0.2812>,<at 0.6372 0.1667>,<at 0.9289 0>]}
          RedCurve {keys [<at 0 2.5 RCIM_Cubic>,<at 0.2463 0.0675>,<at 0.6372 0>,<at 0.9289 0.2>]}
         }})
      )
     ))
    ScaleA=true
    ScaleRGB=true
   )
   &ScaleColor(
    Scale RGB=&VectorFromFloat(Value=&RandomRangeFloat(Minimum=0.05, Maximum=1, Evaluation Type=::NewEnumerator1))
    ScaleA=true
    ScaleRGB=true
   ) {module=ScaleColor001}
   &CurlNoiseForce(
    Noise Strength=&FloatFromCurve(
     FloatCurve=DataInterface() {propertyValues {Curve {keys [{interpMode=RCIM_Cubic, leaveTangent=1.1647, tangentMode=RCTM_Break, time=0, value=0.5}, <at 1 1>]}}}
     Scale Curve=&RandomRangeFloat(Minimum=85, Maximum=100)
    )
    Pan Noise=true
   ) {enabled=false}
   &GravityForce(Gravity=&Lerp_Vector(
     A=<0 0 0>
     B=&RandomRangeVector(Minimum=<0 0 35>, Maximum=<0 0 45>)
     Alpha=&FloatFromCurve(FloatCurve=DataInterface() {propertyValues {Curve {keys [{interpMode=RCIM_Cubic, leaveTangent=1.7654, tangentMode=RCTM_Break, time=0, value=0}, <at 1 1>]}}})
    ))
   &VectorNoiseForce(
    Force Amount=&FloatFromCurve(
     FloatCurve=DataInterface() {propertyValues {Curve {keys [{interpMode=RCIM_Cubic, leaveTangent=1.1647, tangentMode=RCTM_Break, time=0, value=0.5}, <at 1 1>]}}}
     Scale Curve=&RandomRangeFloat(Minimum=300, Maximum=600)
    )
    Module.Use Random Noise Time Interval=true
    Random Noise Time Interval=0.05
   )
   &SolveForcesAndVelocity(Clamp Velocity=false, Limit Acceleration=false, Speed Limit=15)
   /SetVariables(
    Particles.Color2=&MultiplyLinearColor(
     Linear Color=&MakeCustomLinearColorFromBool(
      Bool=$User.bIsSL
      True Color=&self:SelectColor(
       Color1:LinearColor=&ColorFromCurve(ColorCurve=DataInterface() {propertyValues {
          AlphaCurve {keys [<at 0 1 RCIM_Cubic>,<at 1 1 RCIM_Cubic>]}
          BlueCurve {keys [<at 0.3565 0.2 RCIM_Cubic>,<at 0.6866 0.07>]}
          GreenCurve {keys [<at 0.3565 0.0781 RCIM_Cubic>,<at 0.6866 0>]}
          RedCurve {keys [<at 0.3565 0 RCIM_Cubic>,<at 0.6866 0.0233>]}
         }})
       Color2:LinearColor=&ColorFromCurve(ColorCurve=DataInterface() {propertyValues {
          AlphaCurve {keys [<at 0 1 RCIM_Cubic>,<at 1 1 RCIM_Cubic>]}
          BlueCurve {keys [<at 0.3565 0 RCIM_Cubic>,<at 0.6866 0.0117>]}
          GreenCurve {keys [<at 0.3565 0.2 RCIM_Cubic>,<at 0.6866 0.07>]}
          RedCurve {keys [<at 0.3565 0.0333 RCIM_Cubic>,<at 0.6866 0>]}
         }})
       Color3:LinearColor=&ColorFromCurve(ColorCurve=DataInterface() {propertyValues {
          AlphaCurve {keys [<at 0 1 RCIM_Cubic>,<at 1 1 RCIM_Cubic>]}
          BlueCurve {keys [<at 0.3565 0.0112 RCIM_Cubic>,<at 0.6866 0.0081>]}
          GreenCurve {keys [<at 0.3565 0 RCIM_Cubic>,<at 0.6866 0>]}
          RedCurve {keys [<at 0.3565 0.2 RCIM_Cubic>,<at 0.6866 0.07>]}
         }})
       Index:NiagaraInt32=$User.SuperLevelIndex
      )
      False Color=&MakeCustomLinearColorFromBool(
       Bool=$User.bIsWhiteVersion
       True Color=&ColorFromCurve(ColorCurve=DataInterface() {propertyValues {
          AlphaCurve {keys [<at 0 1 RCIM_Cubic>,<at 1 1 RCIM_Cubic>]}
          BlueCurve {keys [<at 0.36 0.02 RCIM_Cubic>,<at 0.6971 0>]}
          GreenCurve {keys [<at 0.36 0.113 RCIM_Cubic>,<at 0.6971 0.0417>]}
          RedCurve {keys [<at 0.36 0.2 RCIM_Cubic>,<at 0.6971 0.1>]}
         }})
       False Color=&ColorFromCurve(ColorCurve=DataInterface() {propertyValues {
          AlphaCurve {keys [<at 0 1 RCIM_Cubic>,<at 1 1 RCIM_Cubic>]}
          BlueCurve {keys [<at 0.3565 0.2 RCIM_Cubic>,<at 0.6866 0.07>]}
          GreenCurve {keys [<at 0.3565 0.0781 RCIM_Cubic>,<at 0.6866 0>]}
          RedCurve {keys [<at 0.3565 0 RCIM_Cubic>,<at 0.6866 0.0233>]}
         }})
      )
     )
     Scale Factor=&MakeVector4(
      X=1
      Y=1
      Z=1
      W=&FloatFromCurve(
       FloatCurve=DataInterface() {propertyValues {Curve {keys [
           {arriveTangent=2.1147, interpMode=RCIM_Cubic, leaveTangent=2.1147, tangentMode=RCTM_User, time=0, value=0}
           {interpMode=RCIM_Cubic, tangentMode=RCTM_User, time=0.5, value=1}
           {interpMode=RCIM_Cubic, tangentMode=RCTM_Break, time=1, value=0}
          ]}}}
       Scale Curve=0.8
      )
     )
    )
    Particles.SpriteSize2=&Multiply_Vector2D_ByFloat(
     Float=&Waveform(
      Global Phase=$Particles.NormalizedAge
      "Waveform [1]"=::NewEnumerator0
      "[1] Phase Scale"=3
      "[1] Frequency"=1
      "[1] Amplitude Min/Max"=<0.5 1>
      "[1] Bias"=1
     )
     Vector2D=&Multiply_Vector2D_ByFloat(Float=&RandomRangeFloat(Minimum=1, Maximum=3), Vector2D=<1 1>)
    )
   ) {isSetParameters=true}
   /SetVariables(Particles.VisibilityTag=&MakeCustomIntFromBool(Bool=$User.bisHeadHidden, True INT=1, False INT=0)) {isSetParameters=true, module=SetVariables_#13}
  ]
  rendererClasses [&NiagaraSpriteRendererProperties]
  renderers [
   {rendererClass=&NiagaraSpriteRendererProperties}
   {rendererClass=&NiagaraSpriteRendererProperties, rendererIndex=1}
  ]
 }
 Star {
  emitterUpdateScript [
   &EmitterState(Life Cycle Mode=::NewEnumerator1, Loop Behavior=::NewEnumerator1, Loop Duration Mode=::NewEnumerator1, UseLoopDelay=false)
   &SpawnRate(SpawnRate=0.25, Use Spawn Probability=false)
   &SpawnBurst_Instantaneous(Spawn Time=2.5, Use Spawn Probability=false, UseLoopCountLimit=false)
  ]
  particleSpawnScript [&InitializeParticle(
    Color Mode=::NewEnumerator0
    Lifetime=0.5
    Sprite Size=<15 35>
    Sprite Size Mode=::NewEnumerator4
    UsePositionOffset=false
    Write Material Random=true
   )]
  particleUpdateScript [
   &ParticleState()
   &ScaleSpriteSize(
    Scale Sprite Size Mode=::NewEnumerator0
    Uniform Scale Factor=&RandomRangeFloat(Minimum=0, Maximum=1, Evaluation Type=::NewEnumerator1)
   ) {module=ScaleSpriteSize001}
   &ScaleSpriteSize(
    Scale Sprite Size Mode=::NewEnumerator0
    Uniform Scale Factor=&FloatFromCurve(FloatCurve=DataInterface() {propertyValues {Curve {keys [<at 0 0>,<at 0.2 1>,<at 1 0>]}}})
   )
   &SkeletalMeshLocation(
    Bone Sampling Mode=::NewEnumerator4
    MaskSpawnGroup=false
    Sampled Position Offset=<6 9 -3>
    Sampled Position Offset Coordinate Space:ENiagaraSkelMeshTransforms=::NewEnumerator2
    Skeletal Mesh=DataInterface() {propertyValues {
      DefaultMesh=null
      ExcludeBoneName=root
      OptionalComponentClass=null
      PreviewMesh=&M_MED_Inferno_FaceAcc
      SkinningMode=SkinOnTheFly
      SoftSourceActor=null
      SourceMode=Default
      UvSetIndex=0
      WholeMeshLOD=-1
      bExcludeBone=false
      bReadDeformedGeometry=true
      bRequireCurrentFrameData=true
      ComponentTags []
      FilteredBones [Head]
      FilteredSockets []
      MeshUserParameter {parameter {name=None, typeDefHandle {registeredTypeIndex=1}, varData []}}
      SamplingRegions []
     }}
   )
   &CameraOffset(Camera Offset Amount=4) {module=CameraOffset001}
   &ScaleColor(
    Scale RGB=&VectorFromFloat(Value=&FloatFromCurve(
      FloatCurve=DataInterface() {propertyValues {Curve {keys [
          {interpMode=RCIM_Cubic, leaveTangent=-3.6787, tangentMode=RCTM_Break, time=0, value=2}
          {arriveTangent=-1.7163, interpMode=RCIM_Cubic, tangentMode=RCTM_Break, time=0.7, value=0}
         ]}}}
      Scale Curve=&RandomRangeFloat(Minimum=1, Maximum=10, Evaluation Type=::NewEnumerator1)
     ))
    ScaleA=true
    ScaleRGB=true
   )
   /SetVariables(Particles.VisibilityTag=&MakeCustomIntFromBool(Bool=$User.bisHeadHidden, True INT=1, False INT=0)) {isSetParameters=true, module=SetVariables_#13}
  ]
  rendererClasses [&NiagaraSpriteRendererProperties]
  renderers [{rendererClass=&NiagaraSpriteRendererProperties}]
 }
 Star003 {
  emitterUpdateScript [
   &EmitterState(Life Cycle Mode=::NewEnumerator1, Loop Behavior=::NewEnumerator1, Loop Duration Mode=::NewEnumerator1, UseLoopDelay=false)
   &SpawnRate(
    SpawnRate=&FloatFromCurve(
     FloatCurve=DataInterface() {propertyValues {Curve {keys [
         {arriveTangent=5, interpMode=RCIM_Cubic, time=0, value=0}
         {arriveTangent=-0.5492, interpMode=RCIM_Cubic, leaveTangent=-0.5492, tangentMode=RCTM_User, time=0.2, value=1}
         {arriveTangent=-0.9241, interpMode=RCIM_Cubic, tangentMode=RCTM_Break, time=1.3, value=0}
        ]}}}
     CurveIndex=$System.Timeline001.Time
     Scale Curve=15
    )
    Use Spawn Probability=false
   ) {module=SpawnRate001}
  ]
  particleSpawnScript [
   &InitializeParticle(
    Color=&MakeCustomLinearColorFromBool(
     Bool=$User.bIsSL
     True Color=&self:SelectColor(
      Color1:LinearColor=&RandomRangeLinearColor(Minimum=<rgba 0.0847 0 1 1>, Maximum=<rgba 0.5 0 0.4845 1>)
      Color2:LinearColor=&RandomRangeLinearColor(Minimum=<rgba 0 1 0.5 1>, Maximum=<rgba 0.1667 0.5 0 1>)
      Color3:LinearColor=&RandomRangeLinearColor(Minimum=<rgba 1 0 0.1568 1>, Maximum=<rgba 0.5 0.2 0.05 1>)
      Index:NiagaraInt32=$User.SuperLevelIndex
     )
     False Color=&MakeCustomLinearColorFromBool(
      Bool=$User.bIsWhiteVersion
      True Color=&RandomRangeLinearColor(Minimum=<rgba 1 0.5567 0.3 1>, Maximum=<rgba 0.5 0.3067 0.1 1>)
      False Color=&RandomRangeLinearColor(Minimum=<rgba 0.2896 0.03 1 1>, Maximum=<rgba 0 0.2437 0.5 1>)
     )
    )
    Lifetime Max=0.3
    Lifetime Min=0.2
    Lifetime Mode=::NewEnumerator1
    Sprite Size=<15 35>
    Sprite Size Mode=::NewEnumerator4
    UsePositionOffset=false
    Write Material Random=true
   )
   &SkeletalMeshLocation(
    Bone Sampling Mode=::NewEnumerator4
    MaskSpawnGroup=false
    Sampled Position Offset=&MakeCustomVectorFromBool(Bool=$User.bIsCoatVersion, True Vector=<0 -20 10>, False Vector=<0 -25 0>)
    Sampled Position Offset Coordinate Space:ENiagaraSkelMeshTransforms=::NewEnumerator2
    Skeletal Mesh=DataInterface() {propertyValues {
      DefaultMesh=null
      ExcludeBoneName=root
      OptionalComponentClass=null
      PreviewMesh=&M_MED_Inferno_FaceAcc
      SkinningMode=SkinOnTheFly
      SoftSourceActor=null
      SourceMode=Default
      UvSetIndex=0
      WholeMeshLOD=-1
      bExcludeBone=false
      bReadDeformedGeometry=true
      bRequireCurrentFrameData=true
      ComponentTags []
      FilteredBones [Head]
      FilteredSockets []
      MeshUserParameter {parameter {name=None, typeDefHandle {registeredTypeIndex=1}, varData []}}
      SamplingRegions []
     }}
   )
   &ShapeLocation(MaskSpawnGroup=false, Override Seed=false, Sphere Radius=15, UseOwnerScale=true)
  ]
  particleUpdateScript [
   &ParticleState()
   &ScaleSpriteSize(
    Uniform Curve Index=$System.Timeline001.Time
    Uniform Curve Sprite Scale=DataInterface() {propertyValues {Curve {keys [{arriveTangent=-0.5492, interpMode=RCIM_Cubic, leaveTangent=-0.5492, tangentMode=RCTM_User, time=0.25, value=1.2}, <at 0.5 0.4>]}}}
   ) {module=ScaleSpriteSize002}
   &ScaleSpriteSize(
    Scale Sprite Size Mode=::NewEnumerator0
    Uniform Scale Factor=&FloatFromCurve(
     FloatCurve=DataInterface() {propertyValues {Curve {keys [<at 0 0>,<at 0.2 1>,<at 1 0>]}}}
     Scale Curve=&RandomRangeFloat(Minimum=0.9, Maximum=1)
    )
   )
   &ScaleSpriteSize(
    Scale Sprite Size Mode=::NewEnumerator0
    Uniform Scale Factor=&RandomRangeFloat(Minimum=0, Maximum=1, Evaluation Type=::NewEnumerator1)
   ) {module=ScaleSpriteSize001}
   &CameraOffset(Camera Offset Amount=4) {module=CameraOffset001}
   &ScaleColor(
    Scale RGB=&VectorFromFloat(Value=&FloatFromCurve(
      FloatCurve=DataInterface() {propertyValues {Curve {keys [
          {interpMode=RCIM_Cubic, leaveTangent=-3.6787, tangentMode=RCTM_Break, time=0, value=1.5}
          {arriveTangent=-1.7163, interpMode=RCIM_Cubic, tangentMode=RCTM_Break, time=0.6, value=0.1}
         ]}}}
      Scale Curve=&RandomRangeFloat(Minimum=2, Maximum=10, Evaluation Type=::NewEnumerator1)
     ))
    ScaleA=true
    ScaleRGB=true
   )
   /SetVariables(
    Particles.Color2=&MultiplyLinearColor(
     Linear Color=<rgba 1 1 1 1>
     Scale Factor=&MakeVector4(
      X=&MakeFloatFromLinearColor(LinearColor=$Particles.Color, Channel:ENiagaraLinearColor_Channels=::NewEnumerator0)
      Y=&MakeFloatFromLinearColor(LinearColor=$Particles.Color, Channel:ENiagaraLinearColor_Channels=::NewEnumerator1)
      Z=&MakeFloatFromLinearColor(LinearColor=$Particles.Color, Channel:ENiagaraLinearColor_Channels=::NewEnumerator2)
      W=&FloatFromCurve(
       FloatCurve=DataInterface() {propertyValues {Curve {keys [
           {arriveTangent=2.1147, interpMode=RCIM_Cubic, leaveTangent=2.1147, tangentMode=RCTM_User, time=0, value=0}
           {interpMode=RCIM_Cubic, tangentMode=RCTM_User, time=0.5, value=1}
           {interpMode=RCIM_Cubic, tangentMode=RCTM_Break, time=1, value=0}
          ]}}}
       Scale Curve=0.25
      )
     )
    )
    Particles.SpriteSize2=&Multiply_Vector2D_ByFloat(
     Float=&Waveform(
      Global Phase=$Particles.NormalizedAge
      "Waveform [1]"=::NewEnumerator0
      "[1] Phase Scale"=3
      "[1] Frequency"=1
      "[1] Amplitude Min/Max"=<0.5 1>
      "[1] Bias"=1
     )
     Vector2D=<12 12>
    )
   ) {isSetParameters=true}
   /SetVariables(Particles.VisibilityTag=&MakeCustomIntFromBool(Bool=$User.bisHeadHidden, True INT=1, False INT=0)) {isSetParameters=true, module=SetVariables_#13}
  ]
  rendererClasses [&NiagaraSpriteRendererProperties]
  renderers [
   {rendererClass=&NiagaraSpriteRendererProperties}
   {rendererClass=&NiagaraSpriteRendererProperties, rendererIndex=1}
  ]
 }
}
summary {
 systemName=NS_Inferno_Hair
 emitters [
  {bEnabled=false, emitterName=Empty001, rendererClasses [&NiagaraMeshRendererProperties]}
  {emitterName=FeatherSimulationProto, rendererClasses [&NiagaraSpriteRendererProperties]}
  {emitterName=Empty, rendererClasses [&NiagaraMeshRendererProperties]}
  {bEnabled=false, emitterName=Empty002, rendererClasses [&NiagaraMeshRendererProperties]}
  {emitterName=Fountain, simTarget=GPUComputeSim, rendererClasses [&NiagaraSpriteRendererProperties]}
  {emitterName=FeatherSimulationProto001, rendererClasses [&NiagaraSpriteRendererProperties]}
  {emitterName=Flame, rendererClasses [&NiagaraSpriteRendererProperties]}
  {emitterName=Star, rendererClasses [&NiagaraSpriteRendererProperties]}
  {emitterName=Star003, rendererClasses [&NiagaraSpriteRendererProperties]}
  {bEnabled=false, emitterName=FlameMesh, rendererClasses [&NiagaraMeshRendererProperties]}
  {bEnabled=false, emitterName=Flame001, rendererClasses [&NiagaraSpriteRendererProperties]}
  {
   bEnabled=false
   emitterName=Flare_Burst
   rendererClasses [&NiagaraSpriteRendererProperties, &NiagaraLightRendererProperties]
  }
  {
   emitterName=Flare_Burst001
   rendererClasses [&NiagaraSpriteRendererProperties, &NiagaraLightRendererProperties]
  }
 ]
 userVariables [
  User.BackHairRootPos {
   type=Texture
   description=""
   defaultValue {
    struct=NiagaraExt_VariableValue_Object
    value {object=/BRCosmetics/Characters/Player/Male/Medium/Bodies/M_MED_Inferno/FX/Mat/T_BackHairPos, objectClass=/Script/Engine.Texture}
   }
  }
  User.bIsCoatVersion {type=NiagaraBool}
  User.bIsFrontEnd {type=NiagaraBool}
  User.bIsFrontEndPreview {type=NiagaraBool}
  User.bIsFullyDeployed {type=NiagaraBool}
  User.bisHeadHidden {type=NiagaraBool}
  User.bIsSL {type=NiagaraBool, defaultValue=true}
  User.bIsWhiteVersion {type=NiagaraBool}
  User.BurstTime {type=NiagaraFloat, defaultValue=0, description=""}
  User.DisableSim {type=NiagaraBool}
  User.EngineDelay {type=NiagaraFloat, defaultValue=0, description=""}
  User.Ex {
   type=NiagaraDataInterfaceExport
   description=""
   defaultValue {
    struct=NiagaraExt_VariableValue_DataInterface
    value {
     dataInterface=&self:NiagaraDataInterfaceExport_0
     dataInterfaceClass=/Script/Niagara.NiagaraDataInterfaceExport
    }
   }
  }
  User.FX Stage {type=NiagaraInt32}
  User.Interpolate_Rate_of_Change {type=NiagaraFloat, defaultValue=0, description=""}
  User.Mat_BackFlame {
   type=MaterialInterface
   description=""
   defaultValue {
    struct=NiagaraExt_VariableValue_Object
    value {
     object=/BRCosmetics/Characters/Player/Male/Medium/Bodies/M_MED_Inferno/FX/SuperLevel/R/MI_Inferno_FlameHair_Back_SL_R
     objectClass=/Script/Engine.MaterialInterface
    }
   }
  }
  User.Mat_BackHair {
   type=MaterialInterface
   description=""
   defaultValue {
    struct=NiagaraExt_VariableValue_Object
    value {
     object=/BRCosmetics/Characters/Player/Male/Medium/Bodies/M_MED_Inferno/FX/SuperLevel/R/MI_Inferno_MainHair_BackHair_SL_R
     objectClass=&MaterialInterface
    }
   }
  }
  User.Mat_EyeSpark {
   type=MaterialInterface
   description=""
   defaultValue {
    struct=NiagaraExt_VariableValue_Object
    value {
     object=/BRCosmetics/Characters/Player/Male/Medium/Bodies/M_MED_Inferno/FX/SuperLevel/R/MI_Inferno_CrossHair_SL_R
     objectClass=&MaterialInterface
    }
   }
  }
  User.Mat_FstLayer {
   type=MaterialInterface
   description=""
   defaultValue {
    struct=NiagaraExt_VariableValue_Object
    value {
     object=/BRCosmetics/Characters/Player/Male/Medium/Bodies/M_MED_Inferno/FX/SuperLevel/R/MI_Inferno_MainHair_FstLayer_SL_R
     objectClass=&MaterialInterface
    }
   }
  }
  User.Mat_HighLayer {
   type=MaterialInterface
   description=""
   defaultValue {
    struct=NiagaraExt_VariableValue_Object
    value {
     object=/BRCosmetics/Characters/Player/Male/Medium/Bodies/M_MED_Inferno/FX/SuperLevel/R/MI_Inferno_MainHair_HighHair_SL_R
     objectClass=&MaterialInterface
    }
   }
  }
  User.Mat_MainFlame {
   type=MaterialInterface
   description=""
   defaultValue {
    struct=NiagaraExt_VariableValue_Object
    value {
     object=/BRCosmetics/Characters/Player/Male/Medium/Bodies/M_MED_Inferno/FX/SuperLevel/R/M_Inferno_FlameHair_ForUse_SL_R
     objectClass=&MaterialInterface
    }
   }
  }
  User.Mat_NewFlame {
   type=MaterialInterface
   description=""
   defaultValue {
    struct=NiagaraExt_VariableValue_Object
    value {
     object=/BRCosmetics/Gadgets/Assets/VinderTech_GliderChute/Glider_Inferno/FX/MI_Glider_Inferno_MeshFlame_01
     objectClass=&MaterialInterface
    }
   }
  }
  User.Mat_SecLayer {
   type=MaterialInterface
   description=""
   defaultValue {
    struct=NiagaraExt_VariableValue_Object
    value {
     object=/BRCosmetics/Characters/Player/Male/Medium/Bodies/M_MED_Inferno/FX/SuperLevel/R/MI_Inferno_MainHair_SecLayer_SL_R
     objectClass=&MaterialInterface
    }
   }
  }
  User.Mat_TempEmissiveArea {
   type=MaterialInterface
   description=""
   defaultValue {
    struct=NiagaraExt_VariableValue_Object
    value {
     object=/BRCosmetics/Characters/Player/Male/Medium/Bodies/M_MED_Inferno/FX/SuperLevel/R/M_Inferno_TempCenterEmissive_SL_R
     objectClass=&MaterialInterface
    }
   }
  }
  User.Mat_TempEmissiveTranslucent {
   type=MaterialInterface
   description=""
   defaultValue {
    struct=NiagaraExt_VariableValue_Object
    value {
     object=/BRCosmetics/Characters/Player/Male/Medium/Bodies/M_MED_Inferno/FX/SuperLevel/R/MI_Inferno_TempCenterEmissive_SL_R
     objectClass=&MaterialInterface
    }
   }
  }
  User.Mat_TopFlame {
   type=MaterialInterface
   description=""
   defaultValue {
    struct=NiagaraExt_VariableValue_Object
    value {
     object=/BRCosmetics/Characters/Player/Male/Medium/Bodies/M_MED_Inferno/FX/SuperLevel/R/MI_Inferno_FlameHair_Top_SL_R
     objectClass=&MaterialInterface
    }
   }
  }
  User.PlayerBody {
   type=NiagaraDataInterfaceSkeletalMesh
   defaultValue {
    struct=NiagaraExt_VariableValue_DataInterface
    value {dataInterface=&self:NiagaraDataInterfaceSkeletalMesh_0}
   }
  }
  User.RotationPreview {type=Vector3f, description="", defaultValue {struct=Vector3f, value=<0 0 0>}}
  User.SkeletalMesh {
   type=NiagaraDataInterfaceSkeletalMesh
   defaultValue {
    struct=NiagaraExt_VariableValue_DataInterface
    value {dataInterface=&self:NiagaraDataInterfaceSkeletalMesh_1}
   }
  }
  User.SM_BackHair {
   type=StaticMesh
   description=""
   defaultValue {
    struct=NiagaraExt_VariableValue_Object
    value {object=/Game/Tools/SplineMeshGenerator/Saves/Meshes/SM_Character_Inferno_BackHair2, objectClass=&StaticMesh}
   }
  }
  User.SM_Flame {
   type=StaticMesh
   description=""
   defaultValue {
    struct=NiagaraExt_VariableValue_Object
    value {object=/Game/Tools/SplineMeshGenerator/Saves/Meshes/SM_Character_Inferno_BackHair, objectClass=&StaticMesh}
   }
  }
  User.SuperLevelIndex {type=NiagaraInt32}
 ]
}
