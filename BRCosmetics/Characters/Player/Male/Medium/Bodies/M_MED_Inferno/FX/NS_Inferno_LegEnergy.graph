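# Niagara system for the Inferno character's leg energy cosmetic, with mirrored burst-spawned emitters pinned to the left and right foot bones, each with sprite and mesh renderers.
# Smoothed particle position and a custom timer feed dynamic material parameters.
# A system timeline drives curves that briefly speed up that timer and flare color.
# User bools adjust alpha for front-end preview, select white-variant flare curves, and hide the foot effect.
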NiagaraSystem /BRCosmetics/Characters/Player/Male/Medium/Bodies/M_MED_Inferno/FX/NS_Inferno_LegEnergy
dependencies {
 usedDataInterfaces [&NiagaraDataInterfaceSkeletalMesh, &NiagaraDataInterfaceCurve]
 usedDynamicInputs [
  &SetBoolByFloatComparison
  &Boolean_OrOperation
  &Boolean_NotOperation
  &ModuloFloat
 ]
 usedModules [
  &self:SystemSpawnScript.SystemScriptSource.SystemScriptGraph.NiagaraNodeAssignment_0.SetVariables_#1
  &self:SystemSpawnScript.SystemScriptSource.SystemScriptGraph.NiagaraNodeAssignment_9.SetVariables_#2
  &self:SystemSpawnScript.SystemScriptSource.SystemScriptGraph.NiagaraNodeAssignment_8.SetVariables_#3
  &SystemState
  &self:SystemSpawnScript.SystemScriptSource.SystemScriptGraph.NiagaraNodeAssignment_1.SetVariables_#4
  &self:SystemSpawnScript.SystemScriptSource.SystemScriptGraph.NiagaraNodeAssignment_2.SetVariables_#5
  &self:SystemSpawnScript.SystemScriptSource.SystemScriptGraph.NiagaraNodeAssignment_3.SetVariables_#6
  &Timeline
  &DoOnce
  &self:SystemSpawnScript.SystemScriptSource.SystemScriptGraph.NiagaraNodeAssignment_4.SetVariables_#7
  &self:SystemSpawnScript.SystemScriptSource.SystemScriptGraph.NiagaraNodeAssignment_5.SetVariables_#8
  &self:SystemSpawnScript.SystemScriptSource.SystemScriptGraph.NiagaraNodeAssignment_6.SetVariables_#9
  &self:SystemSpawnScript.SystemScriptSource.SystemScriptGraph.NiagaraNodeAssignment_7.SetVariables_#10
 ]
}
emitters {
 Empty {
  emitterUpdateScript [
   &EmitterState(
    Inactive Response=::NewEnumerator1
    Life Cycle Mode=::NewEnumerator1
    Loop Behavior=::NewEnumerator1
    Loop Duration Mode=::NewEnumerator1
    UseLoopDelay=false
   )
   &SpawnBurst_Instantaneous(Use Spawn Probability=false, UseLoopCountLimit=false)
  ]
  particleSpawnScript [&InitializeParticle(Color Mode=::NewEnumerator0, Lifetime=5, UsePositionOffset=false, Write Material Random=true)]
  particleUpdateScript [
   &ParticleState(Kill Particles When Lifetime Has Elapsed=false, Let Infinitely Lived Particles Die When Emitter Deactivates=true)
   &InterpolateOverTime(
    Rate Of Change=10
    Target Vector Value:Vector3f=&ConvertPositionToVector(Input Position=$Particles.Position, Convert To Absolute Large World Vector=false)
    Variable Type=::NewEnumerator2
   )
   &SkeletalMeshLocation(
    Bone Sampling Mode=::NewEnumerator4
    MaskSpawnGroup=false
    Skeletal Mesh=DataInterface() {propertyValues {
      DefaultMesh=null
      ExcludeBoneName=root
      OptionalComponentClass=null
      PreviewMesh=/BRCosmetics/Characters/Player/Male/Medium/Bodies/M_MED_Inferno/Meshes/M_MED_Inferno
      SkinningMode=SkinOnTheFly
      SoftSourceActor=null
      SourceMode=Default
      UvSetIndex=0
      WholeMeshLOD=-1
      bExcludeBone=false
      bReadDeformedGeometry=true
      bRequireCurrentFrameData=true
      ComponentTags []
      FilteredBones [foot_l]
      FilteredSockets []
      MeshUserParameter {parameter {name=None, typeDefHandle {registeredTypeIndex=1}, varData []}}
      SamplingRegions []
     }}
   )
   &ScaleColor(
    Scale Alpha=&MakeCustomFloatFromBool(Bool=$User.bIsFrontEndPreview, True Float=1.2, False Float=0.5)
    ScaleA=true
    ScaleRGB=true
   )
   &FN_Timer(Increment Multiplier:NiagaraFloat=&FloatFromCurve(
     FloatCurve=DataInterface() {propertyValues {Curve {keys [
         {interpMode=RCIM_Cubic, leaveTangent=8.8583, tangentMode=RCTM_Break, time=0, value=1}
         {interpMode=RCIM_Cubic, tangentMode=RCTM_Break, time=0.2, value=3}
         <at 1.95 1>
        ]}}}
     CurveIndex=$System.Timeline001.Time
    ))
   &DynamicMaterialParameters(
    Index 1 Param 2=&MakeFloatFromVector(Vector=$Particles.InterpolateOverTime.InterpolateOverTime_V3_.Moving Average, Channel=::NewEnumerator0)
    Index 1 Param 3=&MakeFloatFromVector(Vector=$Particles.InterpolateOverTime.InterpolateOverTime_V3_.Moving Average, Channel=::NewEnumerator1)
    Index 1 Param 4=&MakeFloatFromVector(Vector=$Particles.InterpolateOverTime.InterpolateOverTime_V3_.Moving Average, Channel=::NewEnumerator2)
    Index 3 Param 2=$Particles.FN_Timer.Value
    Module.Param0WriteEnabled1=true
    Module.Param0WriteEnabled3=true
    Module.Param1WriteEnabled1=true
    Module.Param1WriteEnabled3=true
    Module.Param2WriteEnabled1=true
    Module.Param2WriteEnabled3=true
    Module.Param3WriteEnabled1=true
    Module.Param3WriteEnabled3=true
    Write Parameter Index 0=false
    Write Parameter Index 1=true
    Write Parameter Index 3=true
   )
   &ScaleColor(
    Scale Mode=::NewEnumerator1
    Scale RGBA=/Niagara/DynamicInputs/TypeConversions/MakeCustomVector4FromBool(
     Bool:NiagaraBool=$User.bIsWhiteVersion
     True Vector4:Vector4f=&MakeFloat4FromVectorAndFloat(
      "Vector (RGB)"=&VectorFromFloat(Value=&FloatFromCurve(
        FloatCurve=DataInterface() {propertyValues {Curve {keys [
            {interpMode=RCIM_Cubic, leaveTangent=15.0801, tangentMode=RCTM_Break, time=0.0703, value=1}
            {arriveTangent=-1.9624, interpMode=RCIM_Cubic, leaveTangent=-1.9624, tangentMode=RCTM_User, time=0.1697, value=2}
            {arriveTangent=-0.0411, interpMode=RCIM_Cubic, tangentMode=RCTM_Break, time=1.95, value=1}
           ]}}}
        CurveIndex=$System.Timeline001.Time
       ))
      "Float (Alpha)"=&FloatFromCurve(
       FloatCurve=DataInterface() {propertyValues {Curve {keys [
           {interpMode=RCIM_Cubic, leaveTangent=15.0801, tangentMode=RCTM_Break, time=0.0703, value=1}
           {arriveTangent=-1.1708, interpMode=RCIM_Cubic, leaveTangent=-1.1708, tangentMode=RCTM_User, time=0.169, value=2}
           {arriveTangent=-0.0411, interpMode=RCIM_Cubic, tangentMode=RCTM_Break, time=1.95, value=1}
          ]}}}
       CurveIndex=$System.Timeline001.Time
      )
     )
     False Vector4:Vector4f=&MakeFloat4FromVectorAndFloat(
      "Vector (RGB)"=&VectorFromFloat(Value=&FloatFromCurve(
        FloatCurve=DataInterface() {propertyValues {Curve {keys [
            {interpMode=RCIM_Cubic, leaveTangent=15.0801, tangentMode=RCTM_Break, time=0.0703, value=1}
            {arriveTangent=-1.7927, interpMode=RCIM_Cubic, leaveTangent=-1.7927, tangentMode=RCTM_User, time=0.1697, value=2}
            {arriveTangent=-0.0411, interpMode=RCIM_Cubic, tangentMode=RCTM_Break, time=1.95, value=1}
           ]}}}
        CurveIndex=$System.Timeline001.Time
       ))
      "Float (Alpha)"=&FloatFromCurve(
       FloatCurve=DataInterface() {propertyValues {Curve {keys [<at 0.0703 1>, {arriveTangent=-0.0411, interpMode=RCIM_Cubic, tangentMode=RCTM_Break, time=1.95, value=1}]}}}
       CurveIndex=$System.Timeline001.Time
      )
     )
    )
    ScaleRGBA=true
   ) {module=ScaleColor001}
   &ScaleColor(
    Scale Alpha=&MakeCustomFloatFromBool(Bool=$User.bisFootHidden, True Float=0, False Float=1)
    ScaleA=true
    ScaleRGB=false
   ) {module=ScaleColor002}
  ]
  rendererClasses [&NiagaraSpriteRendererProperties, &NiagaraMeshRendererProperties]
  renderers [
   {rendererClass=&NiagaraSpriteRendererProperties}
   {rendererClass=&NiagaraMeshRendererProperties, rendererIndex=1}
  ]
 }
 Empty001 {
  emitterUpdateScript [
   &EmitterState(
    Inactive Response=::NewEnumerator1
    Life Cycle Mode=::NewEnumerator1
    Loop Behavior=::NewEnumerator1
    Loop Duration Mode=::NewEnumerator1
    UseLoopDelay=false
   )
   &SpawnBurst_Instantaneous(Use Spawn Probability=false, UseLoopCountLimit=false)
  ]
  particleSpawnScript [&InitializeParticle(Color Mode=::NewEnumerator0, Lifetime=5, UsePositionOffset=false, Write Material Random=true)]
  particleUpdateScript [
   &ParticleState(Kill Particles When Lifetime Has Elapsed=false, Let Infinitely Lived Particles Die When Emitter Deactivates=true)
   &InterpolateOverTime(
    Rate Of Change=10
    Target Vector Value:Vector3f=&ConvertPositionToVector(Input Position=$Particles.Position, Convert To Absolute Large World Vector=false)
    Variable Type=::NewEnumerator2
   )
   &SkeletalMeshLocation(
    Bone Sampling Mode=::NewEnumerator4
    MaskSpawnGroup=false
    Skeletal Mesh=DataInterface() {propertyValues {
      DefaultMesh=null
      ExcludeBoneName=root
      OptionalComponentClass=null
      PreviewMesh=&M_MED_Inferno
      SkinningMode=SkinOnTheFly
      SoftSourceActor=null
      SourceMode=Default
      UvSetIndex=0
      WholeMeshLOD=-1
      bExcludeBone=false
      bReadDeformedGeometry=true
      bRequireCurrentFrameData=true
      ComponentTags []
      FilteredBones [foot_r]
      FilteredSockets []
      MeshUserParameter {parameter {name=None, typeDefHandle {registeredTypeIndex=1}, varData []}}
      SamplingRegions []
     }}
   )
   &ScaleColor(
    Scale Alpha=&MakeCustomFloatFromBool(Bool=$User.bIsFrontEndPreview, True Float=1.2, False Float=0.5)
    ScaleA=true
    ScaleRGB=true
   )
   &FN_Timer(Increment Multiplier:NiagaraFloat=&FloatFromCurve(
     FloatCurve=DataInterface() {propertyValues {Curve {keys [
         {interpMode=RCIM_Cubic, leaveTangent=8.8583, tangentMode=RCTM_Break, time=0, value=1}
         {interpMode=RCIM_Cubic, tangentMode=RCTM_Break, time=0.2, value=3}
         <at 1.95 1>
        ]}}}
     CurveIndex=$System.Timeline001.Time
    ))
   &DynamicMaterialParameters(
    Index 1 Param 2=&MakeFloatFromVector(Vector=$Particles.InterpolateOverTime.InterpolateOverTime_V3_.Moving Average, Channel=::NewEnumerator0)
    Index 1 Param 3=&MakeFloatFromVector(Vector=$Particles.InterpolateOverTime.InterpolateOverTime_V3_.Moving Average, Channel=::NewEnumerator1)
    Index 1 Param 4=&MakeFloatFromVector(Vector=$Particles.InterpolateOverTime.InterpolateOverTime_V3_.Moving Average, Channel=::NewEnumerator2)
    Index 3 Param 2=$Particles.FN_Timer.Value
    Module.Param0WriteEnabled1=true
    Module.Param0WriteEnabled3=true
    Module.Param1WriteEnabled1=true
    Module.Param1WriteEnabled3=true
    Module.Param2WriteEnabled1=true
    Module.Param2WriteEnabled3=true
    Module.Param3WriteEnabled1=true
    Module.Param3WriteEnabled3=true
    Write Parameter Index 0=false
    Write Parameter Index 1=true
    Write Parameter Index 3=true
   )
   &ScaleColor(
    Scale Mode=::NewEnumerator1
    Scale RGBA=&MakeCustomVector4FromBool(
     Bool:NiagaraBool=$User.bIsWhiteVersion
     True Vector4:Vector4f=&MakeFloat4FromVectorAndFloat(
      "Vector (RGB)"=&VectorFromFloat(Value=&FloatFromCurve(
        FloatCurve=DataInterface() {propertyValues {Curve {keys [
            {interpMode=RCIM_Cubic, leaveTangent=15.0801, tangentMode=RCTM_Break, time=0.0703, value=1}
            {arriveTangent=-1.9624, interpMode=RCIM_Cubic, leaveTangent=-1.9624, tangentMode=RCTM_User, time=0.1697, value=2}
            {arriveTangent=-0.0411, interpMode=RCIM_Cubic, tangentMode=RCTM_Break, time=1.95, value=1}
           ]}}}
        CurveIndex=$System.Timeline001.Time
       ))
      "Float (Alpha)"=&FloatFromCurve(
       FloatCurve=DataInterface() {propertyValues {Curve {keys [
           {interpMode=RCIM_Cubic, leaveTangent=15.0801, tangentMode=RCTM_Break, time=0.0703, value=1}
           {arriveTangent=-1.1708, interpMode=RCIM_Cubic, leaveTangent=-1.1708, tangentMode=RCTM_User, time=0.169, value=2}
           {arriveTangent=-0.0411, interpMode=RCIM_Cubic, tangentMode=RCTM_Break, time=1.95, value=1}
          ]}}}
       CurveIndex=$System.Timeline001.Time
      )
     )
     False Vector4:Vector4f=&MakeFloat4FromVectorAndFloat(
      "Vector (RGB)"=&VectorFromFloat(Value=&FloatFromCurve(
        FloatCurve=DataInterface() {propertyValues {Curve {keys [
            {interpMode=RCIM_Cubic, leaveTangent=15.0801, tangentMode=RCTM_Break, time=0.0703, value=1}
            {arriveTangent=-1.7927, interpMode=RCIM_Cubic, leaveTangent=-1.7927, tangentMode=RCTM_User, time=0.1697, value=2}
            {arriveTangent=-0.0411, interpMode=RCIM_Cubic, tangentMode=RCTM_Break, time=1.95, value=1}
           ]}}}
        CurveIndex=$System.Timeline001.Time
       ))
      "Float (Alpha)"=&FloatFromCurve(
       FloatCurve=DataInterface() {propertyValues {Curve {keys [<at 0.0703 1>, {arriveTangent=-0.0411, interpMode=RCIM_Cubic, tangentMode=RCTM_Break, time=1.95, value=1}]}}}
       CurveIndex=$System.Timeline001.Time
      )
     )
    )
    ScaleRGBA=true
   ) {module=ScaleColor001}
   &ScaleColor(
    Scale Alpha=&MakeCustomFloatFromBool(Bool=$User.bisFootHidden, True Float=0, False Float=1)
    ScaleA=true
    ScaleRGB=false
   ) {module=ScaleColor002}
  ]
  rendererClasses [&NiagaraSpriteRendererProperties, &NiagaraMeshRendererProperties]
  renderers [
   {rendererClass=&NiagaraSpriteRendererProperties}
   {rendererClass=&NiagaraMeshRendererProperties, rendererIndex=1}
  ]
 }
}
summary {
 systemName=NS_Inferno_LegEnergy
 emitters [
  {emitterName=Empty, rendererClasses [&NiagaraSpriteRendererProperties, &NiagaraMeshRendererProperties]}
  {emitterName=Empty001, rendererClasses [&NiagaraSpriteRendererProperties, &NiagaraMeshRendererProperties]}
 ]
 userVariables [
  User.bisFootHidden {type=NiagaraBool}
  User.bIsFrontEndPreview {type=NiagaraBool}
  User.bIsWhiteVersion {type=NiagaraBool}
  User.BurstTime {type=NiagaraFloat, defaultValue=0, description=""}
  User.Mat_LegMat {
   type=MaterialInterface
   description=""
   defaultValue {
    struct=NiagaraExt_VariableValue_Object
    value {
     object=/BRCosmetics/Characters/Player/Male/Medium/Bodies/M_MED_Inferno/FX/Mat/MI_Inferno_FlameHair_Leg
     objectClass=/Script/Engine.MaterialInterface
    }
   }
  }
 ]
}
