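# Control Rig for the M_MED_Inferno cosmetic body that drives front and back hanging chains on both hips.
# Parent constraints anchor chain controls to pelvis, thigh and loop bones, and stretchy B-spline IK fits the chain bones to them.
# Spherical pose readers on the thighs reweight constraints and blend controls toward corrective poses built during construction, with smoothing on the left.
# An is_front_end flag enables an extra right-leg back-out pose.
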ControlRigBlueprint /BRCosmetics/Characters/Player/Male/Medium/Bodies/M_MED_Inferno/Meshes/M_MED_Inferno_CtrlRig
graphs [
 ContainedGraph {nodes [
   n1 = RigVMFunctionEntryNode()
   SequenceAggregate = &RigVMFunction_Sequence(ExecuteContext=@n1.ExecuteContext)
   SequenceAggregate_1 = &RigVMFunction_Sequence(ExecuteContext=@SequenceAggregate.B)
   SequenceAggregate_2 = &RigVMFunction_Sequence(ExecuteContext=@SequenceAggregate_1.B)
   SequenceAggregate_3 = &RigVMFunction_Sequence(ExecuteContext=@SequenceAggregate_2.B)
   SequenceAggregate_4 = &RigVMFunction_Sequence(ExecuteContext=@SequenceAggregate_3.B)
   RigVMFunctionReturnNode(
    A=@SequenceAggregate.A
    B=@SequenceAggregate_1.A
    C=@SequenceAggregate_2.A
    D=@SequenceAggregate_3.A
    E=@SequenceAggregate_4.A
    F=@SequenceAggregate_4.B
   )
  ]}
 ContainedGraph {nodes [
   n1 = RigVMFunctionEntryNode()
   SequenceAggregate = &RigVMFunction_Sequence(ExecuteContext=@n1.ExecuteContext)
   SequenceAggregate_1 = &RigVMFunction_Sequence(ExecuteContext=@SequenceAggregate.B)
   RigVMFunctionReturnNode(A=@SequenceAggregate.A, B=@SequenceAggregate_1.A, C=@SequenceAggregate_1.B)
  ]}
 RigVMFunctionLibrary {links [], nodes []}
 RigVMModel {nodes [
   RigUnit_BeginExecution = &RigUnit_BeginExecution()
   SequenceAggregate = RigVMAggregateNode(ExecuteContext=@RigUnit_BeginExecution.ExecutePin)
   n1 = &RigUnit_SphericalPoseReader(
    ActiveRegionScaleFactors=(PositiveWidth=3.5,NegativeWidth=0.5,PositiveHeight=3,NegativeHeight=3)
    ActiveRegionSize=0.1
    Debug=(bDrawDebug=false,DebugScale=25,DebugSegments=20,DebugThickness=0.25)
    DriverAxis=(X=-1,Y=0,Z=0)
    DriverItem=(Type=Bone,Name="thigh_l")
    FalloffRegionScaleFactors=(PositiveWidth=1,NegativeWidth=1,PositiveHeight=1,NegativeHeight=1)
    FalloffSize=0.45
    OptionalParentItem=(Type=Bone)
    RotationOffset=(X=90,Y=0,Z=0)
    ExecutePin=@SequenceAggregate.A
   )
   > RigVMVariableNode(Value=@n1.OutputParam) {variable=spr_thigh_l_up_alpha}
   > n2 = &RigUnit_SphericalPoseReader(
    ActiveRegionScaleFactors=(PositiveWidth=3.5,NegativeWidth=0.5,PositiveHeight=3,NegativeHeight=3)
    ActiveRegionSize=0.1
    Debug=(bDrawDebug=false,DebugScale=25,DebugSegments=20,DebugThickness=0.25)
    DriverAxis=(X=1,Y=0,Z=0)
    DriverItem=(Type=Bone,Name="thigh_r")
    FalloffRegionScaleFactors=(PositiveWidth=1,NegativeWidth=1,PositiveHeight=1,NegativeHeight=1)
    FalloffSize=0.45
    FlipHeightScaling=false
    FlipWidthScaling=true
    OptionalParentItem=(Type=Bone)
    RotationOffset=(X=-90,Y=0,Z=0)
   )
   > RigVMVariableNode(Value=@n2.OutputParam) {variable=spr_thigh_r_up_alpha}
   > n3 = &RigUnit_SphericalPoseReader(
    ActiveRegionScaleFactors=(PositiveWidth=1,NegativeWidth=1,PositiveHeight=1,NegativeHeight=1)
    ActiveRegionSize=0.05
    Debug=(bDrawDebug=false,DebugScale=25,DebugSegments=20,DebugThickness=0.25)
    DriverAxis=(X=1,Y=0,Z=0)
    DriverItem=(Type=Bone,Name="thigh_r")
    FalloffRegionScaleFactors=(PositiveWidth=1,NegativeWidth=1,PositiveHeight=1,NegativeHeight=1)
    FalloffSize=0.09
    OptionalParentItem=(Type=Bone)
    RotationOffset=(X=15,Y=-80,Z=0)
   )
   > RigVMVariableNode(Value=@n3.OutputParam) {variable=spr_thigh_r_slight_back_out_apha}
   &RigUnit_ParentConstraint(
    AdvancedSettings=(RotationOrderForFilter=XZY)
    Child=(Type=Null,Name="dyn_front_chain_01_l_ctrl_null")
    Parents=((Item=(Type=Bone,Name="pelvis"),Weight=0.62),(Item=(Type=Bone,Name="deform_groin_l"),Weight=0.19),(Item=(Type=Bone,Name="thigh_twist_01_l"),Weight=0.15),(Item=(Type=Bone,Name="spine_01"),Weight=0.04))
    Weight=1
    ExecutePin=@SequenceAggregate.B
   )
   > &RigUnit_ParentConstraint(
    AdvancedSettings=(RotationOrderForFilter=XZY)
    Child=(Type=Null,Name="dyn_back_chain_01_l_ctrl_null")
    Parents=((Item=(Type=Bone,Name="pelvis"),Weight=0.63),(Item=(Type=Bone,Name="deform_glute_l"),Weight=0.28),(Item=(Type=Bone,Name="thigh_twist_01_l"),Weight=0.09))
    Weight=1
   )
   > &RigUnit_ParentConstraint(
    AdvancedSettings=(RotationOrderForFilter=XZY)
    Child=(Type=Null,Name="dyn_front_chain_01_r_ctrl_null")
    Parents=((Item=(Type=Bone,Name="pelvis"),Weight=0.62),(Item=(Type=Bone,Name="deform_groin_r"),Weight=0.19),(Item=(Type=Bone,Name="thigh_twist_01_r"),Weight=0.15),(Item=(Type=Bone,Name="spine_01"),Weight=0.04))
    Weight=1
   )
   > &RigUnit_ParentConstraint(
    AdvancedSettings=(RotationOrderForFilter=XZY)
    Child=(Type=Null,Name="dyn_back_chain_01_r_ctrl_null")
    Parents=((Item=(Type=Bone,Name="pelvis"),Weight=0.77),(Item=(Type=Bone,Name="deform_glute_r"),Weight=0.17),(Item=(Type=Bone,Name="thigh_twist_01_r"),Weight=0.06))
    Weight=1
   )
   SequenceAggregate_1 = &RigVMFunction_Sequence(ExecuteContext=@SequenceAggregate.C)
   &RigUnit_ParentConstraint(
    AdvancedSettings=(RotationOrderForFilter=XZY)
    Child=(Type=Null,Name="dyn_front_chain_03_l_ctrl_null")
    Parents=((Item=(Type=Bone,Name="thigh_l"),Weight=0.75),(Item=(Type=Control,Name="dyn_chain_end_l_ctrl"),Weight=0.25))
    Weight=1
    ExecutePin=@SequenceAggregate_1.A
   )
   > &RigUnit_ParentConstraint(
    AdvancedSettings=(RotationOrderForFilter=XZY)
    Child=(Type=Null,Name="dyn_front_chain_05_l_ctrl_null")
    Parents=((Item=(Type=Bone,Name="thigh_l"),Weight=0.5),(Item=(Type=Control,Name="dyn_chain_end_l_ctrl"),Weight=0.5))
    Weight=1
   )
   n4 = RigVMVariableNode() {variable=spr_thigh_l_up_alpha}
   n5 = &RigVMFunction_MathDoubleSub(A=1, B=@n4.Value)
   > &RigUnit_ParentConstraint(
    AdvancedSettings=(RotationOrderForFilter=XZY)
    Child=(Type=Null,Name="dyn_back_chain_03_l_ctrl_null")
    Parents=((Item=(Type=Bone,Name="thigh_l"),Weight=0.75),(Item=(Type=Control,Name="dyn_chain_end_l_ctrl"),Weight=0.25))
    Weight=@n5.Result
   )
   > &RigUnit_ParentConstraint(
    AdvancedSettings=(RotationOrderForFilter=XZY)
    Child=(Type=Null,Name="dyn_back_chain_05_l_ctrl_null")
    Parents=((Item=(Type=Bone,Name="thigh_l"),Weight=0.25),(Item=(Type=Control,Name="dyn_chain_end_l_ctrl"),Weight=0.75))
    Weight=@n5.Result
   )
   > &RigUnit_ParentConstraint(
    AdvancedSettings=(RotationOrderForFilter=XZY)
    Child=(Type=Null,Name="dyn_front_chain_03_r_ctrl_null")
    Parents=((Item=(Type=Bone,Name="thigh_r"),Weight=0.9),(Item=(Type=Control,Name="dyn_chain_end_r_ctrl"),Weight=0.1))
    Weight=1
   )
   n6 = RigVMVariableNode() {variable=spr_thigh_r_up_alpha}
   n7 = &RigVMFunction_MathDoubleSub(A=1, B=@n6.Value)
   > &RigUnit_ParentConstraint(
    AdvancedSettings=(RotationOrderForFilter=XZY)
    Child=(Type=Null,Name="dyn_front_chain_05_r_ctrl_null")
    Parents=((Item=(Type=Bone,Name="thigh_r"),Weight=0.75),(Item=(Type=Control,Name="dyn_chain_end_r_ctrl"),Weight=0.25))
    Weight=@n7.Result
   )
   > &RigUnit_ParentConstraint(
    AdvancedSettings=(RotationOrderForFilter=XZY)
    Child=(Type=Null,Name="dyn_back_chain_03_r_ctrl_null")
    Parents=((Item=(Type=Bone,Name="thigh_r"),Weight=0.3),(Item=(Type=Control,Name="dyn_chain_end_r_ctrl"),Weight=0.3),(Item=(Type=Control,Name="dyn_back_chain_01_r_ctrl"),Weight=0.4))
    Weight=@n7.Result
   )
   > &RigUnit_ParentConstraint(
    AdvancedSettings=(RotationOrderForFilter=XZY)
    Child=(Type=Null,Name="dyn_back_chain_05_r_ctrl_null")
    Parents=((Item=(Type=Bone,Name="thigh_r"),Weight=0.15),(Item=(Type=Control,Name="dyn_chain_end_r_ctrl"),Weight=0.75),(Item=(Type=Control,Name="dyn_back_chain_01_r_ctrl"),Weight=0.1))
    Weight=@n7.Result
   )
   > &RigUnit_ParentConstraint(
    AdvancedSettings=(RotationOrderForFilter=XZY)
    Child=(Type=Null,Name="dyn_front_chain_05_r_ctrl_null")
    Parents=((Item=(Type=Bone,Name="thigh_r"),Weight=0.5),(Item=(Type=Control,Name="dyn_chain_end_r_ctrl"),Weight=0.5))
    Weight=@n6.Value
   )
   > &RigUnit_ParentConstraint(
    AdvancedSettings=(RotationOrderForFilter=XZY)
    Child=(Type=Null,Name="dyn_back_chain_03_r_ctrl_null")
    Parents=((Item=(Type=Control,Name="dyn_back_chain_01_r_ctrl"),Weight=0.5),(Item=(Type=Control,Name="dyn_chain_end_r_ctrl"),Weight=0.5))
    Weight=@n6.Value
   )
   > &RigUnit_ParentConstraint(
    AdvancedSettings=(RotationOrderForFilter=XZY)
    Child=(Type=Null,Name="dyn_back_chain_05_r_ctrl_null")
    Parents=((Item=(Type=Control,Name="dyn_back_chain_01_r_ctrl"),Weight=0.25),(Item=(Type=Control,Name="dyn_chain_end_r_ctrl"),Weight=0.75))
    Weight=@n6.Value
   )
   &RigUnit_ParentConstraint(
    AdvancedSettings=(RotationOrderForFilter=XZY)
    Child=(Type=Null,Name="dyn_back_chain_03_l_ctrl_null")
    Parents=((Item=(Type=Bone,Name="pelvis"),Weight=0.75),(Item=(Type=Control,Name="dyn_chain_end_l_ctrl"),Weight=0.25))
    ExecutePin=@SequenceAggregate_1.B
    Weight=@n4.Value
   )
   > &RigUnit_ParentConstraint(
    AdvancedSettings=(RotationOrderForFilter=XZY)
    Child=(Type=Null,Name="dyn_back_chain_05_l_ctrl_null")
    Parents=((Item=(Type=Bone,Name="pelvis"),Weight=0.25),(Item=(Type=Control,Name="dyn_chain_end_l_ctrl"),Weight=0.75))
    Weight=@n4.Value
   )
   &RigUnit_ParentConstraint(
    AdvancedSettings=(RotationOrderForFilter=XZY)
    Child=(Type=Null,Name="dyn_chain_end_ctrl_null_l")
    Parents=((Item=(Type=Bone,Name="dyn_loop_03_l"),Weight=1))
    Weight=1
    ExecutePin=@SequenceAggregate.F
   )
   > &RigUnit_ParentConstraint(
    AdvancedSettings=(RotationOrderForFilter=XZY)
    Child=(Type=Null,Name="dyn_chain_end_r_ctrl_null")
    Parents=((Item=(Type=Bone,Name="dyn_loop_03_r"),Weight=1))
    Weight=1
   )
   RerouteNode_1 = RigVMDispatchNode(Values="((Type=Bone,Name=\"dyn_front_chain_01_l\"),((),Type=Bone,Name=\"dyn_front_chain_02_l\"),((),Type=Bone,Name=\"dyn_front_chain_03_l\"),((),Type=Bone,Name=\"dyn_front_chain_04_l\"),((),Type=Bone,Name=\"dyn_front_chain_05_l\"),((),Type=Bone,Name=\"dyn_front_chain_06_l\"), ...(truncated)") {template="DISPATCH_RigVMDispatch_ArrayMake(in Values,out Array)"}
   RerouteNode_13 = RigVMDispatchNode(Values=((Type=Control,Name="dyn_front_chain_01_l_ctrl"),((),Type=Control,Name="dyn_front_chain_03_l_ctrl"),((),Type=Control,Name="dyn_front_chain_05_l_ctrl"),((),Type=Control,Name="dyn_front_chain_07_l_ctrl"),((),Type=Control,Name="dyn_chain_end_l_ctrl"))) {template="DISPATCH_RigVMDispatch_ArrayMake(in Values,out Array)"}
   SplineIK_1 = RigVMFunctionReferenceNode(
    Primary Axis=(X=1,Y=0,Z=0)
    Samples Per Segment=16
    Secondary Spline Direction=(X=0,Y=1,Z=0)
    Spline Mode=BSpline
    Stretch=true
    Up Axis=(X=0,Y=1,Z=0)
    debug=false
    Bones=@RerouteNode_1.Array
    Controls=@RerouteNode_13.Array
    ExecuteContext=@SequenceAggregate.D
   )
   RerouteNode = RigVMDispatchNode(Values="((Type=Bone,Name=\"dyn_back_chain_01_l\"),((),Type=Bone,Name=\"dyn_back_chain_02_l\"),((),Type=Bone,Name=\"dyn_back_chain_03_l\"),((),Type=Bone,Name=\"dyn_back_chain_04_l\"),((),Type=Bone,Name=\"dyn_back_chain_05_l\"),((),Type=Bone,Name=\"dyn_back_chain_06_l\"),((),Ty ...(truncated)") {template="DISPATCH_RigVMDispatch_ArrayMake(in Values,out Array)"}
   RerouteNode_14 = RigVMDispatchNode(Values=((Type=Control,Name="dyn_back_chain_01_l_ctrl"),((),Type=Control,Name="dyn_back_chain_03_l_ctrl"),((),Type=Control,Name="dyn_back_chain_05_l_ctrl"),((),Type=Control,Name="dyn_back_chain_07_l_ctrl"),((),Type=Control,Name="dyn_chain_end_l_ctrl"))) {template="DISPATCH_RigVMDispatch_ArrayMake(in Values,out Array)"}
   > SplineIK = RigVMFunctionReferenceNode(
    Primary Axis=(X=1,Y=0,Z=0)
    Samples Per Segment=16
    Secondary Spline Direction=(X=0,Y=5,Z=0)
    Spline Mode=BSpline
    Stretch=true
    Up Axis=(X=0,Y=1,Z=0)
    Bones=@RerouteNode.Array
    Controls=@RerouteNode_14.Array
   )
   RerouteNode_11 = RigVMDispatchNode(Values="((Type=Bone,Name=\"dyn_front_chain_01_r\"),((),Type=Bone,Name=\"dyn_front_chain_02_r\"),((),Type=Bone,Name=\"dyn_front_chain_03_r\"),((),Type=Bone,Name=\"dyn_front_chain_04_r\"),((),Type=Bone,Name=\"dyn_front_chain_05_r\"),((),Type=Bone,Name=\"dyn_front_chain_06_r\"), ...(truncated)") {template="DISPATCH_RigVMDispatch_ArrayMake(in Values,out Array)"}
   RerouteNode_15 = RigVMDispatchNode(Values=((Type=Control,Name="dyn_front_chain_01_r_ctrl"),((),Type=Control,Name="dyn_front_chain_03_r_ctrl"),((),Type=Control,Name="dyn_front_chain_05_r_ctrl"),((),Type=Control,Name="dyn_front_chain_07_r_ctrl"),((),Type=Control,Name="dyn_chain_end_r_ctrl"))) {template="DISPATCH_RigVMDispatch_ArrayMake(in Values,out Array)"}
   > SplineIK_2 = RigVMFunctionReferenceNode(
    Primary Axis=(X=1,Y=0,Z=0)
    Samples Per Segment=16
    Secondary Spline Direction=(X=0,Y=1,Z=0)
    Spline Mode=BSpline
    Stretch=true
    Up Axis=(X=0,Y=1,Z=0)
    Bones=@RerouteNode_11.Array
    Controls=@RerouteNode_15.Array
   )
   RerouteNode_12 = RigVMDispatchNode(Values="((Type=Bone,Name=\"dyn_back_chain_01_r\"),((),Type=Bone,Name=\"dyn_back_chain_02_r\"),((),Type=Bone,Name=\"dyn_back_chain_03_r\"),((),Type=Bone,Name=\"dyn_back_chain_04_r\"),((),Type=Bone,Name=\"dyn_back_chain_05_r\"),((),Type=Bone,Name=\"dyn_back_chain_06_r\"),((),Ty ...(truncated)") {template="DISPATCH_RigVMDispatch_ArrayMake(in Values,out Array)"}
   RerouteNode_16 = RigVMDispatchNode(Values=((Type=Control,Name="dyn_back_chain_01_r_ctrl"),((),Type=Control,Name="dyn_back_chain_03_r_ctrl"),((),Type=Control,Name="dyn_back_chain_05_r_ctrl"),((),Type=Control,Name="dyn_back_chain_07_r_ctrl"),((),Type=Control,Name="dyn_chain_end_r_ctrl"))) {template="DISPATCH_RigVMDispatch_ArrayMake(in Values,out Array)"}
   > SplineIK_2_1 = RigVMFunctionReferenceNode(
    Primary Axis=(X=1,Y=0,Z=0)
    Samples Per Segment=16
    Secondary Spline Direction=(X=0,Y=1,Z=0)
    Spline Mode=BSpline
    Stretch=true
    Up Axis=(X=0,Y=1,Z=0)
    Bones=@RerouteNode_12.Array
    Controls=@RerouteNode_16.Array
   )
   AlphaInterp = &RigVMFunction_AlphaInterp(
    InterpSpeedDecreasing=10
    InterpSpeedIncreasing=10
    Scale=1
    bClampResult=false
    Value=RigVMVariableNode() {variable=spr_thigh_l_up_alpha}.Value
   )
   n8 = RigVMVariableNode() {variable=SplineIK_chain_l_ctrls}
   /Script/ControlRig.RigUnit_HierarchySetPoseItemArray(
    ElementType=All
    Weight=&RigVMFunction_MathDoubleSub(A=1, B=@AlphaInterp.Result).Result
    ItemsToSet=@n8.Value
    Pose=RigVMVariableNode() {variable=SplineIK_ctrls_l_bind_pose}.Value
    ExecutePin=@SequenceAggregate.E
   ) {is=struct RigVMUnitNode}
   > &RigUnit_HierarchySetPoseItemArray(ElementType=All, Weight=@AlphaInterp.Result, Pose=RigVMVariableNode() {variable=SplineIK_ctrl_thigh_l_up_pose}.Value, ItemsToSet=@n8.Value)
   n9 = RigVMVariableNode() {variable=spr_thigh_r_up_alpha}
   n10 = RigVMVariableNode() {variable=SplineIK_chain_r_ctrls}
   > &RigUnit_HierarchySetPoseItemArray(
    ElementType=All
    Weight=&RigVMFunction_MathDoubleSub(A=1, B=@n9.Value).Result
    ItemsToSet=@n10.Value
    Pose=RigVMVariableNode() {variable=SplineIK_ctrls_r_bind_pose}.Value
   )
   > n11 = &RigUnit_HierarchySetPoseItemArray(ElementType=All, ItemsToSet=@n10.Value, Pose=RigVMVariableNode() {variable=SplineIK_ctrl_thigh_r_up_pose}.Value, Weight=@n9.Value)
   Branch_RigVMUnitNode = &RigVMFunction_ControlFlowBranch(ExecuteContext=@n11.ExecutePin, Condition=RigVMVariableNode() {variable=is_front_end}.Value)
   n12 = RigVMVariableNode() {variable=spr_thigh_r_slight_back_out_apha}
   &RigUnit_HierarchySetPoseItemArray(
    ElementType=All
    ExecutePin=@Branch_RigVMUnitNode.True
    Weight=&RigVMFunction_MathDoubleSub(A=1, B=@n12.Value).Result
    Pose=RigVMVariableNode() {variable=SplineIK_ctrls_r_bind_pose}.Value
    ItemsToSet=@n10.Value
   )
   > &RigUnit_HierarchySetPoseItemArray(
    ElementType=All
    Pose=RigVMVariableNode() {variable=SplineIK_ctrl_thigh_r_leg_slight_back_out_pose}.Value
    Weight=@n12.Value
    ItemsToSet=@n10.Value
   )
  ]}
 RigVMModel contruction {nodes [
   n1 = &RigUnit_PrepareForExecution()
   SequenceAggregate = RigVMAggregateNode(ExecuteContext=@n1.ExecutePin)
   SequenceAggregate_1 = &RigVMFunction_Sequence(ExecuteContext=@SequenceAggregate.A)
   RerouteNode_13_1 = RigVMDispatchNode(Values="((Type=Control,Name=\"dyn_front_chain_01_l_ctrl\"),((),Type=Control,Name=\"dyn_front_chain_03_l_ctrl\"),((),Type=Control,Name=\"dyn_front_chain_05_l_ctrl\"),((),Type=Control,Name=\"dyn_front_chain_07_l_ctrl\"),((),Type=Control,Name=\"dyn_back_chain_01_l_ctrl\"),((), ...(truncated)") {template="DISPATCH_RigVMDispatch_ArrayMake(in Values,out Array)"}
   RigVMVariableNode(Value=@RerouteNode_13_1.Array, ExecuteContext=@SequenceAggregate_1.A) {variable=SplineIK_chain_l_ctrls}
   RerouteNode_2 = RigVMDispatchNode(Values="((Type=Control,Name=\"dyn_front_chain_01_r_ctrl\"),((),Type=Control,Name=\"dyn_front_chain_03_r_ctrl\"),((),Type=Control,Name=\"dyn_front_chain_05_r_ctrl\"),((),Type=Control,Name=\"dyn_front_chain_07_r_ctrl\"),((),Type=Control,Name=\"dyn_back_chain_01_r_ctrl\"),((), ...(truncated)") {template="DISPATCH_RigVMDispatch_ArrayMake(in Values,out Array)"}
   > RigVMVariableNode(Value=@RerouteNode_2.Array) {variable=SplineIK_chain_r_ctrls}
   RigVMVariableNode(
    Value=/Script/ControlRig.RigUnit_HierarchyGetPoseItemArray(ElementType=All, ItemsToGet=RigVMVariableNode() {variable=SplineIK_chain_l_ctrls}.Value) {is=struct RigVMUnitNode}.Pose
    ExecuteContext=@SequenceAggregate_1.B
   ) {variable=SplineIK_ctrls_l_bind_pose}
   > RigVMVariableNode(Value=&RigUnit_HierarchyGetPoseItemArray(ElementType=All, ItemsToGet=RigVMVariableNode() {variable=SplineIK_chain_r_ctrls}.Value).Pose) {variable=SplineIK_ctrls_r_bind_pose}
   SequenceAggregate_2 = &RigVMFunction_Sequence(ExecuteContext=@SequenceAggregate.C)
   &RigUnit_ModifyTransforms(
    ItemToModify=((Item=(Type=Control,Name="dyn_front_chain_03_r_ctrl"),Transform=(Rotation=(X=0,Y=0,Z=0,W=1),Translation=(X=1,Y=-4,Z=2),Scale3D=(X=1,Y=1,Z=1))))
    Weight=1
    ExecutePin=@SequenceAggregate_2.A
   )
   > &RigUnit_ModifyTransforms(
    ItemToModify=((Item=(Type=Control,Name="dyn_front_chain_05_r_ctrl"),Transform=(Rotation=(X=0,Y=0,Z=0,W=1),Translation=(X=0,Y=-5,Z=0),Scale3D=(X=1,Y=1,Z=1))))
    Weight=1
   )
   > &RigUnit_ModifyTransforms(
    ItemToModify=((Item=(Type=Control,Name="dyn_back_chain_05_r_ctrl"),Transform=(Rotation=(X=0,Y=0,Z=0,W=1),Translation=(X=-2,Y=3,Z=0),Scale3D=(X=1,Y=1,Z=1))))
    Weight=1
   )
   > &RigUnit_ModifyTransforms(
    ItemToModify=((Item=(Type=Control,Name="dyn_back_chain_03_r_ctrl"),Transform=(Rotation=(X=0,Y=0,Z=0,W=1),Translation=(X=-2,Y=3,Z=3),Scale3D=(X=1,Y=1,Z=1))))
    Weight=1
   )
   > &RigUnit_ModifyTransforms(
    ItemToModify=((Item=(Type=Control,Name="dyn_chain_end_r_ctrl"),Transform=(Rotation=(X=0,Y=0,Z=0,W=1),Translation=(X=-1,Y=-2,Z=0),Scale3D=(X=1,Y=1,Z=1))))
    Weight=1
   )
   > RigVMVariableNode(Value=&RigUnit_HierarchyGetPoseItemArray(ElementType=All, ItemsToGet=RigVMVariableNode() {variable=SplineIK_chain_r_ctrls}.Value).Pose) {variable=SplineIK_ctrl_thigh_r_up_pose}
   n2 = RigVMVariableNode() {variable=SplineIK_chain_r_ctrls}
   &RigUnit_HierarchySetPoseItemArray(
    ElementType=All
    Weight=1
    ExecutePin=@SequenceAggregate_2.B
    Pose=RigVMVariableNode() {variable=SplineIK_ctrls_r_bind_pose}.Value
    ItemsToSet=@n2.Value
   )
   > &RigUnit_ModifyTransforms(
    ItemToModify=((Item=(Type=Control,Name="dyn_back_chain_05_r_ctrl"),Transform=(Rotation=(X=0,Y=0,Z=0,W=1),Translation=(X=0,Y=-0.5,Z=1),Scale3D=(X=1,Y=1,Z=1))))
    Weight=1
   )
   > &RigUnit_ModifyTransforms(
    ItemToModify=((Item=(Type=Control,Name="dyn_chain_end_r_ctrl"),Transform=(Rotation=(X=0,Y=0,Z=0,W=1),Translation=(X=1,Y=-0.1,Z=0),Scale3D=(X=1,Y=1,Z=1))))
    Weight=1
   )
   > &RigUnit_ModifyTransforms(
    ItemToModify=((Item=(Type=Control,Name="dyn_front_chain_03_r_ctrl"),Transform=(Rotation=(X=0,Y=0,Z=0,W=1),Translation=(X=0,Y=0,Z=0),Scale3D=(X=1,Y=1,Z=1))))
    Weight=1
   )
   > &RigUnit_ModifyTransforms(
    ItemToModify=((Item=(Type=Control,Name="dyn_front_chain_05_r_ctrl"),Transform=(Rotation=(X=0,Y=0,Z=0,W=1),Translation=(X=0,Y=0,Z=0),Scale3D=(X=1,Y=1,Z=1))))
    Weight=1
   )
   > RigVMVariableNode(Value=&RigUnit_HierarchyGetPoseItemArray(ElementType=All, ItemsToGet=@n2.Value).Pose) {variable=SplineIK_ctrl_thigh_r_leg_slight_back_out_pose}
   &RigUnit_ModifyTransforms(
    ItemToModify=((Item=(Type=Control,Name="dyn_front_chain_03_l_ctrl"),Transform=(Rotation=(X=0,Y=0,Z=0,W=1),Translation=(X=0,Y=-1,Z=-1),Scale3D=(X=1,Y=1,Z=1))))
    Weight=1
    ExecutePin=@SequenceAggregate.B
   )
   > &RigUnit_ModifyTransforms(
    ItemToModify=((Item=(Type=Control,Name="dyn_front_chain_05_l_ctrl"),Transform=(Rotation=(X=0,Y=0,Z=0,W=1),Translation=(X=0,Y=-2,Z=-2),Scale3D=(X=1,Y=1,Z=1))))
    Weight=1
   )
   > &RigUnit_ModifyTransforms(
    ItemToModify=((Item=(Type=Control,Name="dyn_chain_end_l_ctrl"),Transform=(Rotation=(X=0,Y=0,Z=0,W=1),Translation=(X=2.5,Y=-7.5,Z=9.5),Scale3D=(X=1,Y=1,Z=1))))
    Weight=1
    ItemToModify.0.Transform.Rotation=&RigVMFunction_MathQuaternionFromEuler(Euler=(X=10,Y=-40,Z=-20)).Result
   )
   > &RigUnit_ModifyTransforms(
    ItemToModify=((Item=(Type=Control,Name="dyn_back_chain_05_l_ctrl"),Transform=(Rotation=(X=0,Y=0,Z=0,W=1),Translation=(X=0,Y=0,Z=-3.5),Scale3D=(X=1,Y=1,Z=1))))
    Weight=1
   )
   > RigVMVariableNode(Value=&RigUnit_HierarchyGetPoseItemArray(ElementType=All, ItemsToGet=RigVMVariableNode() {variable=SplineIK_chain_l_ctrls}.Value).Pose) {variable=SplineIK_ctrl_thigh_l_up_pose}
  ]}
]
hierarchy [
 root {loc=<0 0 0>, rot=<0 0 0 1>}
 pelvis {parent=root, loc=<0 -0.7071 93.1191>, rot=<0.0127 0.707 -0.0127 -0.707>}
 spine_01 {parent=pelvis, loc=<6.7494 -0.6388 0>, rot=<0 0 0.034 -0.9994>}
 spine_02 {parent=spine_01, loc=<7.2518 -0.035 0>, rot=<0 0 0.0238 -0.9997>}
 spine_03 {parent=spine_02, loc=<6.0104 0 0>, rot=<0 0 0.0149 -0.9999>}
 spine_04 {parent=spine_03, loc=<8.1024 0 0>, rot=<0 0 -0.0132 -0.9999>}
 spine_05 {parent=spine_04, loc=<10.0146 0 0>, rot=<0 0 -0.0015 -1>}
 attach_cape {parent=spine_05, loc=<9.017 -11.714 0>, rot=<0.0616 -0.7044 -0.0616 -0.7044>}
 neck_01 {parent=spine_05, loc=<11.3138 0.1529 0>, rot=<0 0 -0.2003 -0.9797>}
 neck_02 {parent=neck_01, loc=<4.5627 0 0>, rot=<0 0 0.0504 -0.9987>}
 head {parent=neck_02, loc=<5.3914 0 0>, rot=<0 0 0.115 -0.9934>}
 attach_backpack {parent=spine_05, loc=<-0.123 -13.473 0>, rot=<-0.0093 -0.707 0.0093 -0.707>}
 clavicle_l {parent=spine_05, loc=<4.3575 -0.1466 -2.3223>, rot=<0.0465 -0.7289 -0.009 -0.6829>}
 upperarm_l {parent=clavicle_l, loc=<15.5111 0 0>, rot=<0.0226 -0.3234 -0.0088 -0.9459>}
 lowerarm_l {parent=upperarm_l, loc=<26.6514 0 0>, rot=<0.0391 -0.0099 -0.2453 -0.9686>}
 hand_l {parent=lowerarm_l, loc=<22.3491 0 0>, rot=<-0.7081 -0.0507 0.0975 -0.6975>}
 thumb_01_l {parent=hand_l, loc=<1.3118 -0.3191 -2.459>, rot=<0.6546 -0.4042 0.0664 -0.6354>}
 thumb_02_l {parent=thumb_01_l, loc=<5.9336 -0.0448 0.0401>, rot=<-0.0037 0.0213 0.298 -0.9543>}
 thumb_03_l {parent=thumb_02_l, loc=<2.2223 0.3682 -0.0907>, rot=<0.0118 -0.013 0.0066 -0.9998>}
 index_metacarpal_l {parent=hand_l, loc=<1.6535 0.0183 -2.96>, rot=<0.0158 -0.0124 -0.0173 -0.9996>}
 index_01_l {parent=index_metacarpal_l, loc=<8.3008 0.0357 -0.0524>, rot=<-0.0044 0.0485 0.0924 -0.9945>}
 index_02_l {parent=index_01_l, loc=<3.3587 -0.3004 0.0452>, rot=<-0.002 0.0027 0.3225 -0.9466>}
 index_03_l {parent=index_02_l, loc=<2.5578 -0.123 -0.0252>, rot=<-0.002 0.0067 0.2558 -0.9667>}
 middle_metacarpal_l {parent=hand_l, loc=<2.1856 -0.1725 -0.5984>, rot=<0.006 0.0168 -0.0329 -0.9993>}
 middle_01_l {parent=middle_metacarpal_l, loc=<7.3275 -0.2447 0.0129>, rot=<-0.0208 0.0042 0.0858 -0.9961>}
 middle_02_l {parent=middle_01_l, loc=<4.0123 0.1401 0.1175>, rot=<-0.0218 0.0258 0.3512 -0.9357>}
 middle_03_l {parent=middle_02_l, loc=<2.6077 0.0534 0.1373>, rot=<0.0082 0.0188 0.1708 -0.9851>}
 ring_metacarpal_l {parent=hand_l, loc=<2.2074 -0.3625 1.3638>, rot=<-0.0284 0.0722 -0.0458 -0.9959>}
 ring_01_l {parent=ring_metacarpal_l, loc=<7.0295 -0.6197 -0.1446>, rot=<-0.0489 -0.0252 0.1892 -0.9804>}
 ring_02_l {parent=ring_01_l, loc=<3.6174 0.4647 -0.1027>, rot=<0 0 0.2806 -0.9598>}
 ring_03_l {parent=ring_02_l, loc=<2.4152 0.114 0.0199>, rot=<0 0 0.1189 -0.9929>}
 pinky_metacarpal_l {parent=hand_l, loc=<1.7616 -0.9753 3.3814>, rot=<-0.0397 0.0814 -0.069 -0.9935>}
 pinky_01_l {parent=pinky_metacarpal_l, loc=<7.1709 -0.4831 -0.1094>, rot=<-0.0654 -0.0035 0.2242 -0.9723>}
 pinky_02_l {parent=pinky_01_l, loc=<2.8274 0.0893 -0.1043>, rot=<0.0037 0.0011 0.2854 -0.9584>}
 pinky_03_l {parent=pinky_02_l, loc=<2.2046 0.0307 -0.0078>, rot=<-0.0002 0.0098 0.1445 -0.9895>}
 hand_attach_l {parent=hand_l, loc=<8.5 -3 0>, rot=<0.7071 0.7071 0 0>}
 weapon_l {parent=hand_l, loc=<-0.1974 -3.4111 0.3811>, rot=<0 0 0 1>}
 deform_wrist_dn_base_l {parent=hand_l, loc=<0 -2.2342 0>, rot=<0 0 0 -1>}
 deform_wrist_dn_l {parent=deform_wrist_dn_base_l, loc=<0 0 0>, rot=<0 0 0 1>}
 deform_wrist_up_base_l {parent=hand_l, loc=<0 2.2682 0>, rot=<0 0 0 -1>}
 deform_wrist_up_l {parent=deform_wrist_up_base_l, loc=<0 0 0>, rot=<0 0 0 1>}
 lowerarm_twist_02_l {parent=lowerarm_l, loc=<9.6199 0 0>, rot=<0 0 0 -1>}
 lowerarm_twist_01_l {parent=lowerarm_l, loc=<17.8549 0 0>, rot=<0 0 0 -1>}
 deform_elbow_in_base_l {parent=lowerarm_l, loc=<0.2112 5.259 0.8274>, rot=<0 0 0 -1>}
 deform_elbow_in_l {parent=deform_elbow_in_base_l, loc=<0 0 0>, rot=<0 0 0 1>}
 deform_elbow_out_base_l {parent=lowerarm_l, loc=<0 -4 0>, rot=<0 0 0 -1>}
 deform_elbow_out_l {parent=deform_elbow_out_base_l, loc=<0 0 0>, rot=<0 0 0 1>}
 upperarm_twist_01_l {parent=upperarm_l, loc=<0.0041 0 0>, rot=<0 0 0 1>}
 deform_upperarm_bk_base_l {parent=upperarm_twist_01_l, loc=<1.0813 -7.2193 -1.9453>, rot=<0 0 0 1>}
 deform_upperarm_bk_l {parent=deform_upperarm_bk_base_l, loc=<0 0 0>, rot=<0 0 0 1>}
 deform_upperarm_fr_base_l {parent=upperarm_twist_01_l, loc=<-0.3691 7.0506 -2.765>, rot=<0 0 0 1>}
 deform_upperarm_fr_l {parent=deform_upperarm_fr_base_l, loc=<0 0 0>, rot=<0 0 0 1>}
 deform_upperarm_up_base_l {parent=upperarm_twist_01_l, loc=<-3.2216 -1.1391 3.3311>, rot=<0 0 0 1>}
 deform_upperarm_up_l {parent=deform_upperarm_up_base_l, loc=<0 0 0>, rot=<0 0 0 1>}
 upperarm_twist_02_l {parent=upperarm_l, loc=<13.7941 0 0>, rot=<0 0 0 1>}
 clavicle_r {parent=spine_05, loc=<4.3575 -0.1466 2.3223>, rot=<0.7289 0.0465 -0.6829 0.009>}
 upperarm_r {parent=clavicle_r, loc=<-15.5111 0 0>, rot=<0.0226 -0.3234 -0.0088 -0.9459>}
 lowerarm_r {parent=upperarm_r, loc=<-26.6514 0 0>, rot=<0.0391 -0.0099 -0.2453 -0.9686>}
 hand_r {parent=lowerarm_r, loc=<-22.3491 0 0>, rot=<-0.7081 -0.0507 0.0975 -0.6975>}
 weapon_r {parent=hand_r, loc=<0.1974 3.4111 -0.3811>, rot=<0 0 0 1>}
 thumb_01_r {parent=hand_r, loc=<-1.3118 0.3191 2.459>, rot=<0.6546 -0.4042 0.0664 -0.6354>}
 thumb_02_r {parent=thumb_01_r, loc=<-5.9336 0.0448 -0.0401>, rot=<-0.0037 0.0213 0.298 -0.9543>}
 thumb_03_r {parent=thumb_02_r, loc=<-2.2223 -0.3682 0.0907>, rot=<0.0118 -0.013 0.0066 -0.9998>}
 index_metacarpal_r {parent=hand_r, loc=<-1.6535 -0.0183 2.96>, rot=<0.0158 -0.0124 -0.0173 -0.9996>}
 index_01_r {parent=index_metacarpal_r, loc=<-8.3008 -0.0357 0.0524>, rot=<-0.0044 0.0485 0.0924 -0.9945>}
 index_02_r {parent=index_01_r, loc=<-3.3587 0.3004 -0.0452>, rot=<-0.002 0.0027 0.3225 -0.9466>}
 index_03_r {parent=index_02_r, loc=<-2.5578 0.123 0.0252>, rot=<-0.002 0.0067 0.2558 -0.9667>}
 middle_metacarpal_r {parent=hand_r, loc=<-2.1856 0.1725 0.5984>, rot=<0.006 0.0168 -0.0329 -0.9993>}
 middle_01_r {parent=middle_metacarpal_r, loc=<-7.3275 0.2447 -0.0129>, rot=<-0.0208 0.0042 0.0858 -0.9961>}
 middle_02_r {parent=middle_01_r, loc=<-4.0123 -0.1401 -0.1175>, rot=<-0.0218 0.0258 0.3512 -0.9357>}
 middle_03_r {parent=middle_02_r, loc=<-2.6077 -0.0534 -0.1373>, rot=<0.0082 0.0188 0.1708 -0.9851>}
 ring_metacarpal_r {parent=hand_r, loc=<-2.2074 0.3625 -1.3638>, rot=<-0.0284 0.0722 -0.0458 -0.9959>}
 ring_01_r {parent=ring_metacarpal_r, loc=<-7.0295 0.6197 0.1446>, rot=<-0.0489 -0.0252 0.1892 -0.9804>}
 ring_02_r {parent=ring_01_r, loc=<-3.6174 -0.4647 0.1027>, rot=<0 0 0.2806 -0.9598>}
 ring_03_r {parent=ring_02_r, loc=<-2.4152 -0.114 -0.0199>, rot=<0 0 0.1189 -0.9929>}
 pinky_metacarpal_r {parent=hand_r, loc=<-1.7616 0.9753 -3.3814>, rot=<-0.0397 0.0814 -0.069 -0.9935>}
 pinky_01_r {parent=pinky_metacarpal_r, loc=<-7.1709 0.4831 0.1094>, rot=<-0.0654 -0.0035 0.2242 -0.9723>}
 pinky_02_r {parent=pinky_01_r, loc=<-2.8274 -0.0893 0.1043>, rot=<0.0037 0.0011 0.2854 -0.9584>}
 pinky_03_r {parent=pinky_02_r, loc=<-2.2046 -0.0307 0.0078>, rot=<-0.0002 0.0098 0.1445 -0.9895>}
 hand_attach_r {parent=hand_r, loc=<-8.5001 2.9999 0>, rot=<0 0 -0.7071 -0.7071>}
 deform_wrist_dn_base_r {parent=hand_r, loc=<-0.0004 2.2345 -0.0001>, rot=<0 0 0 -1>}
 deform_wrist_dn_r {parent=deform_wrist_dn_base_r, loc=<0 0 0>, rot=<0 0 0 1>}
 deform_wrist_up_base_r {parent=hand_r, loc=<0.0003 -2.2684 0.0001>, rot=<0 0 0 -1>}
 deform_wrist_up_r {parent=deform_wrist_up_base_r, loc=<0 0 0>, rot=<0 0 0 1>}
 lowerarm_twist_02_r {parent=lowerarm_r, loc=<-9.6199 0 0>, rot=<0 0 0 -1>}
 lowerarm_twist_01_r {parent=lowerarm_r, loc=<-17.8549 0 0>, rot=<0 0 0 -1>}
 deform_elbow_in_base_r {parent=lowerarm_r, loc=<-0.2111 -5.2591 -0.8275>, rot=<0 0 0 -1>}
 deform_elbow_in_r {parent=deform_elbow_in_base_r, loc=<-0.0001 0 0.0002>, rot=<0 0 0 1>}
 deform_elbow_out_base_r {parent=lowerarm_r, loc=<0 4 -0.0001>, rot=<0 0 0 -1>}
 deform_elbow_out_r {parent=deform_elbow_out_base_r, loc=<0 0 0>, rot=<0 0 0 1>}
 upperarm_twist_01_r {parent=upperarm_r, loc=<-0.0041 0 0>, rot=<0 0 0 -1>}
 deform_upperarm_bk_base_r {parent=upperarm_twist_01_r, loc=<-1.0814 7.2192 1.9455>, rot=<0 0 0 1>}
 deform_upperarm_bk_r {parent=deform_upperarm_bk_base_r, loc=<0 0 0>, rot=<0 0 0 1>}
 deform_upperarm_fr_base_r {parent=upperarm_twist_01_r, loc=<0.3689 -7.0506 2.7652>, rot=<0 0 0 1>}
 deform_upperarm_fr_r {parent=deform_upperarm_fr_base_r, loc=<0 0 0>, rot=<0 0 0 1>}
 deform_upperarm_up_base_r {parent=upperarm_twist_01_r, loc=<3.2214 1.1391 -3.3308>, rot=<0 0 0 1>}
 deform_upperarm_up_r {parent=deform_upperarm_up_base_r, loc=<0 0 0>, rot=<0 0 0 -1>}
 upperarm_twist_02_r {parent=upperarm_r, loc=<-13.7941 0 0>, rot=<0 0 0 -1>}
 deform_pec_aim_base_l {parent=spine_04, loc=<7.8732 12.6388 -2.4123>, rot=<0.0102 -0.4602 0.3979 -0.7936>}
 deform_pec_aim_l {parent=deform_pec_aim_base_l, loc=<0 0 0>, rot=<0 0 0 1>}
 deform_pec_base_l {parent=deform_pec_aim_l, loc=<10.766 0 0>, rot=<0 0 0 1>}
 deform_pec_l {parent=deform_pec_base_l, loc=<0.055 0 0>, rot=<0 0 0 1>}
 deform_pec_aim_base_r {parent=spine_04, loc=<7.8727 12.6388 2.4123>, rot=<-0.4602 -0.0102 0.7936 0.3979>}
 deform_pec_aim_r {parent=deform_pec_aim_base_r, loc=<0 0 0>, rot=<0 0 0 1>}
 deform_pec_base_r {parent=deform_pec_aim_r, loc=<-10.7662 -0.0006 -0.0007>, rot=<0 0 0 1>}
 deform_pec_r {parent=deform_pec_base_r, loc=<0 0 0>, rot=<0 0 0 1>}
 dyn_mstr_zipper {parent=spine_04, loc=<5.5826 12.9725 0>, rot=<0.0294 -0.7065 -0.0294 -0.7065>}
 dyn_zipper {parent=dyn_mstr_zipper, loc=<0 0.4401 -0.3623>, rot=<0.707 0.0117 -0.707 -0.0117>}
 dyn_mstr_zipper_02 {parent=spine_01, loc=<2.739 9.7454 0>, rot=<0.0114 -0.707 -0.0114 -0.707>}
 dyn_zipper_02 {parent=dyn_mstr_zipper_02, loc=<0 0.2647 -0.4794>, rot=<0.7056 0.0462 -0.7056 -0.0462>}
 dyn_mstr_flap_back {parent=spine_01, loc=<-0.3351 -10.7401 0>, rot=<0.0896 -0.6746 0.7323 -0.0247>}
 dyn_flap_back_01 {parent=dyn_mstr_flap_back, loc=<0 0 0>, rot=<0 0 0 1>}
 thigh_l {parent=pelvis, loc=<-2.5693 -0.0109 -7.7501>, rot=<0.0012 0.0572 0.0155 -0.9982>}
 calf_l {parent=thigh_l, loc=<-39.0484 0 0>, rot=<0.0006 -0.0161 -0.0391 -0.9991>}
 foot_l {parent=calf_l, loc=<-42.561 0.1077 0.4292>, rot=<0.0418 -0.0455 0.0343 -0.9975>}
 ball_l {parent=foot_l, loc=<-6.1599 13.8264 0.6806>, rot=<0.0001 -0.0001 -0.7063 -0.7079>}
 calf_twist_02_l {parent=calf_l, loc=<-18.8064 -0.0967 0.0397>, rot=<0.0237 0.0095 -0.0035 -0.9997>}
 dfrm_knee_prnt_l {parent=calf_twist_02_l, loc=<6.6705 -2.7703 -6.6524>, rot=<-0.576 -0.1869 -0.0596 -0.7936>}
 dfrm_mstr_knee_tear_08_l {parent=dfrm_knee_prnt_l, loc=<0.7332 0.136 -2.6031>, rot=<-0.2017 -0.1269 0.1352 -0.9617>}
 dfrm_knee_tear_08_l {parent=dfrm_mstr_knee_tear_08_l, loc=<0 0 0>, rot=<0 0 0 1>}
 dfrm_mstr_knee_tear_09_l {parent=dfrm_knee_prnt_l, loc=<2.3506 0 0>, rot=<0.0147 0.0033 0.1667 -0.9859>}
 dfrm_knee_tear_09_l {parent=dfrm_mstr_knee_tear_09_l, loc=<0 0 0>, rot=<0 0 0 1>}
 dfrm_mstr_knee_tear_10_l {parent=dfrm_knee_prnt_l, loc=<3.1715 0.555 1.9912>, rot=<0.1741 0.081 0.082 -0.978>}
 dfrm_knee_tear_10_l {parent=dfrm_mstr_knee_tear_10_l, loc=<0 0 0>, rot=<0 0 0 1>}
 dfrm_knee_tear_07_l {parent=calf_twist_02_l, loc=<5.924 3.1589 -3.0689>, rot=<0.9657 0.1035 0.0436 0.2341>}
 dfrm_knee_tear_06_l {parent=calf_twist_02_l, loc=<7.3933 -1.686 4.7035>, rot=<0.7577 0.2883 0.0672 -0.5816>}
 calf_twist_01_l {parent=calf_l, loc=<-35.813 -0.0967 0.0397>, rot=<0.0237 0.0095 -0.0035 -0.9997>}
 dfrm_ankle_tear_01_l {parent=calf_twist_01_l, loc=<1.4136 -4.8933 -2.6493>, rot=<0.6477 -0.7591 0.0646 -0.0084>}
 dfrm_ankle_tear_02_l {parent=calf_twist_01_l, loc=<-2.2347 -4.2043 -2.5595>, rot=<0.5061 -0.8574 0.0689 -0.0627>}
 dfrm_ankle_tear_03_l {parent=calf_twist_01_l, loc=<-5.6209 -3.5229 -2.6613>, rot=<0.5758 -0.8045 0.1422 0.0333>}
 dfrm_ankle_tear_04_l {parent=calf_twist_01_l, loc=<1.3094 -5.1374 1.6541>, rot=<0.5493 -0.8351 0.004 0.028>}
 dfrm_ankle_tear_05_l {parent=calf_twist_01_l, loc=<-2.616 -4.5303 2.1392>, rot=<0.5014 -0.8516 -0.1172 0.098>}
 dfrm_ankle_tear_06_l {parent=calf_twist_01_l, loc=<-6.4265 -3.7893 3.3162>, rot=<0.4619 -0.7987 -0.3643 0.1262>}
 deform_knee_in_base_l {parent=calf_l, loc=<1.298 -5.2957 -0.0496>, rot=<0 0 0 -1>}
 deform_knee_in_l {parent=deform_knee_in_base_l, loc=<0 0 0>, rot=<0 0 0 1>}
 deform_knee_out_base_l {parent=calf_l, loc=<0 4 0>, rot=<0 0 0 -1>}
 deform_knee_out_l {parent=deform_knee_out_base_l, loc=<0 0 0>, rot=<0 0 0 1>}
 thigh_twist_01_l {parent=thigh_l, loc=<-21.6303 0 0>, rot=<0.0013 0 0 -1>}
 dfrm_knee_tear_04_l {parent=thigh_twist_01_l, loc=<-8.3138 -0.3285 -5.0905>, rot=<0.188 -0.7127 -0.6212 -0.2661>}
 dfrm_knee_tear_03_l {parent=thigh_twist_01_l, loc=<-8.2577 4.0956 -4.2635>, rot=<-0.1678 0.5136 0.8391 -0.0627>}
 dfrm_knee_tear_02_l {parent=thigh_twist_01_l, loc=<-9.1367 6.5275 2.0098>, rot=<-0.0571 -0.0426 0.9904 0.1185>}
 dfrm_knee_tear_01_l {parent=thigh_twist_01_l, loc=<-10.2035 1.6118 5.8418>, rot=<0.1395 -0.5943 0.7897 0.061>}
 deform_groin_base_l {parent=thigh_l, loc=<-2.1908 8.1142 -1.7442>, rot=<0 0 0 -1>}
 deform_groin_l {parent=deform_groin_base_l, loc=<0 0 0>, rot=<0 0 0 1>}
 dfrm_mstr_knee_tear_05_l {parent=thigh_l, loc=<-36.6816 -3.2146 -3.9193>, rot=<-0.0409 -0.793 -0.3249 -0.5136>}
 dfrm_knee_tear_05_l {parent=dfrm_mstr_knee_tear_05_l, loc=<0 0 0>, rot=<0 0 0 1>}
 dfrm_knee_tear_11_l {parent=thigh_l, loc=<-40.7987 -1.8401 6.2185>, rot=<0.084 -0.8081 0.5012 0.298>}
 thigh_r {parent=pelvis, loc=<-2.5693 -0.0109 7.7501>, rot=<0.0572 -0.0012 0.9982 0.0155>}
 calf_r {parent=thigh_r, loc=<39.0484 0 0>, rot=<0.0006 -0.0161 -0.0391 -0.9991>}
 foot_r {parent=calf_r, loc=<42.561 -0.1077 -0.4292>, rot=<0.0418 -0.0455 0.0343 -0.9975>}
 ball_r {parent=foot_r, loc=<6.1599 -13.8264 -0.6806>, rot=<0.0001 -0.0001 -0.7063 -0.7079>}
 calf_twist_02_r {parent=calf_r, loc=<18.8064 0.0967 -0.0397>, rot=<0.0237 0.0095 -0.0035 -0.9997>}
 calf_twist_01_r {parent=calf_r, loc=<35.813 0.0967 -0.0397>, rot=<0.0237 0.0095 -0.0035 -0.9997>}
 dfrm_ankle_tear_01_r {parent=calf_twist_01_r, loc=<-2.5575 5.2575 -1.7987>, rot=<0.8581 0.5126 0.0129 0.029>}
 dfrm_ankle_tear_02_r {parent=calf_twist_01_r, loc=<4.4545 4.7919 -2.2391>, rot=<0.826 0.5601 -0.0275 0.0567>}
 dfrm_ankle_tear_03_r {parent=calf_twist_01_r, loc=<7.8829 3.7285 -3.6642>, rot=<0.8526 0.3264 -0.0804 -0.4>}
 dfrm_ankle_tear_04_r {parent=calf_twist_01_r, loc=<1.6679 4.364 2.2711>, rot=<0.7932 0.586 0.0663 -0.152>}
 dfrm_ankle_tear_05_r {parent=calf_twist_01_r, loc=<5.2901 4.4033 2.3389>, rot=<0.8883 0.4019 0.0069 0.2219>}
 dfrm_ankle_tear_06_r {parent=calf_twist_01_r, loc=<9.3724 1.8962 3.0052>, rot=<0.9255 0.362 -0.0129 0.1104>}
 deform_knee_in_base_r {parent=calf_r, loc=<-1.298 5.2957 0.0496>, rot=<0 0 0 -1>}
 deform_knee_in_r {parent=deform_knee_in_base_r, loc=<0 0 0>, rot=<0 0 0 1>}
 deform_knee_out_base_r {parent=calf_r, loc=<0 -4 0>, rot=<0 0 0 -1>}
 deform_knee_out_r {parent=deform_knee_out_base_r, loc=<0 0 0>, rot=<0 0 0 1>}
 thigh_twist_01_r {parent=thigh_r, loc=<21.6303 0 0>, rot=<0.0013 0 0 -1>}
 dfrm_mstr_loop_01_r {parent=thigh_r, loc=<9.7345 -8.0431 4.3619>, rot=<0 0 0 1>}
 dfrm_loop_01_r {parent=dfrm_mstr_loop_01_r, loc=<0 0 0>, rot=<0 0 0 1>}
 deform_groin_base_r {parent=pelvis, loc=<-4.2941 8.1593 9.7386>, rot=<0.0572 -0.0012 0.9982 0.0155>}
 deform_groin_r {parent=deform_groin_base_r, loc=<0 0 0>, rot=<0 0 0 1>}
 deform_glute_base_l {parent=pelvis, loc=<-8.8064 -8.8222 -8.4287>, rot=<0.0275 0.0501 -0.4472 -0.8926>}
 deform_glute_l {parent=deform_glute_base_l, loc=<0 0 0>, rot=<0 0 0 1>}
 deform_glute_base_r {parent=pelvis, loc=<-8.8062 -8.8222 8.4287>, rot=<0.0501 -0.0275 0.8926 -0.4472>}
 deform_glute_r {parent=deform_glute_base_r, loc=<0 0 0>, rot=<0 0 0 1>}
 dyn_mstr_loop_front_l {parent=pelvis, loc=<0.7018 7.5353 -8.6883>, rot=<-0.0239 -0.8254 -0.5346 -0.1801>}
 dyn_loop_front_01_l {parent=dyn_mstr_loop_front_l, loc=<0 0 0>, rot=<0 0 0 1>}
 dyn_mstr_loop_back_l {parent=pelvis, loc=<0.2637 -9.5109 -9.9824>, rot=<-0.1455 -0.8767 0.4092 -0.2066>}
 dyn_loop_back_01_l {parent=dyn_mstr_loop_back_l, loc=<0 0 0>, rot=<0 0 0 1>}
 dyn_mstr_loop_l {parent=pelvis, loc=<0.4275 -0.1183 -8.6883>, rot=<-0.1061 0.6581 0.7216 0.1867>}
 dyn_loop_01_l {parent=dyn_mstr_loop_l, loc=<0 0 0>, rot=<0 0 0 1>}
 dyn_loop_02_l {parent=dyn_loop_01_l, loc=<11 0 0>, rot=<0 0 0 1>}
 dyn_loop_03_l {parent=dyn_loop_02_l, loc=<11 0 0>, rot=<0 0 0 1>}
 dyn_front_chain_mstr_l {parent=pelvis, loc=<0.7018 7.5353 -8.6883>, rot=<-0.0237 0.6327 0.7508 0.1884>}
 dyn_front_chain_01_l {parent=dyn_front_chain_mstr_l, loc=<0 0 0>, rot=<0 0 0 1>}
 dyn_front_chain_02_l {parent=dyn_front_chain_01_l, loc=<4.6067 0 0>, rot=<0.0062 -0.0193 0.0148 -0.9997>}
 dyn_front_chain_03_l {parent=dyn_front_chain_02_l, loc=<3.8795 0 0>, rot=<-0.0162 -0.0979 -0.0243 -0.9948>}
 dyn_front_chain_04_l {parent=dyn_front_chain_03_l, loc=<3.9751 0 0>, rot=<-0.0205 -0.0881 0.1148 -0.9893>}
 dyn_front_chain_05_l {parent=dyn_front_chain_04_l, loc=<4.0095 0 0>, rot=<0.0911 -0.2114 0.0044 -0.9731>}
 dyn_front_chain_06_l {parent=dyn_front_chain_05_l, loc=<4.0364 0 0>, rot=<-0.0651 -0.1685 -0.0397 -0.9827>}
 dyn_front_chain_07_l {parent=dyn_front_chain_06_l, loc=<4.0162 0 0>, rot=<0.1009 -0.1995 -0.0475 -0.9735>}
 dyn_front_chain_08_l {parent=dyn_front_chain_07_l, loc=<2.0172 0 0>, rot=<0.1178 -0.094 -0.0999 -0.9835>}
 dyn_back_chain_mstr_l {parent=pelvis, loc=<0.2637 -9.5109 -9.9824>, rot=<-0.1423 0.6709 0.6952 0.2154>}
 dyn_back_chain_01_l {parent=dyn_back_chain_mstr_l, loc=<0 0 0>, rot=<0 0 0 1>}
 dyn_back_chain_02_l {parent=dyn_back_chain_01_l, loc=<4.6359 0 0>, rot=<-0.0331 0.0602 -0.0512 -0.9963>}
 dyn_back_chain_03_l {parent=dyn_back_chain_02_l, loc=<3.7588 0 0>, rot=<-0.0022 0.0384 -0.0618 -0.9973>}
 dyn_back_chain_04_l {parent=dyn_back_chain_03_l, loc=<3.9634 0 0>, rot=<0.0537 0.0376 0.0281 -0.9975>}
 dyn_back_chain_05_l {parent=dyn_back_chain_04_l, loc=<3.9798 0 0>, rot=<-0.0138 0.1083 0.0158 -0.9939>}
 dyn_back_chain_06_l {parent=dyn_back_chain_05_l, loc=<3.9845 0 0>, rot=<-0.0211 0.1322 0.0115 -0.9909>}
 dyn_back_chain_07_l {parent=dyn_back_chain_06_l, loc=<3.9514 0 0>, rot=<-0.0196 0.35 0.0219 -0.9363>}
 dyn_back_chain_08_l {parent=dyn_back_chain_07_l, loc=<2.0172 0 0>, rot=<-0.0341 0.0068 -0.0294 -0.999>}
 dyn_chain_end_ctrl_null_l {parent=pelvis, type=Null, loc=<-19.5259 1.8891 -18.24>, rot=<-0.4318 -0.4166 -0.5555 -0.5757>}
 dyn_chain_end_l_ctrl {controlType=EULER_TRANSFORM, parent=dyn_chain_end_ctrl_null_l, type=Control, loc=<0 0 0>, rot=<0 0 0 1>}
 dyn_front_chain_07_l_ctrl {controlType=EULER_TRANSFORM, parent=dyn_chain_end_l_ctrl, type=Control, loc=<17.909 2.4443 73.8643>, rot=<-0.7463 0.6459 -0.0286 0.1582>}
 dyn_back_chain_07_l_ctrl {controlType=EULER_TRANSFORM, parent=dyn_chain_end_l_ctrl, type=Control, loc=<0 0 0>, rot=<0 0 0 1>}
 dyn_front_chain_01_l_ctrl_null {parent=pelvis, type=Null, loc=<0.7018 7.5353 -8.6883>, rot=<0.0237 -0.6327 -0.7508 -0.1884>}
 dyn_front_chain_01_l_ctrl {controlType=EULER_TRANSFORM, parent=dyn_front_chain_01_l_ctrl_null, type=Control, loc=<0 0 0>, rot=<0 0 0 1>}
 dyn_front_chain_05_l_ctrl_null {parent=pelvis, type=Null, loc=<-14.3611 9.0985 -14.3734>, rot=<0.3962 -0.6114 -0.6844 0.0288>}
 dyn_front_chain_05_l_ctrl {controlType=EULER_TRANSFORM, parent=dyn_front_chain_05_l_ctrl_null, type=Control, loc=<0 0 0>, rot=<0 0 0 1>}
 dyn_front_chain_03_l_ctrl_null {parent=pelvis, type=Null, loc=<-7.1623 9.5525 -11.1498>, rot=<0.1035 -0.6595 -0.7368 -0.1069>}
 dyn_front_chain_03_l_ctrl {controlType=EULER_TRANSFORM, parent=dyn_front_chain_03_l_ctrl_null, type=Control, loc=<0 0 0>, rot=<0 0 0 1>}
 dyn_back_chain_01_l_ctrl_null {parent=pelvis, type=Null, loc=<0.2637 -9.5109 -9.9824>, rot=<0.1423 -0.6709 -0.6952 -0.2154>}
 dyn_back_chain_01_l_ctrl {controlType=EULER_TRANSFORM, parent=dyn_back_chain_01_l_ctrl_null, type=Control, loc=<0 0 0>, rot=<0 0 0 1>}
 dyn_back_chain_03_l_ctrl_null {parent=pelvis, type=Null, loc=<-7.1099 -8.1459 -13.6976>, rot=<-0.0112 -0.6804 -0.7028 -0.2074>}
 dyn_back_chain_03_l_ctrl {controlType=EULER_TRANSFORM, parent=dyn_back_chain_03_l_ctrl_null, type=Control, loc=<0 0 0>, rot=<0 0 0 1>}
 dyn_back_chain_05_l_ctrl_null {parent=pelvis, type=Null, loc=<-14.2023 -5.4223 -15.9859>, rot=<-0.0799 -0.6124 -0.7133 -0.3312>}
 dyn_back_chain_05_l_ctrl {controlType=EULER_TRANSFORM, parent=dyn_back_chain_05_l_ctrl_null, type=Control, loc=<0 0 0>, rot=<0 0 0 1>}
 dyn_mstr_loop_r {parent=pelvis, loc=<1.5303 -0.1578 9.8553>, rot=<0.0827 -0.6715 0.7234 0.1378>}
 dyn_loop_01_r {parent=dyn_mstr_loop_r, loc=<0 0 0>, rot=<0 0 0 1>}
 dyn_loop_02_r {parent=dyn_loop_01_r, loc=<11 0 0>, rot=<0 0 0 1>}
 dyn_loop_03_r {parent=dyn_loop_02_r, loc=<11 0 0>, rot=<0 0 0 1>}
 dyn_front_chain_mstr_r {parent=pelvis, loc=<1.1464 7.8067 8.4592>, rot=<0.0777 -0.7211 0.6706 0.1559>}
 dyn_front_chain_01_r {parent=dyn_front_chain_mstr_r, loc=<0 0 0>, rot=<0 0 0 1>}
 dyn_front_chain_02_r {parent=dyn_front_chain_01_r, loc=<4.2444 0 0>, rot=<-0.0002 0.0132 0.018 -0.9998>}
 dyn_front_chain_03_r {parent=dyn_front_chain_02_r, loc=<4.2693 0 0>, rot=<-0.0022 0.0552 0.0031 -0.9985>}
 dyn_front_chain_04_r {parent=dyn_front_chain_03_r, loc=<4.2647 0 0>, rot=<-0.0019 0.0341 0.012 -0.9993>}
 dyn_front_chain_05_r {parent=dyn_front_chain_04_r, loc=<4.2667 0 0>, rot=<-0.0149 0.1834 0.0144 -0.9828>}
 dyn_front_chain_06_r {parent=dyn_front_chain_05_r, loc=<4.2061 0 0>, rot=<0.6948 0 0 -0.7192>}
 dyn_front_chain_07_r {parent=dyn_front_chain_06_r, loc=<0 0 0>, rot=<-0.4884 0.1698 0.3596 -0.7767>}
 dyn_front_chain_08_r {parent=dyn_front_chain_07_r, loc=<4.2322 0 0>, rot=<0 0 0 1>}
 dyn_back_chain_mstr_r {parent=pelvis, loc=<1.1877 -9.715 9.8553>, rot=<0.1648 -0.3455 0.9173 0.1097>}
 dyn_back_chain_01_r {parent=dyn_back_chain_mstr_r, loc=<0 0 0>, rot=<0 0 0 1>}
 dyn_back_chain_02_r {parent=dyn_back_chain_01_r, loc=<4.2393 0 0>, rot=<0.0035 -0.0816 0.0428 -0.9957>}
 dyn_back_chain_03_r {parent=dyn_back_chain_02_r, loc=<4.261 0 0>, rot=<-0.036 -0.0721 0.0447 -0.9957>}
 dyn_back_chain_04_r {parent=dyn_back_chain_03_r, loc=<4.3304 0 0>, rot=<0.0861 -0.0015 0.0366 -0.9956>}
 dyn_back_chain_05_r {parent=dyn_back_chain_04_r, loc=<4.3206 0 0>, rot=<-0.101 -0.0371 0.0664 -0.992>}
 dyn_back_chain_06_r {parent=dyn_back_chain_05_r, loc=<4.2668 0 0>, rot=<-0.0337 -0.167 0.1848 -0.9679>}
 dyn_back_chain_07_r {parent=dyn_back_chain_06_r, loc=<4.2306 0 0>, rot=<0 0 0 1>}
 dyn_front_chain_01_r_ctrl_null {parent=pelvis, type=Null, loc=<1.1464 7.8067 8.4592>, rot=<-0.0777 0.7211 -0.6706 -0.1559>}
 dyn_front_chain_01_r_ctrl {controlType=EULER_TRANSFORM, parent=dyn_front_chain_01_r_ctrl_null, type=Control, loc=<0 0 0>, rot=<0 0 0 1>}
 dyn_front_chain_03_r_ctrl_null {parent=pelvis, type=Null, loc=<-6.8109 8.5075 11.3981>, rot=<0.1388 -0.7273 0.6613 0.1202>}
 dyn_front_chain_03_r_ctrl {controlType=EULER_TRANSFORM, parent=dyn_front_chain_03_r_ctrl_null, type=Control, loc=<0 0 0>, rot=<0 0 0 1>}
 dyn_front_chain_05_r_ctrl_null {parent=pelvis, type=Null, loc=<-14.7171 7.8435 14.519>, rot=<0.2996 -0.7222 0.6229 -0.0254>}
 dyn_front_chain_05_r_ctrl {controlType=EULER_TRANSFORM, parent=dyn_front_chain_05_r_ctrl_null, type=Control, loc=<0 0 0>, rot=<0 0 0 1>}
 dyn_back_chain_01_r_ctrl_null {parent=pelvis, type=Null, loc=<1.1877 -9.715 9.8553>, rot=<-0.1648 0.3455 -0.9173 -0.1097>}
 dyn_back_chain_01_r_ctrl {controlType=EULER_TRANSFORM, parent=dyn_back_chain_01_r_ctrl_null, type=Control, loc=<0 0 0>, rot=<0 0 0 1>}
 dyn_back_chain_03_r_ctrl_null {parent=pelvis, type=Null, loc=<-6.6209 -8.2463 12.7726>, rot=<-0.058 0.2792 -0.9285 -0.2378>}
 dyn_back_chain_03_r_ctrl {controlType=EULER_TRANSFORM, parent=dyn_back_chain_03_r_ctrl_null, type=Control, loc=<0 0 0>, rot=<0 0 0 1>}
 dyn_back_chain_05_r_ctrl_null {parent=pelvis, type=Null, loc=<-14.0755 -4.4878 15.0179>, rot=<-0.0642 0.2492 -0.9039 -0.3417>}
 dyn_back_chain_05_r_ctrl {controlType=EULER_TRANSFORM, parent=dyn_back_chain_05_r_ctrl_null, type=Control, loc=<0 0 0>, rot=<0 0 0 1>}
 dyn_chain_end_r_ctrl_null {parent=pelvis, type=Null, loc=<-19.0908 1.8258 16.6647>, rot=<0.6071 -0.7143 0.3154 -0.1474>}
 dyn_chain_end_r_ctrl {controlType=EULER_TRANSFORM, parent=dyn_chain_end_r_ctrl_null, type=Control, loc=<0 0 0>, rot=<0 0 0 1>}
 dyn_front_chain_07_r_ctrl {controlType=EULER_TRANSFORM, parent=dyn_chain_end_r_ctrl, type=Control, loc=<0 0 0>, rot=<0 0 0 1>}
 dyn_back_chain_07_r_ctrl {controlType=EULER_TRANSFORM, parent=dyn_chain_end_r_ctrl, type=Control, loc=<0 0 0>, rot=<0 0 0 1>}
 attach {parent=root, loc=<0 0 0>, rot=<0 0 0 1>}
 ik_hand_root {parent=root, loc=<0 0 0>, rot=<0 0 0 1>}
 ik_hand_gun {parent=ik_hand_root, loc=<-51.9727 4.0883 102.6935>, rot=<0.6084 0.1793 0.4613 -0.6205>}
 ik_hand_r {parent=ik_hand_gun, loc=<0 0 0>, rot=<0 0 0 1>}
 ik_hand_l {parent=ik_hand_gun, loc=<53.0285 82.1781 35.2051>, rot=<-0.5102 -0.7906 -0.3387 0>}
 ik_foot_root {parent=root, loc=<0 0 0>, rot=<0 0 0 1>}
 ik_foot_l {parent=ik_foot_root, loc=<15.2608 -4.4176 9.4432>, rot=<0.0357 0.7045 0.0284 -0.7082>}
 ik_foot_r {parent=ik_foot_root, loc=<-15.2608 -4.4176 9.4432>, rot=<0.7082 0.0284 -0.7045 0.0357>}
 attach_fp {parent=root, loc=<0 0 0>, rot=<0 0 0 1>}
 camera_root {parent=root, loc=<0 0 0>, rot=<0 0 0 1>}
 camera {parent=camera_root, loc=<0 0 0>, rot=<0 0 0 1>}
 warp_point_01 {parent=root, loc=<0 0 0>, rot=<0 0 0 1>}
 warp_point_02 {parent=root, loc=<0 0 0>, rot=<0 0 0 1>}
 warp_point_03 {parent=root, loc=<0 0 0>, rot=<0 0 0 1>}
]
variables [
 SplineIK_back_chain_l_ctrls {
  type="TArray<FRigElementKey>"
  default=((Type=Control,Name="dyn_back_chain_01_l_ctrl"),(Type=Control,Name="dyn_back_chain_03_l_ctrl"),(Type=Control,Name="dyn_back_chain_05_l_ctrl"),(Type=Control,Name="dyn_back_chain_07_l_ctrl"),(Type=Control,Name="dyn_chain_end_l_ctrl"))
 }
 SplineIK_back_chain_r_ctrls {
  type="TArray<FRigElementKey>"
  default=((Type=Control,Name="dyn_back_chain_01_r_ctrl"),(Type=Control,Name="dyn_back_chain_03_r_ctrl"),(Type=Control,Name="dyn_back_chain_05_r_ctrl"),(Type=Control,Name="dyn_back_chain_07_r_ctrl"),(Type=Control,Name="dyn_chain_end_r_ctrl"))
 }
 SplineIK_chain_l_ctrls {
  type="TArray<FRigElementKey>"
  default="((Type=Control,Name=\"dyn_front_chain_01_l_ctrl\"),(Type=Control,Name=\"dyn_front_chain_03_l_ctrl\"),(Type=Control,Name=\"dyn_front_chain_05_l_ctrl\"),(Type=Control,Name=\"dyn_front_chain_07_l_ctrl\"),(Type=Control,Name=\"dyn_back_chain_01_l_ctrl\"),(Type=Control,Na ...(truncated)"
 }
 SplineIK_chain_r_ctrls {
  type="TArray<FRigElementKey>"
  default="((Type=Control,Name=\"dyn_front_chain_01_r_ctrl\"),(Type=Control,Name=\"dyn_front_chain_03_r_ctrl\"),(Type=Control,Name=\"dyn_front_chain_05_r_ctrl\"),(Type=Control,Name=\"dyn_front_chain_07_r_ctrl\"),(Type=Control,Name=\"dyn_back_chain_01_r_ctrl\"),(Type=Control,Na ...(truncated)"
 }
 SplineIK_ctrl_r_crouch_walk_left_pose {type=FRigPose, default=(Elements=,HierarchyTopologyVersion=-1,PoseHash=-1)}
 SplineIK_ctrl_thigh_l_up_pose {type=FRigPose, default=(Elements=,HierarchyTopologyVersion=-1,PoseHash=-1)}
 SplineIK_ctrl_thigh_r_leg_slight_back_out_pose {type=FRigPose, default=(Elements=,HierarchyTopologyVersion=-1,PoseHash=-1)}
 SplineIK_ctrl_thigh_r_up_pose {type=FRigPose, default=(Elements=,HierarchyTopologyVersion=-1,PoseHash=-1)}
 SplineIK_ctrls_l_bind_pose {type=FRigPose, default=(Elements=,HierarchyTopologyVersion=-1,PoseHash=-1)}
 SplineIK_ctrls_r_bind_pose {type=FRigPose, default=(Elements=,HierarchyTopologyVersion=-1,PoseHash=-1)}
 SplineIK_front_chain_l_ctrls {
  type="TArray<FRigElementKey>"
  default=((Type=Control,Name="dyn_front_chain_01_l_ctrl"),(Type=Control,Name="dyn_front_chain_03_l_ctrl"),(Type=Control,Name="dyn_front_chain_05_l_ctrl"),(Type=Control,Name="dyn_front_chain_07_l_ctrl"),(Type=Control,Name="dyn_chain_end_l_ctrl"))
 }
 SplineIK_front_chain_r_ctrls {
  type="TArray<FRigElementKey>"
  default=((Type=Control,Name="dyn_front_chain_01_r_ctrl"),(Type=Control,Name="dyn_front_chain_03_r_ctrl"),(Type=Control,Name="dyn_front_chain_05_r_ctrl"),(Type=Control,Name="dyn_front_chain_07_r_ctrl"),(Type=Control,Name="dyn_chain_end_r_ctrl"))
 }
 is_front_end {type=bool, default=False}
 spr_thigh_l_up_alpha {type=double, default=0}
 spr_thigh_r_slight_back_out_apha {type=double, default=0}
 spr_thigh_r_up_alpha {type=double, default=0}
]
