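# Physics asset for the M_MED_Inferno male medium body cosmetic, covering its dynamic back flap and left and right loop bones.
# Each loop is a two-bone capsule chain constrained to a small master bone, and the flap is a single bone on its own master, all with tight twist limits.
# Overlapping capsules on pelvis and thighs form unconstrained body volumes, with noticeably bulkier shapes on the right thigh and loop.
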PhysicsAsset /BRCosmetics/Characters/Player/Male/Medium/Bodies/M_MED_Inferno/Meshes/M_MED_Inferno_Physics
bodies {
 dyn_flap_back_01 [{shapeType=Capsule, length=3.7237, radius=1.6156, shapeName=dyn_flap_front_01_capsule, center=<6.5 0 1.5>, rotation=<rot 0 0 90>}]
 dyn_loop_01_l [{shapeType=Capsule, length=21.5984, radius=1.4688, shapeName=dyn_loop_01_l_capsule, center=<11 -5 0>}]
 dyn_loop_01_r [{
   shapeType=Capsule
   length=24.1807
   radius=3.6526
   shapeName=dyn_loop_01_l_capsule
   center=<11 -6.4 0.5599>
   rotation=<rot -0.4361 4.981 4.9813>
  }]
 dyn_loop_02_l [{shapeType=Capsule, length=17.462, radius=1.4688, shapeName=dyn_loop_01_l_capsule, center=<11 -0.9684 0>}]
 dyn_loop_02_r [
  {shapeType=Capsule, length=23.0083, radius=4.825, shapeName=dyn_loop_01_l_capsule, center=<11 -4.5 0>, rotation=<rot -0.0007 -0.0001 5.0002>}
  {
   shapeType=Capsule
   length=23.0693
   radius=5.2255
   shapeName=dyn_loop_01_l_capsule
   center=<10.9998 -0.5841 -15.172>
   rotation=<rot -0.0011 -0.0001 -29.9997>
  }
 ]
 dyn_mstr_flap_back [{shapeType=Capsule, length=0.505, radius=0.505, shapeName=dyn_mstr_flap_back_capsule}]
 dyn_mstr_loop_l [{shapeType=Capsule, length=0, radius=0.5861, shapeName=dyn_mstr_loop_l_capsule}]
 dyn_mstr_loop_r [{shapeType=Capsule, length=0, radius=0.5, shapeName=dyn_mstr_loop_r_capsule}]
 pelvis [
  {shapeType=Capsule, length=10.6628, radius=10.176, shapeName=pelvis_capsule, center=<0 -1.8814 0>}
  {
   shapeType=Capsule
   length=22.6393
   radius=7.1665
   shapeName=thigh_l_capsule
   center=<-13.8083 0 -9.4452>
   rotation=<rot -88.795 -0.0013 -179.9989>
  }
  {
   shapeType=Capsule
   length=14.25
   radius=13.3535
   shapeName=pelvis_capsule
   center=<-11.5338 0.0001 3.4338>
   rotation=<rot 90 -0.0796 -0.0785>
  }
  {shapeType=Capsule, length=15, radius=12.6035, shapeName=pelvis_capsule, center=<-12.1173 0.0001 -1.9887>, rotation=<rot 80 -0.0009 0>}
 ]
 thigh_l [
  {
   shapeType=Capsule
   length=22.6393
   radius=7.1665
   shapeName=thigh_l_capsule
   center=<-11.025 0 0.149>
   rotation=<rot -88.795 -0.0013 -179.9989>
  }
  {
   shapeType=Capsule
   length=22.6393
   radius=7.1665
   shapeName=thigh_l_capsule
   center=<-10.5774 -7.4395 1.8049>
   rotation=<rot -88.795 -0.002 165.0017>
  }
 ]
 thigh_r [
  {
   shapeType=Capsule
   length=18.386
   radius=12.8183
   shapeName=thigh_l_capsule
   center=<11.0194 -1.171 -5.7085>
   rotation=<rot -83.7259 127.1077 -127.274>
  }
  {
   shapeType=Capsule
   length=13.7884
   radius=16.2116
   shapeName=thigh_r_capsule
   center=<11.1884 9.3173 -9.4207>
   rotation=<rot 4.9808 29.6203 -83.6821>
  }
 ]
}
constraints [
 {bone1Name=dyn_flap_back_01, bone2Name=dyn_mstr_flap_back, swing1LimitDegrees=2, swing2LimitDegrees=25, twistLimitDegrees=3}
 {bone1Name=dyn_loop_02_l, bone2Name=dyn_loop_01_l, swing1LimitDegrees=35, swing2LimitDegrees=10, twistLimitDegrees=2}
 {bone1Name=dyn_loop_01_l, bone2Name=dyn_mstr_loop_l, swing1LimitDegrees=35, swing2LimitDegrees=5, twistLimitDegrees=2}
 {bone1Name=dyn_loop_02_r, bone2Name=dyn_loop_01_r, swing1LimitDegrees=35, swing2LimitDegrees=15, twistLimitDegrees=2}
 {bone1Name=dyn_loop_01_r, bone2Name=dyn_mstr_loop_r, swing1LimitDegrees=25, swing2LimitDegrees=10, twistLimitDegrees=2}
]
